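# Reusable text label widget in the player widgets folder, made of a text block inside a colored border and a clipped size box.
# PreConstruct runs SetUpBox, which applies the configured box size, background color, padding and horizontal alignment, then calls RefreshText.
# RefreshText sets text and color and rebuilds the font with the chosen size and a Regular or Bold typeface.
# UpdateText stores new text and refreshes it with the current style.
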WidgetBlueprint /Game/Characters/Player/Widgets/WB_FAnim_SimpleTextBlock
graphs [
 EventGraph {nodes [
   Event PreConstruct()
   > SetUpBox!(execute, self:Self Object Reference) -> (then)
  ]}
 RefreshText {nodes [
   RefreshText!() -> (then, Text:Text, Color:Linear Color Structure=(R=0.5,G=0.5,B=0.5,A=1), IsBold:bool=~) {is=title K2Node_FunctionEntry}
   > "SetText (Text)"(self=Get TextBox() -> (TextBox:Text Object Reference), InText=Get Text!() -> (Text))
   > SetColorAndOpacity!(
    execute
    self=Get TextBox()
    InColorAndOpacity_SpecifiedColor:Linear Color Structure=Get Color() -> (Color:Linear Color Structure)
    InColorAndOpacity_ColorUseRule:ESlateColorStylingMode Enum
   ) -> (then)
   n1 = Get TextBox()
   n2 = Get!(self:Text Object Reference=@n1) -> (
    Font_FontObject:Object Reference
    Font_FontMaterial:Object Reference
    Font_OutlineSettings:Font Outline Settings Structure
    Font_TypefaceFontName:Name
    Font_Size:float=~
    Font_LetterSpacing:int
    Font_SkewAmount:float=~
    Font_bForceMonospaced:bool=~
    Font_bMaterialIsStencil:bool=~
    Font_MonospacedWidth:float=~
   )
   > SetFont!(
    execute
    self=@n1
    InFontInfo_FontObject:Object Reference=@n2
    InFontInfo_FontMaterial:Object Reference=@n2.Font_FontMaterial
    InFontInfo_OutlineSettings:Font Outline Settings Structure=@n2.Font_OutlineSettings
    InFontInfo_TypefaceFontName:Name=Select(Option 0:Name=Regular, Option 1:Name=Bold, Index=Get IsBold() -> (IsBold:bool)) -> (ReturnValue:Name)
    InFontInfo_Size:float=Get FontSize() -> (FontSize:double)
    InFontInfo_LetterSpacing:int=@n2.Font_LetterSpacing
    InFontInfo_SkewAmount:float=0
    InFontInfo_bForceMonospaced:bool=false
    InFontInfo_bMaterialIsStencil:bool=false
    InFontInfo_MonospacedWidth:float=1
   ) -> (then)
  ]}
 SetUpBox {nodes [
   SetUpBox() {is=title K2Node_FunctionEntry}
   > SetPadding!(
    execute
    self:Border Slot Object Reference=SlotAsBorderSlot!(Widget:Widget Object Reference=Get TextBox()) -> (ReturnValue:Border Slot Object Reference)
    InPadding:Margin Structure=Get TextPadding() -> (TextPadding:Margin Structure)
   ) -> (then)
   > SetHorizontalAlignment(
    self:Border Slot Object Reference=SlotAsBorderSlot(Widget=Get TextBox())
    InHorizontalAlignment=Get TextAlignment() -> (TextAlignment:EHorizontalAlignment Enum=HAlign_Fill)
   )
   > RefreshText!(
    execute
    self
    Text:Text=Get Text()
    Color:Linear Color Structure=Get TextColor() -> (TextColor:Linear Color Structure)
    IsBold:bool=Get IsBold()
   ) -> (then) {is=title K2Node_CallFunction}
   > SetWidthOverride(self=Get SizeBox() -> (SizeBox:Size Box Object Reference), InWidthOverride=Get!() -> (BoxSize_X:double, BoxSize_Y:double))
   > SetHeightOverride(self=Get SizeBox(), InHeightOverride=Get().BoxSize_Y)
   > SetBrushColor(self=Get Border() -> (Border:Border Object Reference), InBrushColor=Get BoxColor() -> (BoxColor:Linear Color Structure))
  ]}
 UpdateText {nodes [
   n1 = UpdateText!() -> (then, Text:Text) {is=title K2Node_FunctionEntry}
   > n2 = Set Text!(execute, Text=@n1) -> (then, Output_Get)
   > RefreshText(Text=@n2, Color=Get TextColor(), IsBold=Get IsBold())
  ]}
]
variables [
 BoxColor {type=LinearColor, default=<rgba 0.0157 0.0157 0.0157 1>}
 BoxSize {type=Vector2D, default=<100 25>}
 FontSize {type=number, default=8}
 IsBold {type=boolean}
 Text {type=string, default=Text}
 TextAlignment {type=EHorizontalAlignment, default=HAlign_Left}
 TextColor {type=LinearColor, default=<rgba 0.5 0.5 0.5 1>}
 TextPadding {type=Margin, default=<ltrb 0 0 0 0>}
]
widgets {
 info {inheritedWidgetCount=0, parentClass=&UserWidget, rootWidgetClass=&SizeBox, widgetCount=3}
 widgets [
  SizeBox {
   widgetClassPath=&SizeBox
   bIsVariable=true
   widget=SizeBox
   properties {clipping=ClipToBounds, heightOverride=25, widthOverride=75}
  }
  Border {
   widgetClassPath=&Border
   bIsVariable=true
   parent=SizeBox
   widget=Border
   properties {horizontalAlignment=HAlign_Left, verticalAlignment=VAlign_Center, brushColor=<rgba 0.0176 0.0176 0.0176 0>, padding=<ltrb 0 0 0 0>}
   slot {class=SizeBoxSlot}
  }
  TextBox {
   widgetClassPath=&TextBlock
   bIsVariable=true
   parent=Border
   widget=TextBox
   properties {
    text=Text
    colorAndOpacity {colorUseRule=UseColor_Specified, specifiedColor=<rgba 0.5 0.5 0.5 1>}
    font {
     bForceMonospaced=false
     bMaterialIsStencil=false
     fontMaterial=null
     fontObject=&Roboto
     letterSpacing=0
     monospacedWidth=1
     size=8
     skewAmount=0
     typefaceFontName=Regular
     outlineSettings {
      bApplyOutlineToDropShadows=false
      bMiteredCorners=false
      bSeparateFillAlpha=false
      outlineMaterial=null
      outlineSize=0
      outlineColor=<rgba 0 0 0 1>
     }
    }
    shadowOffset=<0 0>
   }
   slot {class=BorderSlot, properties {horizontalAlignment=HAlign_Left, verticalAlignment=VAlign_Center, padding=<ltrb 0 0 0 0>}}
  }
 ]
}
