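# Editor utility widget that shows a labeled asset name beside a button that opens that asset in its editor.
# PreConstruct builds the label (object name or "None"), sets the button caption and sizes the root SizeBox from the Button Width/Height variables.
# SetAsset forwards only changed assets to OnAssetChanged, which refreshes the label and enables the button only for valid assets.
# Clicking opens the asset through AssetEditorSubsystem, then broadcasts OnClicked.
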EditorUtilityWidgetBlueprint /Game/Characters/Player/Widgets/EUW_FAnim_AssetEditControl
graphs [
 BuildAssetText {nodes [
   BuildAssetText() {is=title K2Node_FunctionEntry}
   n1 = Get Asset() -> (Asset:Object Reference)
   > n2 = Branch(Condition=IsValid(Object=@n1))
   n3 = Get AssetLabelText() -> (AssetLabelText:Text)
   > n4 = Set ResultingText(ResultingText:String=Append!(A="To String (Text)"(InText=@n3), B=": ", C:String=GetObjectName(Object=@n1)) -> (ReturnValue)) -> (Output_Get:String)
   Return Node!(execute=[@n4.then, @n6.then], NewParam:Text="To Text (String)"(InString=Get ResultingText() -> (ResultingText:String)))
   n5 = Get AssetLabelText()
   n6 = Set ResultingText(execute=@n2.else, ResultingText=Append!(A="To String (Text)"(InText=@n5), B=: None) -> (ReturnValue))
  ]}
 EventGraph {nodes [
   "On Clicked (Button)"() {is=title K2Node_ComponentBoundEvent}
   n1 = Get Asset()
   > Branch(Condition=IsValid(Object=@n1))
   > OpenEditorForAssets!(
    execute
    self:Asset Editor Subsystem Object Reference=Get AssetEditorSubsystem()
    Assets:Array of Object References=Make Array("[0]":Object Reference=@n1) -> (Array:Array of Object References)
    OpenedMethod:EAssetTypeActivationOpenedMethod Enum=Edit
   ) -> (then, ReturnValue:bool)
   > Call On Clicked!(execute:Exec, self:EUW FAnim Asset Edit Control Object Reference) -> (then:Exec) {is=title K2Node_CallDelegate}
   Event PreConstruct()
   > OnAssetChanged!(execute, self, NewAsset:Object Reference=Get Asset()) -> (then)
   > n2 = BuildAssetText!(execute, self) -> (then, NewParam:Text) {is=title K2Node_CallFunction}
   > UpdateText!(
    execute
    self:WB FAnim Simple Text Block Object Reference=Get AssetLabel() -> (AssetLabel:WB FAnim Simple Text Block Object Reference)
    Text:Text=@n2
   ) -> (then)
   > "SetText (Text)"(self=Get ButtonText() -> (ButtonText:Text Object Reference), InText=Get ButtonName() -> (ButtonName:Text))
   > SetHeightOverride(self=Get Size() -> (Size:Size Box Object Reference), InHeightOverride=Get Button Height() -> (Button Height:double))
   > SetWidthOverride(self=Get Size(), InWidthOverride=Get Button Width() -> (Button Width:double))
  ]}
 OnAssetChanged {nodes [
   n1 = OnAssetChanged!() -> (then, NewAsset:Object Reference) {is=title K2Node_FunctionEntry}
   > Set Asset(Asset:Object Reference=@n1) -> (Output_Get:Object Reference)
   > Set bIsEnabled!(execute, bIsEnabled:bool=IsValid(Object=@n1), self:Widget Object Reference=Get Button() -> (Button:Editor Utility Button Object Reference)) -> (then, Output_Get:bool)
   > n2 = BuildAssetText()
   > UpdateText(self=Get AssetLabel(), Text=@n2)
   > Return Node!(execute)
  ]}
 OnClicked {nodes [OnClicked() {is=title K2Node_FunctionEntry}]}
 SetAsset {nodes [
   n1 = SetAsset!() -> (then, InAsset:Object Reference) {is=title K2Node_FunctionEntry}
   > n2 = Branch(Condition="Equal (Object)"(A=@n1, B=Get Asset()))
   Return Node(execute=[@n3.then, @n2.then])
   n3 = OnAssetChanged(execute=@n2.else, NewAsset=@n1) {is=title K2Node_CallFunction}
  ]}
]
variables [
 Asset {type=Object, default=null}
 AssetLabelText {type=string, default=Asset}
 Button Height {type=number, default=25}
 Button Width {type=number, default=75}
 ButtonName {type=string, default=Open}
 OnClicked {default=()}
]
widgets {
 info {rootWidgetClass=&SizeBox, widgetCount=5}
 widgets [
  Size {
   widgetClassPath=&SizeBox
   bIsVariable=true
   widget=Size
   properties {bOverride_HeightOverride=true, bOverride_WidthOverride=true, clipping=ClipToBounds, heightOverride=25, widthOverride=75}
  }
  HorizontalBox_0 {widgetClassPath=&HorizontalBox, parent=Size, widget=HorizontalBox_0, slot {class=SizeBoxSlot}}
  AssetLabel {
   widgetClassPath=&/WB_FAnim_SimpleTextBlock
   bIsVariable=true
   parent=HorizontalBox_0
   widget=AssetLabel
   slot {class=HorizontalBoxSlot, properties {size {sizeRule=Fill, value=1}}}
  }
  Button {
   widgetClassPath=&EditorUtilityButton
   parent=HorizontalBox_0
   widget=Button
   properties {
    backgroundColor=<rgba 0.3229 0.3229 0.3229 1>
    widgetStyle {
     clickedSlateSound {resourceObject=null}
     disabled {
      bIsDynamicallyLoaded=false
      drawAs=Box
      imageType=NoImage
      mirroring=NoMirror
      resourceName=None
      resourceObject=null
      tiling=NoTile
      imageSize=<0 0>
      margin=<ltrb 0 0 0 0>
      outlineSettings {
       bUseBrushTransparency=false
       roundingType=FixedRadius
       width=1
       color {colorUseRule=UseColor_ColorTable, specifiedColor=<rgba 0 0 0 0>}
       cornerRadii=<4 4 4 4>
      }
      tintColor {colorUseRule=UseColor_Specified, specifiedColor=<rgba 0.0176 0.0176 0.0176 1>}
      uVRegion {bIsValid=false, max=<0 0>, min=<0 0>}
     }
     disabledForeground {colorUseRule=UseColor_ColorTable, specifiedColor=<rgba 0 0 0 0>}
     hovered {
      bIsDynamicallyLoaded=false
      drawAs=RoundedBox
      imageType=NoImage
      mirroring=NoMirror
      resourceName=None
      resourceObject=null
      tiling=NoTile
      imageSize=<0 0>
      margin=<ltrb 0 0 0 0>
      outlineSettings {
       bUseBrushTransparency=true
       roundingType=FixedRadius
       width=1
       color {colorUseRule=UseColor_ColorTable, specifiedColor=<rgba 0 0 0 0>}
       cornerRadii=<4 4 4 4>
      }
      tintColor {colorUseRule=UseColor_Specified, specifiedColor=<rgba 0.6087 0.6087 0.6087 1>}
      uVRegion {bIsValid=false, max=<0 0>, min=<0 0>}
     }
     hoveredForeground {colorUseRule=UseColor_ColorTable, specifiedColor=<rgba 0 0 0 0>}
     hoveredSlateSound {resourceObject=null}
     normal {
      bIsDynamicallyLoaded=false
      drawAs=RoundedBox
      imageType=NoImage
      mirroring=NoMirror
      resourceName=None
      resourceObject=null
      tiling=NoTile
      imageSize=<0 0>
      margin=<ltrb 0 0 0 0>
      outlineSettings {
       bUseBrushTransparency=true
       roundingType=FixedRadius
       width=1
       color {colorUseRule=UseColor_ColorTable, specifiedColor=<rgba 0 0 0 0>}
       cornerRadii=<4 4 4 4>
      }
      tintColor {colorUseRule=UseColor_Specified, specifiedColor=<rgba 0.087 0.087 0.087 1>}
      uVRegion {bIsValid=false, max=<0 0>, min=<0 0>}
     }
     normalForeground {colorUseRule=UseColor_ColorTable, specifiedColor=<rgba 0 0 0 0>}
     normalPadding=<ltrb 0 0 0 0>
     pressed {
      bIsDynamicallyLoaded=false
      drawAs=Box
      imageType=NoImage
      mirroring=NoMirror
      resourceName=None
      resourceObject=null
      tiling=NoTile
      imageSize=<0 0>
      margin=<ltrb 0 0 0 0>
      outlineSettings {
       bUseBrushTransparency=true
       roundingType=FixedRadius
       width=1
       color {colorUseRule=UseColor_ColorTable, specifiedColor=<rgba 0 0 0 0>}
       cornerRadii=<4 4 4 4>
      }
      tintColor {colorUseRule=UseColor_Specified, specifiedColor=<rgba 0.0075 0.0075 0.0075 1>}
      uVRegion {bIsValid=false, max=<0 0>, min=<0 0>}
     }
     pressedForeground {colorUseRule=UseColor_ColorTable, specifiedColor=<rgba 0 0 0 0>}
     pressedPadding=<ltrb 0 0 0 0>
     pressedSlateSound {resourceObject=null}
    }
   }
   slot {class=HorizontalBoxSlot, properties {size {sizeRule=Fill, value=0.25}}}
  }
  ButtonText {
   widgetClassPath=&TextBlock
   bIsVariable=true
   parent=Button
   widget=ButtonText
   properties {
    text=Open...
    colorAndOpacity {colorUseRule=UseColor_Specified, specifiedColor=<rgba 0.9763 1 0.918 1>}
    font {
     bForceMonospaced=false
     bMaterialIsStencil=false
     fontMaterial=null
     fontObject=&Roboto
     letterSpacing=0
     monospacedWidth=1
     size=8
     skewAmount=0
     typefaceFontName=Regular
     outlineSettings {
      bApplyOutlineToDropShadows=false
      bMiteredCorners=false
      bSeparateFillAlpha=false
      outlineMaterial=null
      outlineSize=0
      outlineColor=<rgba 0 0 0 1>
     }
    }
    strikeBrush {
     bIsDynamicallyLoaded=false
     drawAs=Image
     imageType=NoImage
     mirroring=NoMirror
     resourceName=None
     resourceObject=null
     tiling=NoTile
     imageSize=<32 32>
     margin=<ltrb 0 0 0 0>
     outlineSettings {
      bUseBrushTransparency=false
      roundingType=HalfHeightRadius
      width=0
      color {colorUseRule=UseColor_Specified, specifiedColor=<rgba 0 0 0 0>}
      cornerRadii=<0 0 0 0>
     }
     tintColor {colorUseRule=UseColor_Foreground, specifiedColor=<rgba 1 1 1 1>}
     uVRegion {bIsValid=false, max=<0 0>, min=<0 0>}
    }
   }
   slot {class=ButtonSlot}
  }
 ]
}
