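# Editor Utility Widget providing a reusable labeled button, a Size Box around a button and text, with an optional on/off state.
# PreConstruct applies the ButtonName label, width and height, and runs ComputeButtonColor, which tints the background grey without a state, or green/red by State when HasState is true.
# SetState and SetButtonName update color and label at runtime.
# Clicks are forwarded through the OnClicked dispatcher.
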EditorUtilityWidgetBlueprint /Game/Characters/Player/Widgets/EUW_FAnim_SimpleButton
graphs [
 ComputeButtonColor {nodes [
   ComputeButtonColor() {is=title K2Node_FunctionEntry}
   > n1 = Branch(Condition=Get HasState() -> (HasState:bool))
   > n2 = Branch(Condition=Get State() -> (State:bool))
   n3 = Get Button() -> (Button:Editor Utility Button Object Reference)
   > n4 = SetBackgroundColor!(execute, self:Button Object Reference=@n3, InBackgroundColor:Linear Color Structure=(R=0.1,G=0.4,B=0,A=1)) -> (then)
   Return Node(execute=[@n5.then, @n6.then, @n4.then])
   n5 = SetBackgroundColor(execute=@n2.else, self=@n3, InBackgroundColor=(R=1,G=0,B=0.08,A=1))
   n6 = SetBackgroundColor(execute=@n1.else, self=Get Button(), InBackgroundColor=(R=0.3229,G=0.3229,B=0.3229,A=1))
  ]}
 EventGraph {nodes [
   "On Clicked (Button)"() {is=title K2Node_ComponentBoundEvent}
   > Call On Clicked!(execute:Exec, self:EUW FAnim Simple Button Object Reference) -> (then:Exec) {is=title K2Node_CallDelegate}
   Event PreConstruct()
   > ComputeButtonColor() {is=title K2Node_CallFunction}
   > "SetText (Text)"(self=Get Text!() -> (Text:Text Object Reference), InText=Get ButtonName() -> (ButtonName:Text))
   > SetHeightOverride(self=Get Size() -> (Size:Size Box Object Reference), InHeightOverride=Get Button Height() -> (Button Height:double))
   > SetWidthOverride(self=Get Size(), InWidthOverride=Get Button Width() -> (Button Width:double))
  ]}
 Get_Text_ColorAndOpacity {nodes [
   Get_Text_ColorAndOpacity() {is=title K2Node_FunctionEntry}
   > Return Node!(execute, ReturnValue:Slate Color Structure=Get ButtonColor() -> (ButtonColor:Slate Color Structure))
  ]}
 OnClicked {nodes [OnClicked() {is=title K2Node_FunctionEntry}]}
 SetButtonName {nodes [
   n1 = SetButtonName!() -> (then, NewName:Text) {is=title K2Node_FunctionEntry}
   > n2 = Set ButtonName(ButtonName:Text=@n1) -> (Output_Get:Text)
   > "SetText (Text)"(self=Get Text(), InText=@n2)
   > Return Node!(execute)
  ]}
 SetState {nodes [
   n1 = SetState!() -> (then, Enabled:bool=~) {is=title K2Node_FunctionEntry}
   > Set State(State:bool=@n1) -> (Output_Get:bool)
   > ComputeButtonColor()
   > Return Node()
  ]}
]
variables [
 Button Height {type=number, default=25}
 Button Width {type=number, default=75}
 ButtonColor {type=SlateColor, default {colorUseRule=UseColor_Specified, specifiedColor=<rgba 0.0009 1 0 1>}}
 ButtonName {type=string, default=Button Name}
 HasState {type=boolean}
 OnClicked {default=()}
 State {type=boolean}
]
widgets {
 info {rootWidgetClass=&SizeBox, widgetCount=3}
 widgets [
  Size {
   widgetClassPath=&SizeBox
   bIsVariable=true
   widget=Size
   properties {bOverride_HeightOverride=true, bOverride_WidthOverride=true, clipping=ClipToBounds, heightOverride=25, widthOverride=75}
  }
  Button {
   widgetClassPath=&EditorUtilityButton
   parent=Size
   widget=Button
   properties {widgetStyle {
     clickedSlateSound {resourceObject=null}
     disabled {
      bIsDynamicallyLoaded=false
      drawAs=Box
      imageType=NoImage
      mirroring=NoMirror
      resourceName=None
      resourceObject=null
      tiling=NoTile
      imageSize=<0 0>
      margin=<ltrb 0 0 0 0>
      outlineSettings {
       bUseBrushTransparency=false
       roundingType=FixedRadius
       width=1
       color {colorUseRule=UseColor_ColorTable, specifiedColor=<rgba 0 0 0 0>}
       cornerRadii=<4 4 4 4>
      }
      tintColor {colorUseRule=UseColor_Specified, specifiedColor=<rgba 0.0176 0.0176 0.0176 1>}
      uVRegion {bIsValid=false, max=<0 0>, min=<0 0>}
     }
     disabledForeground {colorUseRule=UseColor_ColorTable, specifiedColor=<rgba 0 0 0 0>}
     hovered {
      bIsDynamicallyLoaded=false
      drawAs=RoundedBox
      imageType=NoImage
      mirroring=NoMirror
      resourceName=None
      resourceObject=null
      tiling=NoTile
      imageSize=<0 0>
      margin=<ltrb 0 0 0 0>
      outlineSettings {
       bUseBrushTransparency=true
       roundingType=FixedRadius
       width=1
       color {colorUseRule=UseColor_ColorTable, specifiedColor=<rgba 0 0 0 0>}
       cornerRadii=<4 4 4 4>
      }
      tintColor {colorUseRule=UseColor_Specified, specifiedColor=<rgba 0.6087 0.6087 0.6087 1>}
      uVRegion {bIsValid=false, max=<0 0>, min=<0 0>}
     }
     hoveredForeground {colorUseRule=UseColor_ColorTable, specifiedColor=<rgba 0 0 0 0>}
     hoveredSlateSound {resourceObject=null}
     normal {
      bIsDynamicallyLoaded=false
      drawAs=RoundedBox
      imageType=NoImage
      mirroring=NoMirror
      resourceName=None
      resourceObject=null
      tiling=NoTile
      imageSize=<0 0>
      margin=<ltrb 0 0 0 0>
      outlineSettings {
       bUseBrushTransparency=true
       roundingType=FixedRadius
       width=1
       color {colorUseRule=UseColor_ColorTable, specifiedColor=<rgba 0 0 0 0>}
       cornerRadii=<4 4 4 4>
      }
      tintColor {colorUseRule=UseColor_Specified, specifiedColor=<rgba 0.087 0.087 0.087 1>}
      uVRegion {bIsValid=false, max=<0 0>, min=<0 0>}
     }
     normalForeground {colorUseRule=UseColor_ColorTable, specifiedColor=<rgba 0 0 0 0>}
     normalPadding=<ltrb 0 0 0 0>
     pressed {
      bIsDynamicallyLoaded=false
      drawAs=Box
      imageType=NoImage
      mirroring=NoMirror
      resourceName=None
      resourceObject=null
      tiling=NoTile
      imageSize=<0 0>
      margin=<ltrb 0 0 0 0>
      outlineSettings {
       bUseBrushTransparency=true
       roundingType=FixedRadius
       width=1
       color {colorUseRule=UseColor_ColorTable, specifiedColor=<rgba 0 0 0 0>}
       cornerRadii=<4 4 4 4>
      }
      tintColor {colorUseRule=UseColor_Specified, specifiedColor=<rgba 0.0075 0.0075 0.0075 1>}
      uVRegion {bIsValid=false, max=<0 0>, min=<0 0>}
     }
     pressedForeground {colorUseRule=UseColor_ColorTable, specifiedColor=<rgba 0 0 0 0>}
     pressedPadding=<ltrb 0 0 0 0>
     pressedSlateSound {resourceObject=null}
    }}
   slot {class=SizeBoxSlot}
  }
  Text {
   widgetClassPath=&TextBlock
   bIsVariable=true
   parent=Button
   widget=Text
   properties {
    text=Button Name
    colorAndOpacity {colorUseRule=UseColor_Specified, specifiedColor=<rgba 0.9763 1 0.918 1>}
    font {
     bForceMonospaced=false
     bMaterialIsStencil=false
     fontMaterial=null
     fontObject=&Roboto
     letterSpacing=0
     monospacedWidth=1
     size=8
     skewAmount=0
     typefaceFontName=Regular
     outlineSettings {
      bApplyOutlineToDropShadows=false
      bMiteredCorners=false
      bSeparateFillAlpha=false
      outlineMaterial=null
      outlineSize=0
      outlineColor=<rgba 0 0 0 1>
     }
    }
    strikeBrush {
     bIsDynamicallyLoaded=false
     drawAs=Image
     imageType=NoImage
     mirroring=NoMirror
     resourceName=None
     resourceObject=null
     tiling=NoTile
     imageSize=<32 32>
     margin=<ltrb 0 0 0 0>
     outlineSettings {
      bUseBrushTransparency=false
      roundingType=HalfHeightRadius
      width=0
      color {colorUseRule=UseColor_Specified, specifiedColor=<rgba 0 0 0 0>}
      cornerRadii=<0 0 0 0>
     }
     tintColor {colorUseRule=UseColor_Foreground, specifiedColor=<rgba 1 1 1 1>}
     uVRegion {bIsValid=false, max=<0 0>, min=<0 0>}
    }
   }
   slot {class=ButtonSlot}
  }
 ]
}
