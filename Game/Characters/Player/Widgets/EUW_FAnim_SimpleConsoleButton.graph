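# Editor utility widget providing a reusable, labeled console-command button, with text, size and color applied on PreConstruct.
# With Command Has Int State set, a click toggles the CVar between enabled and disabled values without PIE world context, then recolors the button green or red, with optional inversion.
# Otherwise it is gray and runs the command in the first PIE world, or without context if none exists.
# OnClicked and OnPostClicked dispatchers expose click hooks.
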EditorUtilityWidgetBlueprint /Game/Characters/Player/Widgets/EUW_FAnim_SimpleConsoleButton
graphs [
 EventGraph {nodes [
   n1 = GetConsoleVariableIntValue!(execute=@n9.then, VariableName:String=Get Console Command() -> (Console Command:String)) -> (then, ReturnValue:int=0)
   n2 = Get DisabledIntValue() -> (DisabledIntValue:int=0)
   n3 = "Not Equal (Integer)"(A=@n1, B=@n2)
   n4 = Get EnabledIntValue() -> (EnabledIntValue:int=0)
   n5 = Append!(
    A=Get Console Command()
    B=" "
    C:String="To String (Integer)"(InInt=Select(Option 0:int=@n4, Option 1:int=@n2, Index=@n3) -> (ReturnValue:int))
   ) -> (ReturnValue)
   n6 = Get Run On Server() -> (Run On Server:bool)
   > n7 = GetPIEWorlds!(execute, bIncludeDedicatedServer:bool=@n6) -> (then, ReturnValue:Array of World Object References)
   n8 = ExecuteConsoleCommand(WorldContextObject="Get (a copy)"(Array:Array of World Object References=@n7) -> (Output:World Object Reference), Command=@n5)
   UpdateButtonColor(execute=[@n8.then, @n13.then])
   > Call On Post Clicked!(execute:Exec, self:EUW FAnim Simple Console Button Object Reference) -> (then:Exec) {is=title K2Node_CallDelegate}
   "On Clicked (Button)"() {is=title K2Node_ComponentBoundEvent}
   > Call On Clicked!(execute:Exec, self:EUW FAnim Simple Console Button Object Reference) -> (then:Exec) {is=title K2Node_CallDelegate}
   > n9 = Branch(Condition=Get Command Has Int State() -> (Command Has Int State:bool))
   n10 = GetPIEWorlds(execute=@n9.else, bIncludeDedicatedServer=@n6)
   > n11 = Branch(Condition=Is Not Empty!(TargetArray:Array of World Object References=@n10) -> (ReturnValue:bool))
   > ExecuteConsoleCommand(WorldContextObject="Get (a copy)"(Array=@n10), Command=Get Console Command())
   > Call On Post Clicked()
   ExecuteConsoleCommand(execute=@n11.else, Command=Get Console Command())
   Event PreConstruct()
   > UpdateButtonColor()
   > "SetText (Text)"(self=Get Text!() -> (Text:Text Object Reference), InText=Get ButtonName() -> (ButtonName:Text))
   > SetHeightOverride(self=Get Size() -> (Size:Size Box Object Reference), InHeightOverride=Get Button Height() -> (Button Height:double))
   > SetWidthOverride(self=Get Size(), InWidthOverride=Get Button Width() -> (Button Width:double))
   n12 = Is Not Empty(TargetArray=@n7)
   n13 = ExecuteConsoleCommand(execute=[@n14.else, @n14.then], Command=@n5)
   n14 = Branch(execute=@n7.then, Condition=@n12)
  ]}
 Get_Text_ColorAndOpacity {nodes [
   Get_Text_ColorAndOpacity() {is=title K2Node_FunctionEntry}
   > Return Node!(execute, ReturnValue:Slate Color Structure=Get ButtonColor() -> (ButtonColor:Slate Color Structure))
  ]}
 OnClicked {nodes [OnClicked() {is=title K2Node_FunctionEntry}]}
 OnPostClicked {nodes [OnPostClicked() {is=title K2Node_FunctionEntry}]}
 UpdateButtonColor {nodes [
   UpdateButtonColor() {is=title K2Node_FunctionEntry}
   > n1 = Branch(Condition=Get Command Has Int State())
   > n2 = GetConsoleVariableIntValue(VariableName=Get Console Command()) -> (ReturnValue=0)
   n3 = Get DisabledIntValue() -> (DisabledIntValue=0)
   > n4 = Branch(Condition=XOR Boolean!(A:bool="Not Equal (Integer)"(A=@n2, B=@n3), B:bool=Get InvertColorLogic() -> (InvertColorLogic:bool)) -> (ReturnValue:bool))
   n5 = Get Button() -> (Button:Editor Utility Button Object Reference)
   > n6 = SetBackgroundColor!(execute, self:Button Object Reference=@n5, InBackgroundColor:Linear Color Structure=(R=0.1,G=0.4,B=0,A=1)) -> (then)
   Return Node!(execute=[@n7.then, @n8.then, @n6.then])
   n7 = SetBackgroundColor(execute=@n4.else, self=@n5, InBackgroundColor=(R=1,G=0,B=0.08,A=1))
   n8 = SetBackgroundColor(execute=@n1.else, self=Get Button(), InBackgroundColor=(R=0.3229,G=0.3229,B=0.3229,A=1))
  ]}
]
variables [
 Button Height {type=number, default=25}
 Button Width {type=number, default=75}
 ButtonColor {type=SlateColor, default {colorUseRule=UseColor_Specified, specifiedColor=<rgba 0.0009 1 0 1>}}
 ButtonName {type=string, default=Button Name}
 Command Has Int State {type=boolean}
 Console Command {type=string, default=""}
 DisabledIntValue {type=integer}
 EnabledIntValue {type=integer, default=1}
 InvertColorLogic {type=boolean}
 OnClicked {default=()}
 OnPostClicked {default=()}
 Run On Server {type=boolean, default=true}
]
widgets {
 info {rootWidgetClass=&SizeBox, widgetCount=3}
 widgets [
  Size {
   widgetClassPath=&SizeBox
   bIsVariable=true
   widget=Size
   properties {bOverride_HeightOverride=true, bOverride_WidthOverride=true, clipping=ClipToBounds, heightOverride=25, widthOverride=75}
  }
  Button {
   widgetClassPath=&EditorUtilityButton
   parent=Size
   widget=Button
   properties {widgetStyle {
     clickedSlateSound {resourceObject=null}
     disabled {
      bIsDynamicallyLoaded=false
      drawAs=Box
      imageType=NoImage
      mirroring=NoMirror
      resourceName=None
      resourceObject=null
      tiling=NoTile
      imageSize=<0 0>
      margin=<ltrb 0 0 0 0>
      outlineSettings {
       bUseBrushTransparency=false
       roundingType=FixedRadius
       width=1
       color {colorUseRule=UseColor_ColorTable, specifiedColor=<rgba 0 0 0 0>}
       cornerRadii=<4 4 4 4>
      }
      tintColor {colorUseRule=UseColor_Specified, specifiedColor=<rgba 0.0176 0.0176 0.0176 1>}
      uVRegion {bIsValid=false, max=<0 0>, min=<0 0>}
     }
     disabledForeground {colorUseRule=UseColor_ColorTable, specifiedColor=<rgba 0 0 0 0>}
     hovered {
      bIsDynamicallyLoaded=false
      drawAs=RoundedBox
      imageType=NoImage
      mirroring=NoMirror
      resourceName=None
      resourceObject=null
      tiling=NoTile
      imageSize=<0 0>
      margin=<ltrb 0 0 0 0>
      outlineSettings {
       bUseBrushTransparency=true
       roundingType=FixedRadius
       width=1
       color {colorUseRule=UseColor_ColorTable, specifiedColor=<rgba 0 0 0 0>}
       cornerRadii=<4 4 4 4>
      }
      tintColor {colorUseRule=UseColor_Specified, specifiedColor=<rgba 0.6087 0.6087 0.6087 1>}
      uVRegion {bIsValid=false, max=<0 0>, min=<0 0>}
     }
     hoveredForeground {colorUseRule=UseColor_ColorTable, specifiedColor=<rgba 0 0 0 0>}
     hoveredSlateSound {resourceObject=null}
     normal {
      bIsDynamicallyLoaded=false
      drawAs=RoundedBox
      imageType=NoImage
      mirroring=NoMirror
      resourceName=None
      resourceObject=null
      tiling=NoTile
      imageSize=<0 0>
      margin=<ltrb 0 0 0 0>
      outlineSettings {
       bUseBrushTransparency=true
       roundingType=FixedRadius
       width=1
       color {colorUseRule=UseColor_ColorTable, specifiedColor=<rgba 0 0 0 0>}
       cornerRadii=<4 4 4 4>
      }
      tintColor {colorUseRule=UseColor_Specified, specifiedColor=<rgba 0.087 0.087 0.087 1>}
      uVRegion {bIsValid=false, max=<0 0>, min=<0 0>}
     }
     normalForeground {colorUseRule=UseColor_ColorTable, specifiedColor=<rgba 0 0 0 0>}
     normalPadding=<ltrb 0 0 0 0>
     pressed {
      bIsDynamicallyLoaded=false
      drawAs=Box
      imageType=NoImage
      mirroring=NoMirror
      resourceName=None
      resourceObject=null
      tiling=NoTile
      imageSize=<0 0>
      margin=<ltrb 0 0 0 0>
      outlineSettings {
       bUseBrushTransparency=true
       roundingType=FixedRadius
       width=1
       color {colorUseRule=UseColor_ColorTable, specifiedColor=<rgba 0 0 0 0>}
       cornerRadii=<4 4 4 4>
      }
      tintColor {colorUseRule=UseColor_Specified, specifiedColor=<rgba 0.0075 0.0075 0.0075 1>}
      uVRegion {bIsValid=false, max=<0 0>, min=<0 0>}
     }
     pressedForeground {colorUseRule=UseColor_ColorTable, specifiedColor=<rgba 0 0 0 0>}
     pressedPadding=<ltrb 0 0 0 0>
     pressedSlateSound {resourceObject=null}
    }}
   slot {class=SizeBoxSlot}
  }
  Text {
   widgetClassPath=&TextBlock
   bIsVariable=true
   parent=Button
   widget=Text
   properties {
    text=Button Name
    colorAndOpacity {colorUseRule=UseColor_Specified, specifiedColor=<rgba 0.9763 1 0.918 1>}
    font {
     bForceMonospaced=false
     bMaterialIsStencil=false
     fontMaterial=null
     fontObject=&Roboto
     letterSpacing=0
     monospacedWidth=1
     size=8
     skewAmount=0
     typefaceFontName=Regular
     outlineSettings {
      bApplyOutlineToDropShadows=false
      bMiteredCorners=false
      bSeparateFillAlpha=false
      outlineMaterial=null
      outlineSize=0
      outlineColor=<rgba 0 0 0 1>
     }
    }
    strikeBrush {
     bIsDynamicallyLoaded=false
     drawAs=Image
     imageType=NoImage
     mirroring=NoMirror
     resourceName=None
     resourceObject=null
     tiling=NoTile
     imageSize=<32 32>
     margin=<ltrb 0 0 0 0>
     outlineSettings {
      bUseBrushTransparency=false
      roundingType=HalfHeightRadius
      width=0
      color {colorUseRule=UseColor_Specified, specifiedColor=<rgba 0 0 0 0>}
      cornerRadii=<0 0 0 0>
     }
     tintColor {colorUseRule=UseColor_Foreground, specifiedColor=<rgba 1 1 1 1>}
     uVRegion {bIsValid=false, max=<0 0>, min=<0 0>}
    }
   }
   slot {class=ButtonSlot}
  }
 ]
}
