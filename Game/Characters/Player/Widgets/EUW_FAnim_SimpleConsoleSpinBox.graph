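# Reusable editor utility spin box that edits a numeric console variable named by its Console Command string.
# Committing a value runs that command with the value appended in the first PIE world, if one exists.
# PreConstruct applies box size, optional min/max and slider settings, plus a default value when console syncing is off.
# With syncing on, Tick copies the console variable's value into the box unless it has user focus.
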EditorUtilityWidgetBlueprint /Game/Characters/Player/Widgets/EUW_FAnim_SimpleConsoleSpinBox
graphs [EventGraph {nodes [
   n1 = "On Value Committed (SpinBox)"!() -> (OutputDelegate, then, InValue:float, CommitMethod:ETextCommit Enum=Default) {is=title K2Node_ComponentBoundEvent}
   > n2 = GetPIEWorlds!(execute, bIncludeDedicatedServer:bool=true) -> (then, ReturnValue:Array of World Object References)
   > Branch(Condition=Is Not Empty!(TargetArray:Array of World Object References=@n2) -> (ReturnValue:bool))
   > ExecuteConsoleCommand(
    WorldContextObject="Get (a copy)"(Array:Array of World Object References=@n2) -> (Output:World Object Reference)
    Command=Append!(A=Get Console Command() -> (Console Command:String), B=" ", C:String="To String (Float)"(InDouble=@n1.InValue)) -> (ReturnValue)
   )
   Event PreConstruct()
   > n3 = Branch(Condition=Get UpdateValueToMatchConsoleVariable() -> (UpdateValueToMatchConsoleVariable:bool))
   SetHeightOverride(execute=[@n3.then, @n9.then], self=Get Size() -> (Size:Size Box Object Reference), InHeightOverride=Get Box Height() -> (Box Height:double))
   > SetWidthOverride(self=Get Size(), InWidthOverride=Get Box Width() -> (Box Width:double))
   > n4 = Sequence!(execute) -> (then_0, then_1, then_2:Exec)
   > n5 = Branch(Condition=Get Use Min() -> (Use Min:bool))
   > SetMinValue!(
    execute
    self:Spin Box Object Reference=Get SpinBox() -> (SpinBox:Editor Utility Spin Box Object Reference)
    NewValue:float=Get Min Value() -> (Min Value:float)
   ) -> (then)
   > SetMinSliderValue!(execute, self:Spin Box Object Reference=Get SpinBox(), NewValue:float=Get Min Value()) -> (then)
   ClearMinValue(execute=@n5.else, self:Spin Box Object Reference=Get SpinBox())
   > ClearMinSliderValue(self:Spin Box Object Reference=Get SpinBox())
   n6 = Branch(execute=@n4.then_1, Condition=Get Use Max() -> (Use Max:bool))
   > SetMaxValue!(execute, self:Spin Box Object Reference=Get SpinBox(), NewValue:float=Get Max Value() -> (Max Value:float)) -> (then)
   > SetMaxSliderValue!(execute, self:Spin Box Object Reference=Get SpinBox(), NewValue:float=Get Max Value()) -> (then)
   ClearMaxValue(execute=@n6.else, self:Spin Box Object Reference=Get SpinBox())
   > ClearMaxSliderValue(self:Spin Box Object Reference=Get SpinBox())
   n7 = Set bEnableSlider!(execute=@n4.then_2, bEnableSlider:bool=Get Enable Slider() -> (Enable Slider:bool), self:Spin Box Object Reference=Get SpinBox()) -> (then, Output_Get:bool)
   > n8 = Branch(Condition=@n7)
   > Set Delta!(execute, Delta:float=Get SliderSnapDelta() -> (SliderSnapDelta:float), self:Spin Box Object Reference=Get SpinBox()) -> (then, Output_Get:float)
   ClearMinSliderValue(execute=@n8.else, self=Get SpinBox())
   > ClearMaxSliderValue(self=Get SpinBox())
   n9 = Set Value!(execute=@n3.else, Value:float=Get DefaultValue() -> (DefaultValue:double), self:Spin Box Object Reference=Get SpinBox()) -> (then, Output_Get:float)
   Event Tick()
   > Branch(Condition=Get UpdateValueToMatchConsoleVariable())
   n10 = Get SpinBox()
   > n11 = Branch(Condition=HasAnyUserFocus!(self:Widget Object Reference=@n10) -> (ReturnValue:bool))
   n12 = GetConsoleVariableFloatValue!(execute=@n11.else, VariableName:String=Get Console Command()) -> (then, ReturnValue:float)
   > Set Value(Value=@n12, self=@n10)
  ]}]
variables [
 Box Height {type=number, default=25}
 Box Width {type=number, default=75}
 Console Command {type=string, default=""}
 DefaultValue {type=number}
 Enable Slider {type=boolean, default=true}
 Max Value {type=number}
 Min Value {type=number}
 SliderSnapDelta {type=number}
 UpdateValueToMatchConsoleVariable {type=boolean}
 Use Max {type=boolean}
 Use Min {type=boolean}
]
widgets {
 info {rootWidgetClass=&SizeBox, widgetCount=2}
 widgets [
  Size {
   widgetClassPath=&SizeBox
   bIsVariable=true
   widget=Size
   properties {bOverride_HeightOverride=true, bOverride_WidthOverride=true, clipping=ClipToBounds, heightOverride=25, widthOverride=75}
  }
  SpinBox {
   widgetClassPath=/Script/Blutility.EditorUtilitySpinBox
   bIsVariable=true
   parent=Size
   widget=SpinBox
   properties {
    minFractionalDigits=0
    font {
     bForceMonospaced=false
     bMaterialIsStencil=false
     fontMaterial=null
     fontObject=&Roboto
     letterSpacing=0
     monospacedWidth=1
     size=8
     skewAmount=0
     typefaceFontName=Regular
     outlineSettings {
      bApplyOutlineToDropShadows=false
      bMiteredCorners=false
      bSeparateFillAlpha=false
      outlineMaterial=null
      outlineSize=0
      outlineColor=<rgba 0 0 0 1>
     }
    }
    foregroundColor {colorUseRule=UseColor_Specified, specifiedColor=<rgba 0.5 0.5 0.5 1>}
    widgetStyle {
     activeBackgroundBrush {
      bIsDynamicallyLoaded=false
      drawAs=Image
      imageType=NoImage
      mirroring=NoMirror
      resourceName=None
      resourceObject=null
      tiling=NoTile
      imageSize=<32 32>
      margin=<ltrb 0 0 0 0>
      outlineSettings {
       bUseBrushTransparency=false
       roundingType=FixedRadius
       width=1
       color {colorUseRule=UseColor_ColorTable, specifiedColor=<rgba 0 0 0 0>}
       cornerRadii=<4 4 4 4>
      }
      tintColor {colorUseRule=UseColor_Specified, specifiedColor=<rgba 0.0048 0.0048 0.0048 1>}
      uVRegion {bIsValid=false, max=<0 0>, min=<0 0>}
     }
     activeFillBrush {
      bIsDynamicallyLoaded=false
      drawAs=Image
      imageType=NoImage
      mirroring=NoMirror
      resourceName=None
      resourceObject=null
      tiling=NoTile
      imageSize=<32 32>
      margin=<ltrb 0 0 0 0>
      outlineSettings {
       bUseBrushTransparency=false
       roundingType=FixedRadius
       width=1
       color {colorUseRule=UseColor_Specified, specifiedColor=<rgba 0 0 0 0>}
       cornerRadii=<4 4 4 4>
      }
      tintColor {colorUseRule=UseColor_Specified, specifiedColor=<rgba 0.028 0.028 0.028 1>}
      uVRegion {bIsValid=false, max=<0 0>, min=<0 0>}
     }
     arrowsImage {
      bIsDynamicallyLoaded=false
      drawAs=NoDrawType
      imageType=NoImage
      mirroring=NoMirror
      resourceName=None
      resourceObject=null
      tiling=NoTile
      imageSize=<0 0>
      margin=<ltrb 0 0 0 0>
      outlineSettings {
       bUseBrushTransparency=false
       roundingType=HalfHeightRadius
       width=0
       color {colorUseRule=UseColor_Specified, specifiedColor=<rgba 0 0 0 0>}
       cornerRadii=<0 0 0 0>
      }
      tintColor {colorUseRule=UseColor_Specified, specifiedColor=<rgba 1 1 1 1>}
      uVRegion {bIsValid=false, max=<0 0>, min=<0 0>}
     }
     backgroundBrush {
      bIsDynamicallyLoaded=false
      drawAs=Image
      imageType=NoImage
      mirroring=NoMirror
      resourceName=None
      resourceObject=null
      tiling=NoTile
      imageSize=<32 32>
      margin=<ltrb 0 0 0 0>
      outlineSettings {
       bUseBrushTransparency=false
       roundingType=FixedRadius
       width=1
       color {colorUseRule=UseColor_ColorTable, specifiedColor=<rgba 0 0 0 0>}
       cornerRadii=<4 4 4 4>
      }
      tintColor {colorUseRule=UseColor_Specified, specifiedColor=<rgba 0.0048 0.0048 0.0048 1>}
      uVRegion {bIsValid=false, max=<0 0>, min=<0 0>}
     }
     foregroundColor {colorUseRule=UseColor_Specified, specifiedColor=<rgba 1 1 1 1>}
     hoveredBackgroundBrush {
      bIsDynamicallyLoaded=false
      drawAs=Image
      imageType=NoImage
      mirroring=NoMirror
      resourceName=None
      resourceObject=null
      tiling=NoTile
      imageSize=<32 32>
      margin=<ltrb 0 0 0 0>
      outlineSettings {
       bUseBrushTransparency=false
       roundingType=FixedRadius
       width=1
       color {colorUseRule=UseColor_ColorTable, specifiedColor=<rgba 0 0 0 0>}
       cornerRadii=<4 4 4 4>
      }
      tintColor {colorUseRule=UseColor_Specified, specifiedColor=<rgba 0.0048 0.0048 0.0048 1>}
      uVRegion {bIsValid=false, max=<0 0>, min=<0 0>}
     }
     hoveredFillBrush {
      bIsDynamicallyLoaded=false
      drawAs=Image
      imageType=NoImage
      mirroring=NoMirror
      resourceName=None
      resourceObject=null
      tiling=NoTile
      imageSize=<32 32>
      margin=<ltrb 0 0 0 0>
      outlineSettings {
       bUseBrushTransparency=false
       roundingType=FixedRadius
       width=1
       color {colorUseRule=UseColor_Specified, specifiedColor=<rgba 0 0 0 0>}
       cornerRadii=<4 4 4 4>
      }
      tintColor {colorUseRule=UseColor_Specified, specifiedColor=<rgba 0.0284 0.0284 0.0284 1>}
      uVRegion {bIsValid=false, max=<0 0>, min=<0 0>}
     }
     inactiveFillBrush {
      bIsDynamicallyLoaded=false
      drawAs=Image
      imageType=NoImage
      mirroring=NoMirror
      resourceName=None
      resourceObject=null
      tiling=NoTile
      imageSize=<32 32>
      margin=<ltrb 0 0 0 0>
      outlineSettings {
       bUseBrushTransparency=false
       roundingType=FixedRadius
       width=1
       color {colorUseRule=UseColor_Specified, specifiedColor=<rgba 0 0 0 0>}
       cornerRadii=<4 4 4 4>
      }
      tintColor {colorUseRule=UseColor_Specified, specifiedColor=<rgba 0.0176 0.0176 0.0176 1>}
      uVRegion {bIsValid=false, max=<0 0>, min=<0 0>}
     }
     insetPadding=<ltrb 0 0 0 0>
     textPadding=<ltrb 5 0 0 0>
    }
   }
   slot {class=SizeBoxSlot}
  }
 ]
}
