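# Small removable token widget that shows a text label beside an "x" button, apparently one item in a listbox-style UI.
# PreConstruct calls UpdateText to copy the Text variable into the label.
# Clicking the button removes the widget from its parent, broadcasts OnItemRemoved with that text, then clears all the dispatcher's bindings.
# SetParent stores ListView and WrapBox references, which these graphs never read.
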WidgetBlueprint /Game/Characters/Player/Widgets/WB_FAnim_ListboxToken
graphs [
 EventGraph {nodes [
   "On Clicked (Button_RemoveItem)"() {is=title K2Node_ComponentBoundEvent}
   > RemoveFromParent()
   > Call On Item Removed!(execute:Exec, self:WB FAnim Listbox Token Object Reference, RemovedItem:Text=Get Text!() -> (Text)) -> (then:Exec) {is=title K2Node_CallDelegate}
   > Unbind all Events from On Item Removed!(execute:Exec, self:WB FAnim Listbox Token Object Reference) -> (then:Exec) {is=title K2Node_ClearDelegate}
   Event PreConstruct()
   > UpdateText()
   Event Construct()
   Event Tick()
  ]}
 OnItemRemoved {nodes [OnItemRemoved!() -> (then, RemovedItem:Text) {is=title K2Node_FunctionEntry}]}
 SetParent {nodes [
   n1 = SetParent!() -> (then, ParentListView:List View Object Reference, ParentWrapBox:Wrap Box Object Reference) {is=title K2Node_FunctionEntry}
   > Set ParentListBox(ParentListBox:List View Object Reference=@n1) -> (Output_Get:List View Object Reference)
   > Set ParentWrapBox(ParentWrapBox:Wrap Box Object Reference=@n1.ParentWrapBox) -> (Output_Get:Wrap Box Object Reference)
   > Return Node()
  ]}
 UpdateText {nodes [
   UpdateText() {is=title K2Node_FunctionEntry}
   > Set Text(Text=Get Text(), self=Get TextBox_ItemText() -> (TextBox_ItemText:Text Object Reference))
   > Return Node()
  ]}
]
variables [
 ItemData {type=WB_FAnim_ListboxToken_C, default=null}
 OnItemRemoved {}
 ParentListBox {type=ListView, default=null}
 ParentWrapBox {type=WrapBox, default=null}
 Text {type=string, default=""}
]
widgets {
 info {inheritedWidgetCount=0, parentClass=&UserWidget, rootWidgetClass=&SizeBox, widgetCount=6}
 widgets [
  SizeBox_108 {widgetClassPath=&SizeBox, widget=SizeBox_108, properties {bOverride_HeightOverride=true, heightOverride=25, widthOverride=75}}
  HorizontalBox_352 {
   widgetClassPath=&HorizontalBox
   parent=SizeBox_108
   widget=HorizontalBox_352
   slot {class=SizeBoxSlot, properties {horizontalAlignment=HAlign_Left, verticalAlignment=VAlign_Center, padding=<ltrb 0 0 4 0>}}
  }
  TextBox_ItemText {
   widgetClassPath=&TextBlock
   bIsVariable=true
   parent=HorizontalBox_352
   widget=TextBox_ItemText
   properties {
    text=Text
    colorAndOpacity {colorUseRule=UseColor_Specified, specifiedColor=<rgba 0.5 0.5 0.5 1>}
    font {
     bForceMonospaced=false
     bMaterialIsStencil=false
     fontMaterial=null
     fontObject=&Roboto
     letterSpacing=0
     monospacedWidth=1
     size=8
     skewAmount=0
     typefaceFontName=Regular
     outlineSettings {
      bApplyOutlineToDropShadows=false
      bMiteredCorners=false
      bSeparateFillAlpha=false
      outlineMaterial=null
      outlineSize=0
      outlineColor=<rgba 0 0 0 1>
     }
    }
    shadowOffset=<0 0>
   }
   slot {class=HorizontalBoxSlot, properties {horizontalAlignment=HAlign_Left, verticalAlignment=VAlign_Center, padding=<ltrb 2 0 2 0>}}
  }
  SizeBox_219 {
   widgetClassPath=&SizeBox
   parent=HorizontalBox_352
   widget=SizeBox_219
   properties {bOverride_HeightOverride=true, bOverride_WidthOverride=true, heightOverride=15, widthOverride=15}
   slot {class=HorizontalBoxSlot, properties {horizontalAlignment=HAlign_Center, verticalAlignment=VAlign_Center}}
  }
  Button_RemoveItem {
   widgetClassPath=&Button
   parent=SizeBox_219
   widget=Button_RemoveItem
   properties {
    backgroundColor=<rgba 0.3229 0.3229 0.3229 1>
    widgetStyle {
     clickedSlateSound {resourceObject=null}
     disabled {
      bIsDynamicallyLoaded=false
      drawAs=NoDrawType
      imageType=NoImage
      mirroring=NoMirror
      resourceName=None
      resourceObject=null
      tiling=NoTile
      imageSize=<0 0>
      margin=<ltrb 0 0 0 0>
      outlineSettings {
       bUseBrushTransparency=false
       roundingType=HalfHeightRadius
       width=0
       color {colorUseRule=UseColor_Specified, specifiedColor=<rgba 0 0 0 0>}
       cornerRadii=<0 0 0 0>
      }
      tintColor {colorUseRule=UseColor_Specified, specifiedColor=<rgba 1 1 1 1>}
      uVRegion {bIsValid=false, max=<0 0>, min=<0 0>}
     }
     disabledForeground {colorUseRule=UseColor_Specified, specifiedColor=<rgba 0.5271 0.5271 0.5271 1>}
     hovered {
      bIsDynamicallyLoaded=false
      drawAs=RoundedBox
      imageType=NoImage
      mirroring=NoMirror
      resourceName=None
      resourceObject=null
      tiling=NoTile
      imageSize=<32 32>
      margin=<ltrb 0 0 0 0>
      outlineSettings {
       bUseBrushTransparency=true
       roundingType=FixedRadius
       width=1
       color {colorUseRule=UseColor_Specified, specifiedColor=<rgba 0.7243 0.7243 0.7243 1>}
       cornerRadii=<4 4 4 4>
      }
      tintColor {colorUseRule=UseColor_Specified, specifiedColor=<rgba 0.6087 0.6087 0.6087 0.6087>}
      uVRegion {bIsValid=false, max=<0 0>, min=<0 0>}
     }
     hoveredForeground {colorUseRule=UseColor_Specified, specifiedColor=<rgba 0 1 1 1>}
     hoveredSlateSound {resourceObject=null}
     normal {
      bIsDynamicallyLoaded=false
      drawAs=RoundedBox
      imageType=NoImage
      mirroring=NoMirror
      resourceName=None
      resourceObject=null
      tiling=NoTile
      imageSize=<32 32>
      margin=<ltrb 0 0 0 0>
      outlineSettings {
       bUseBrushTransparency=true
       roundingType=FixedRadius
       width=1
       color {colorUseRule=UseColor_Specified, specifiedColor=<rgba 0.087 0.087 0.087 1>}
       cornerRadii=<4 4 4 4>
      }
      tintColor {colorUseRule=UseColor_Specified, specifiedColor=<rgba 0.087 0.087 0.087 1>}
      uVRegion {bIsValid=false, max=<0 0>, min=<0 0>}
     }
     normalForeground {colorUseRule=UseColor_Specified, specifiedColor=<rgba 0 1 1 1>}
     normalPadding=<ltrb 0 1.5 0 1.5>
     pressed {
      bIsDynamicallyLoaded=false
      drawAs=RoundedBox
      imageType=NoImage
      mirroring=NoMirror
      resourceName=None
      resourceObject=null
      tiling=NoTile
      imageSize=<32 32>
      margin=<ltrb 0 0 0 0>
      outlineSettings {
       bUseBrushTransparency=true
       roundingType=FixedRadius
       width=1
       color {colorUseRule=UseColor_Specified, specifiedColor=<rgba 0.7243 0.7243 0.7243 1>}
       cornerRadii=<4 4 4 4>
      }
      tintColor {colorUseRule=UseColor_Specified, specifiedColor=<rgba 0.3843 0.3843 0.3843 1>}
      uVRegion {bIsValid=false, max=<0 0>, min=<0 0>}
     }
     pressedForeground {colorUseRule=UseColor_Specified, specifiedColor=<rgba 0 1 1 1>}
     pressedPadding=<ltrb 0 1.5 0 1.5>
     pressedSlateSound {resourceObject=null}
    }
   }
   slot {class=SizeBoxSlot, properties {padding=<ltrb 0 0 1 0>}}
  }
  UEFN_TextBlock_C_55 {
   widgetClassPath=&UEFN_TextBlock
   parent=Button_RemoveItem
   widget=UEFN_TextBlock_C_55
   properties {
    text=x
    font {
     bForceMonospaced=false
     bMaterialIsStencil=false
     fontMaterial=null
     fontObject=&BurbankSmall
     letterSpacing=0
     monospacedWidth=1
     size=7.5
     skewAmount=0
     typefaceFontName=Black
     outlineSettings {
      bApplyOutlineToDropShadows=false
      bMiteredCorners=false
      bSeparateFillAlpha=false
      outlineMaterial=null
      outlineSize=0
      outlineColor=<rgba 0 0 0 1>
     }
    }
   }
   slot {class=ButtonSlot, properties {padding=<ltrb 0 0 0 0>}}
  }
 ]
}
