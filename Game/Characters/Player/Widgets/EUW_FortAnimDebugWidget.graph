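# Editor utility widget that serves as a control panel for debugging Fortnite character animation during Play In Editor.
# Buttons and combo boxes issue console commands to the first PIE world for debug targets, bone space and IK overrides.
# Property and bone filters, including IK presets, refresh as removable tokens.
# Tick polls the target's item animation assets into asset rows and gates the UI on a game world and the main debug cvar.
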EditorUtilityWidgetBlueprint /Game/Characters/Player/Widgets/EUW_FortAnimDebugWidget
graphs [
 AddBoneFilters {nodes [
   n1 = AddBoneFilters!() -> (then, Filters:Array of Strings, ExactMatch:bool=~) {is=title K2Node_FunctionEntry}
   > n2 = For Each Loop(Array=@n1)
   Return Node(execute=@n2.Completed)
   > RunStringCommand!(execute, self, Command:Text="LOCTEXT(\"AddBoneFilter \")", String:Text="To Text (String)"(InString=@n2)) -> (then)
  ]}
 AddPropertyFilters {nodes [
   n1 = AddPropertyFilters!() -> (then, Filters:Array of Strings, ExactMatch:bool=~) {is=title K2Node_FunctionEntry}
   > n2 = Sequence()
   > n3 = Branch()
   > Set ExactMatchString(ExactMatchString:String=" 1") -> (Output_Get:String)
   Set ExactMatchString(execute=@n3.else, ExactMatchString=" 0")
   n4 = For Each Loop(Exec=@n2.then_1, Array=@n1)
   Return Node(execute=@n4.Completed)
   n5 = Get ExactMatchString() -> (ExactMatchString:String)
   > RunStringCommand(Command="LOCTEXT(\"AddPropertyFilter \")", String="To Text (String)"(InString=Append(A=@n4, B=@n5)))
  ]}
 AddWrapBoxEntry {nodes [
   n1 = AddWrapBoxEntry!() -> (then, FilterText:Text, WrapBox:Wrap Box Object Reference) {is=title K2Node_FunctionEntry}
   > Set WrapBoxVar(WrapBoxVar:Wrap Box Object Reference=@n1.WrapBox) -> (Output_Get:Wrap Box Object Reference)
   > n2 = Set FilterTextVar(FilterTextVar:Text=@n1) -> (Output_Get:Text)
   n3 = Trim!(SourceString:String="To String (Text)"(InText=@n2)) -> (ReturnValue:String)
   > Branch(Condition="integer > integer"(A=Len(S=@n3)))
   > n4 = Create Widget!(
    execute:Exec
    WorldContextObject:Object Reference
    Class:User Widget Class Reference=&/WB_FAnim_ListboxToken.WB_FAnim_ListboxToken_C
    OwningPlayer:Player Controller Object Reference
   ) -> (then:Exec, ReturnValue:WB FAnim Listbox Token Object Reference)
   > n5 = Set NewWidget(NewWidget:WB FAnim Listbox Token Object Reference=@n4) -> (Output_Get:WB FAnim Listbox Token Object Reference)
   > SetParent!(
    execute
    self:WB FAnim Listbox Token Object Reference=@n5
    ParentListView:List View Object Reference
    ParentWrapBox:Wrap Box Object Reference=Get WrapBoxVar() -> (WrapBoxVar:Wrap Box Object Reference)
   ) -> (then)
   n6 = Get NewWidget() -> (NewWidget:WB FAnim Listbox Token Object Reference)
   > Set Text(Text=Get FilterTextVar() -> (FilterTextVar:Text), self:WB FAnim Listbox Token Object Reference=@n6)
   > UpdateText(self:WB FAnim Listbox Token Object Reference=@n6)
   n7 = Get NewWidget()
   > AddChild(self=Get WrapBoxVar(), Content=@n7)
   > Return Node!(execute, CreatedWidget:WB FAnim Listbox Token Object Reference=@n7)
  ]}
 EventGraph {nodes [
   "On Clicked (Button_AddPropertyFilter)"() {is=title K2Node_ComponentBoundEvent}
   > RunPropertyFilterCommand!(
    execute
    self
    SringCommandButton:EUW FAnim Simple String Console Button Object Reference=Get Button_AddPropertyFilter() -> (Button_AddPropertyFilter:EUW FAnim Simple String Console Button Object Reference)
   ) -> (then)
   > Set Text(self:Text Box Object Reference=Get TextBox_PropertyFilter() -> (TextBox_PropertyFilter:Editor Utility Editable Text Box Object Reference))
   "On Clicked (Button_AddBoneFilter)"() {is=title K2Node_ComponentBoundEvent}
   > RunBoneFilterCommand!(
    execute
    self
    SringCommandButton:EUW FAnim Simple String Console Button Object Reference=Get Button_AddBoneFilter() -> (Button_AddBoneFilter:EUW FAnim Simple String Console Button Object Reference)
   ) -> (then)
   > Set Text(self=Get TextBox_BoneFilter() -> (TextBox_BoneFilter:Editor Utility Editable Text Box Object Reference))
   "On Selection Changed (ComboBox_HandIK)"!() -> (OutputDelegate, then, SelectedItem:String, SelectionType:ESelectInfo Enum) {is=title K2Node_ComponentBoundEvent}
   n1 = GetSelectedIndex!(self:"Combo Box (String) Object Reference"=Get ComboBox_HandIK() -> (ComboBox_HandIK:"Combo Box (String) Object Reference")) -> (ReturnValue:int=0)
   > n2 = Branch(Condition="Equal (Integer)"(A=@n1))
   > n3 = RunConsoleCommand!(execute, self, Command:Text="LOCTEXT(\"a.fort.handikoverride -1\")", RunOnServer:bool=false) -> (then)
   RunConsoleCommand(execute=[@n5.then, @n6.then, @n3.then], Command="LOCTEXT(\"a.fort.lefthandikoverride -1\")", RunOnServer=false)
   > RunConsoleCommand(Command="LOCTEXT(\"a.fort.righthandikoverride -1\")", RunOnServer=false)
   n4 = Branch(execute=@n2.else, Condition="Equal (Integer)"(A=@n1, B=1))
   > n5 = RunConsoleCommand(Command="LOCTEXT(\"a.fort.handikoverride 1\")", RunOnServer=false)
   n6 = RunConsoleCommand(execute=@n4.else, Command="LOCTEXT(\"a.fort.handikoverride 0\")", RunOnServer=false)
   "On Clicked (Button_AddIKFilters)"() {is=title K2Node_ComponentBoundEvent}
   n7 = Get IKFilters() -> (IKFilters:Array of Strings)
   > n8 = AddPropertyFilters!(execute, self, Filters:Array of Strings=@n7, ExactMatch:bool=true) -> (then) {is=title K2Node_CallFunction}
   n9 = GetActivePropertyFilters!(execute=[@n41.then, @n8.then, @n12.then, @n19.then, @n21.then, @n29.then_1]) -> (then, OutProperties:Array of Strings)
   > n10 = RefreshWrapBox!(
    execute
    self
    WrapBox:Wrap Box Object Reference=Get WrapBox_PropertyFilters() -> (WrapBox_PropertyFilters:Wrap Box Object Reference)
    TextArray:Array of Strings=@n9
   ) -> (then, CreatedWidget:Array of WB FAnim Listbox Token Object References)
   n11 = OnPropertyFilterRemoved!() -> (OutputDelegate, then, RemovedItem:Text) {is=title K2Node_CustomEvent}
   Bind Event to On Item Removed(execute=@n10.then, self:WB FAnim Listbox Token Object Reference=@n10, Delegate=@n11) {is=title K2Node_AddDelegate}
   "On Clicked (Button_RemoveIKFilters)"() {is=title K2Node_ComponentBoundEvent}
   > n12 = RemovePropertyFilters!(execute, self, Filters:Array of Strings=@n7) -> (then)
   "On Clicked (Button_AddIKBones)"() {is=title K2Node_ComponentBoundEvent}
   n13 = Get IKBones() -> (IKBones:Array of Strings)
   > n14 = AddBoneFilters!(execute, self, Filters:Array of Strings=@n13, ExactMatch:bool=false) -> (then) {is=title K2Node_CallFunction}
   n15 = GetActiveBoneFilters!(execute=[@n40.then, @n29.then_2, @n18.then, @n14.then, @n23.then, @n24.then]) -> (then, OutBones:Array of Strings)
   > n16 = RefreshWrapBox(WrapBox=Get WrapBox_BoneFilters() -> (WrapBox_BoneFilters:Wrap Box Object Reference), TextArray=@n15)
   n17 = OnBoneFilterRemoved!() -> (OutputDelegate, then, RemovedItem:Text) {is=title K2Node_CustomEvent}
   Bind Event to On Item Removed(execute=@n16.then, self=@n16, Delegate=@n17)
   "On Post Clicked (Button_EnableMainAnimDebug)"() {is=title K2Node_ComponentBoundEvent}
   > UpdateButtonColor(self:EUW FAnim Simple Console Button Object Reference=Get Button_GameplayDebugInfo() -> (Button_GameplayDebugInfo:EUW FAnim Simple Console Button Object Reference))
   "On Clicked (Button_RemoveIKBones)"() {is=title K2Node_ComponentBoundEvent}
   > n18 = RemoveBoneFilters!(execute, self, Filters:Array of Strings=@n13) -> (then)
   n19 = "On Post Clicked (Button_AddPropertyFilter)"!() -> (OutputDelegate, then, CommandExecuted:bool) {is=title K2Node_ComponentBoundEvent}
   n20 = "On Text Committed (TextBox_PropertyFilter)"!() -> (OutputDelegate, then, Text:"Text (by ref)", CommitMethod:ETextCommit Enum=Default) {is=title K2Node_ComponentBoundEvent}
   > OnFilterTextCommitted!(
    execute
    self
    TextCommitMethd:ETextCommit Enum=@n20.CommitMethod
    AddFilterButton:EUW FAnim Simple String Console Button Object Reference=Get Button_AddPropertyFilter()
   ) -> (then)
   n21 = "On Post Clicked (Button_ResetPropertyFilter)"() {is=title K2Node_ComponentBoundEvent}
   n22 = "On Text Committed (TextBox_BoneFilter)"!() -> (OutputDelegate, then, Text:"Text (by ref)", CommitMethod:ETextCommit Enum=Default) {is=title K2Node_ComponentBoundEvent}
   > OnFilterTextCommitted(TextCommitMethd=@n22.CommitMethod, AddFilterButton=Get Button_AddBoneFilter())
   n23 = "On Post Clicked (Button_ResetBoneFilter)"() {is=title K2Node_ComponentBoundEvent}
   n24 = "On Post Clicked (Button_AddBoneFilter)"!() -> (OutputDelegate, then, CommandExecuted:bool) {is=title K2Node_ComponentBoundEvent}
   "On Selection Changed (ComboBox_BoneTransformSpace)"!() -> (OutputDelegate, then, SelectedItem:String, SelectionType:ESelectInfo Enum) {is=title K2Node_ComponentBoundEvent}
   n25 = "To String (Integer)"(InInt=GetSelectedIndex(self=Get ComboBox_BoneTransformSpace() -> (ComboBox_BoneTransformSpace:"Combo Box (String) Object Reference")) -> (ReturnValue=0))
   > RunConsoleCommand(Command="To Text (String)"(InString=Append(A="Fort.Anim.Debug.BoneTransformSpace ", B=@n25)), RunOnServer=false)
   "On Selection Changed (ComboBox_LegIK)"!() -> (OutputDelegate, then, SelectedItem:String, SelectionType:ESelectInfo Enum) {is=title K2Node_ComponentBoundEvent}
   n26 = GetSelectedIndex(self=Get ComboBox_LegIK() -> (ComboBox_LegIK:"Combo Box (String) Object Reference")) -> (ReturnValue=0)
   > n27 = Branch(Condition="Equal (Integer)"(A=@n26))
   > RunConsoleCommand(Command="LOCTEXT(\"a.fort.legikoverride -1\")", RunOnServer=false)
   n28 = Branch(execute=@n27.else, Condition="Equal (Integer)"(A=@n26, B=1))
   > RunConsoleCommand(Command="LOCTEXT(\"a.fort.legikoverride 1\")", RunOnServer=false)
   RunConsoleCommand(execute=@n28.else, Command="LOCTEXT(\"a.fort.legikoverride 0\")", RunOnServer=false)
   Event Construct()
   > n29 = Sequence!(execute) -> (then_0, then_1, then_2:Exec)
   > InitControls()
   Event Tick()
   > n30 = Sequence()
   n31 = Get ItemAnimAssetDebugInfo() -> (ItemAnimAssetDebugInfo:Item Anim Asset Debug Info Structure)
   > GetCurrentDebugTargetItemAnimAssets!(execute, ItemAnimAssetDebugInfo:"Item Anim Asset Debug Info Structure (by ref)"=@n31) -> (then)
   n32 = Break Item Anim Asset Debug Info!(ItemAnimAssetDebugInfo:"Item Anim Asset Debug Info Structure (by ref)"=@n31) -> (
    ProxyTable:Proxy Table Object Reference
    WeaponAnimSet:Fort Weapon Anim Set Object Reference
    WeaponAdditiveAnimSet:Fort Weapon Additive Anim Set Object Reference
    TacticalStateAnimAsset:Fort Tactical State Anim Asset Object Reference
    ItemChooserData:Blueprint Object Reference
    ItemAnimLayer:Blueprint Object Reference
   ) {is=title K2Node_BreakStruct}
   > SetAsset!(
    execute
    self:EUW FAnim Asset Edit Control Object Reference=Get ProxyTableAssetRow() -> (ProxyTableAssetRow:EUW FAnim Asset Edit Control Object Reference)
    InAsset:Object Reference=@n32
   ) -> (then)
   > SetAsset(
    self=Get WeaponAnimSet_AssetControl() -> (WeaponAnimSet_AssetControl:EUW FAnim Asset Edit Control Object Reference)
    InAsset=@n32.WeaponAnimSet
   )
   > SetAsset(
    self=Get WeaponAdditiveAnimSet_AssetControl() -> (WeaponAdditiveAnimSet_AssetControl:EUW FAnim Asset Edit Control Object Reference)
    InAsset=@n32.WeaponAdditiveAnimSet
   )
   > SetAsset(
    self=Get TacticalAnimSet_AssetControl() -> (TacticalAnimSet_AssetControl:EUW FAnim Asset Edit Control Object Reference)
    InAsset=@n32.TacticalStateAnimAsset
   )
   > SetAsset(self=Get ItemLayer_AssetControl() -> (ItemLayer_AssetControl:EUW FAnim Asset Edit Control Object Reference), InAsset=@n32.ItemAnimLayer)
   > SetAsset(
    self=Get ChooserData_AssetControl() -> (ChooserData_AssetControl:EUW FAnim Asset Edit Control Object Reference)
    InAsset=@n32.ItemChooserData
   )
   n33 = GetGameWorld!(execute=@n30.then_1, self:Unreal Editor Subsystem Object Reference=Get UnrealEditorSubsystem()) -> (then, ReturnValue:World Object Reference)
   > n34 = Branch(Condition=IsValid(Object=@n33))
   n35 = Get IsEnabled() -> (IsEnabled:bool)
   > n36 = Branch(Condition=@n35)
   Set IsEnabled(execute=@n36.else, IsEnabled:bool=true) -> (Output_Get:bool)
   > Set bIsEnabled!(
    execute
    bIsEnabled:bool=true
    self:Widget Object Reference=Get MainScrollBox() -> (MainScrollBox:Editor Utility Scroll Box Object Reference)
   ) -> (then, Output_Get:bool)
   n37 = Get TextBox_Main() -> (TextBox_Main:WB FAnim Simple Text Block Object Reference)
   > UpdateText!(execute, self:WB FAnim Simple Text Block Object Reference=@n37, Text:Text="LOCTEXT(\"Main\")") -> (then)
   Branch(execute=@n34.else, Condition=@n35)
   > Set IsEnabled(IsEnabled=false)
   > Set bIsEnabled(bIsEnabled=false, self=Get MainScrollBox())
   > UpdateText(self=@n37, Text="LOCTEXT(\"START PIE TO ENABLE UI\")")
   n38 = GetConsoleVariableIntValue!(
    execute=@n30.then_2
    VariableName:String=Get Console Command!(self:EUW FAnim Simple Console Button Object Reference=Get Button_EnableMainAnimDebug() -> (Button_EnableMainAnimDebug:EUW FAnim Simple Console Button Object Reference)) -> (Console Command:String)
   ) -> (then, ReturnValue:int=0)
   n39 = "Not Equal (Integer)"(A=@n38)
   > SetIsEnabled(
    self=Get Expandable_ProceduralLayering() -> (Expandable_ProceduralLayering:Editor Utility Expandable Area Object Reference)
    bInIsEnabled=@n39
   )
   > SetIsEnabled(self=Get HorizontalBox_Targets() -> (HorizontalBox_Targets:Horizontal Box Object Reference), bInIsEnabled=@n39)
   > SetIsEnabled(self=Get HorizontalBox_PropertyFilters() -> (HorizontalBox_PropertyFilters:Horizontal Box Object Reference), bInIsEnabled=@n39)
   > SetIsEnabled(self=Get HorizontalBox_BoneFilters() -> (HorizontalBox_BoneFilters:Horizontal Box Object Reference), bInIsEnabled=@n39)
   > SetIsEnabled(self=Get HorizontalBox_PropertyFilterTokens() -> (HorizontalBox_PropertyFilterTokens:Horizontal Box Object Reference), bInIsEnabled=@n39)
   > SetIsEnabled(self=Get HorizontalBox_BoneFilterTokens() -> (HorizontalBox_BoneFilterTokens:Horizontal Box Object Reference), bInIsEnabled=@n39)
   > SetIsEnabled(self=Get Expandable_AnimationAssets() -> (Expandable_AnimationAssets:Editor Utility Expandable Area Object Reference), bInIsEnabled=@n39)
   > SetIsEnabled(self=Get Expandable_IK() -> (Expandable_IK:Editor Utility Expandable Area Object Reference), bInIsEnabled=@n39)
   > SetIsEnabled(self=Get Expandable_2DDisplay() -> (Expandable_2DDisplay:Editor Utility Expandable Area Object Reference), bInIsEnabled=@n39)
   > SetIsEnabled(self=Get Expandable_3DDisplay() -> (Expandable_3DDisplay:Editor Utility Expandable Area Object Reference), bInIsEnabled=@n39)
   Event PreConstruct()
   n40 = RunStringCommand(execute=@n17.then, Command="LOCTEXT(\"RemoveBoneFilter\")", String=@n17.RemovedItem)
   n41 = RunPropertyRemovedCommand!(execute=@n11.then, self, PropertyName:Text=@n11.RemovedItem) -> (then)
  ]}
 InitControls {nodes [
   InitControls() {is=title K2Node_FunctionEntry}
   > n1 = Sequence!(execute) -> (then_0, then_1)
   n2 = Get ComboBox_BoneTransformSpace()
   > AddOption!(execute, self:"Combo Box (String) Object Reference"=@n2, Option:String=World) -> (then)
   > AddOption(self=@n2, Option=Mesh)
   > AddOption(self=@n2, Option=Local)
   > n3 = GetConsoleVariableIntValue(VariableName=Fort.Anim.Debug.BoneTransformSpace) -> (ReturnValue=0)
   > n4 = Branch(Condition="In Range (Integer)"!(Value:int=@n3, Min:int=0, Max:int=2, InclusiveMin:bool=true, InclusiveMax:bool=true) -> (ReturnValue:bool))
   n5 = Get ComboBox_BoneTransformSpace()
   > SetSelectedIndex!(execute, self:"Combo Box (String) Object Reference"=@n5, Index:int=@n3) -> (then)
   SetSelectedIndex(execute=@n4.else, self=@n5, Index=1)
   SetupIKComboBox!(execute=@n1.then_1, self, ComboBox:"Combo Box (String) Object Reference"=Get ComboBox_LegIK(), VariableName:String=a.Fort.LegIKOverride) -> (then)
   > SetupIKComboBox(ComboBox=Get ComboBox_HandIK(), VariableName=a.Fort.HandIkOverride)
  ]}
 OnFilterTextCommitted {nodes [
   n1 = OnFilterTextCommitted!() -> (then, TextCommitMethd:ETextCommit Enum=Default, AddFilterButton:EUW FAnim Simple String Console Button Object Reference) {is=title K2Node_FunctionEntry}
   > Branch(Condition="Equal (Enum)"(A:ETextCommit Enum=@n1, B:ETextCommit Enum=OnEnter))
   > ExecuteClick(self:EUW FAnim Simple String Console Button Object Reference=@n1.AddFilterButton)
   > Return Node!(execute)
  ]}
 OnProxyTableChanged {nodes [
   OnProxyTableChanged() {is=title K2Node_FunctionEntry}
   n1 = Get ItemAnimAssetDebugInfo()
   > n2 = Branch(Condition=IsValid(Object=Break Item Anim Asset Debug Info(ItemAnimAssetDebugInfo=@n1)))
   Return Node(execute=@n2.else)
  ]}
 RefreshWrapBox {nodes [
   n1 = RefreshWrapBox!() -> (then, WrapBox:Wrap Box Object Reference, TextArray:Array of Strings) {is=title K2Node_FunctionEntry}
   > Set WrapBoxVar(WrapBoxVar=@n1)
   > ClearChildren(self=Get WrapBoxVar())
   > n2 = For Each Loop(Array=@n1.TextArray)
   > n3 = AddWrapBoxEntry!(execute, self, FilterText:Text="To Text (String)"(InString=@n2), WrapBox:Wrap Box Object Reference=Get WrapBoxVar()) -> (then, CreatedWidget:WB FAnim Listbox Token Object Reference) {is=title K2Node_CallFunction}
   n4 = Get CreatedWidgets() -> (CreatedWidgets:Array of WB FAnim Listbox Token Object References)
   > Add(TargetArray:Array of WB FAnim Listbox Token Object References=@n4, NewItem:"WB FAnim Listbox Token Object Reference (by ref)"=@n3) {is=title K2Node_CallArrayFunction}
   Return Node!(execute=@n2.Completed, CreatedWidget:Array of WB FAnim Listbox Token Object References=@n4)
  ]}
 RemoveBoneFilters {nodes [
   n1 = RemoveBoneFilters!() -> (then, Filters:Array of Strings) {is=title K2Node_FunctionEntry}
   > n2 = Sequence()
   n3 = For Each Loop(Exec=@n2.then_1, Array=@n1)
   Return Node!(execute=@n3.Completed)
   > RunStringCommand(Command="LOCTEXT(\"RemoveBoneFilter \")", String="To Text (String)"(InString=@n3))
  ]}
 RemovePropertyFilters {nodes [
   n1 = RemovePropertyFilters!() -> (then, Filters:Array of Strings) {is=title K2Node_FunctionEntry}
   > n2 = Sequence()
   n3 = For Each Loop(Exec=@n2.then_1, Array=@n1)
   Return Node(execute=@n3.Completed)
   > RunStringCommand(Command="LOCTEXT(\"RemovePropertyFilter \")", String="To Text (String)"(InString=@n3))
  ]}
 RunBoneFilterCommand {nodes [
   n1 = RunBoneFilterCommand!() -> (then, SringCommandButton:EUW FAnim Simple String Console Button Object Reference) {is=title K2Node_FunctionEntry}
   n2 = Get Text(self=Get TextBox_BoneFilter())
   > Set Console Command String Parameter!(
    execute
    Console Command String Parameter:String=Trim(SourceString="To String (Text)"(InText=@n2))
    self:EUW FAnim Simple String Console Button Object Reference=@n1
   ) -> (then, Output_Get:String)
   > Return Node()
  ]}
 RunConsoleCommand {nodes [
   n1 = RunConsoleCommand!() -> (then, Command:Text, RunOnServer:bool=~) {is=title K2Node_FunctionEntry}
   > n2 = GetPIEWorlds!(execute, bIncludeDedicatedServer:bool=@n1.RunOnServer) -> (then, ReturnValue:Array of World Object References)
   > n3 = Branch(Condition=Is Not Empty!(TargetArray:Array of World Object References=@n2) -> (ReturnValue:bool))
   > ExecuteConsoleCommand(
    WorldContextObject="Get (a copy)"(Array:Array of World Object References=@n2) -> (Output:World Object Reference)
    Command="To String (Text)"(InText=@n1)
   )
   ExecuteConsoleCommand(execute=@n3.else, Command="To String (Text)"(InText=@n1))
  ]}
 RunPropertyFilterCommand {nodes [
   n1 = RunPropertyFilterCommand!() -> (then, SringCommandButton:EUW FAnim Simple String Console Button Object Reference) {is=title K2Node_FunctionEntry}
   > Set StringCommandButton(StringCommandButton:EUW FAnim Simple String Console Button Object Reference=@n1) -> (Output_Get:EUW FAnim Simple String Console Button Object Reference)
   n2 = Get TextBox_PropertyFilter()
   n3 = Trim(SourceString="To String (Text)"(InText=Get Text(self=@n2)))
   > Set FilterString(FilterString:String=@n3) -> (Output_Get:String)
   > n4 = Branch(Condition="integer > integer"(A=Len(S=@n3)))
   n5 = Get CheckBox_ExactMatch() -> (CheckBox_ExactMatch:Editor Utility Check Box Object Reference)
   > n6 = Branch(Condition="Equal (Enum)"(
     A:ECheckBoxState Enum=Get CheckedState!(self:Check Box Object Reference=@n5) -> (CheckedState:ECheckBoxState Enum=Unchecked)
     B:ECheckBoxState Enum=Checked
    ))
   > n7 = Set Console Command String Parameter(
    Console Command String Parameter=Append(A=Get FilterString() -> (FilterString:String), B=" 1")
    self=Get StringCommandButton() -> (StringCommandButton:EUW FAnim Simple String Console Button Object Reference)
   )
   Return Node(execute=[@n7.then, @n9.then, @n8.then])
   n8 = Set Console Command String Parameter(execute=@n6.else, Console Command String Parameter=Append(A=Get FilterString(), B=" 0"), self=Get StringCommandButton())
   n9 = Set Console Command String Parameter(execute=@n4.else, self=Get StringCommandButton())
  ]}
 RunPropertyRemovedCommand {nodes [
   n1 = RunPropertyRemovedCommand!() -> (then, PropertyName:Text) {is=title K2Node_FunctionEntry}
   > n2 = GetPIEWorlds(bIncludeDedicatedServer=false)
   > n3 = Branch(Condition=Is Not Empty(TargetArray=@n2))
   n4 = Append!(A="RemovePropertyFilter ", B="To String (Text)"(InText=@n1), C:String=" 0") -> (ReturnValue)
   > ExecuteConsoleCommand(WorldContextObject="Get (a copy)"(Array=@n2), Command=@n4)
   ExecuteConsoleCommand(execute=@n3.else, Command=@n4)
  ]}
 RunStringCommand {nodes [
   n1 = RunStringCommand!() -> (then, Command:Text, String:Text) {is=title K2Node_FunctionEntry}
   > n2 = GetPIEWorlds(bIncludeDedicatedServer=false)
   > n3 = Branch(Condition=Is Not Empty(TargetArray=@n2))
   n4 = Append(A="To String (Text)"(InText=@n1), B=" ", C="To String (Text)"(InText=@n1.String))
   > ExecuteConsoleCommand(WorldContextObject="Get (a copy)"(Array=@n2), Command=@n4)
   ExecuteConsoleCommand(execute=@n3.else, Command=@n4)
  ]}
 SetupIKComboBox {nodes [
   n1 = SetupIKComboBox!() -> (then, ComboBox:"Combo Box (String) Object Reference", VariableName:String) {is=title K2Node_FunctionEntry}
   > Set ComboBoxVar(ComboBoxVar:"Combo Box (String) Object Reference"=@n1) -> (Output_Get:"Combo Box (String) Object Reference")
   n2 = Get ComboBoxVar() -> (ComboBoxVar:"Combo Box (String) Object Reference")
   > AddOption(self=@n2, Option=Default)
   > AddOption(self=@n2, Option=Force IK)
   > AddOption(self=@n2, Option=Force FK)
   > n3 = GetConsoleVariableFloatValue!(execute, VariableName:String=@n1.VariableName) -> (then, ReturnValue:float)
   > n4 = Branch(Condition="float < float"(A=@n3))
   n5 = Get ComboBoxVar()
   > SetSelectedIndex(self=@n5, Index=0)
   n6 = Branch(execute=@n4.else, Condition="float > float"(B=@n3))
   > SetSelectedIndex(self=@n5, Index=1)
   SetSelectedIndex(execute=@n6.else, self=@n5, Index=2)
  ]}
]
variables [
 IKBones {type="array<string>", default [ik_hand_l,ik_hand_r,ik_foot_l,ik_foot_r,ik_hand_gun]}
 IKFilters {type="array<string>", default [LeftHandIkAlpha, RightHandIkAlpha, LegIkAlpha, LeftLegIkAlpha, RightLegIkAlpha]}
 IsEnabled {type=boolean, default=true}
 ItemAnimAssetDebugInfo {
  type=ItemAnimAssetDebugInfo
  default {itemAnimLayer=null, itemChooserData=null, proxyTable=null, tacticalStateAnimAsset=null, weaponAdditiveAnimSet=null, weaponAnimSet=null}
 }
 RefreshBoneFilters {type=boolean, default=true}
]
widgets {
 info {namedSlotCount=14, rootWidgetClass=&Border, widgetCount=105}
 widgets [
  Border_0 {widgetClassPath=&Border, widget=Border_0, properties {brushColor=<rgba 0.0118 0.0118 0.0118 1>}}
  MainScrollBox {widgetClassPath=&EditorUtilityScrollBox, bIsVariable=true, parent=Border_0, widget=MainScrollBox, slot {class=BorderSlot}}
  Expandable_Main {
   widgetClassPath=/Script/Blutility.EditorUtilityExpandableArea
   parent=MainScrollBox
   widget=Expandable_Main
   properties {bIsExpanded=true}
   slot {class=ScrollBoxSlot}
  }
  TextBox_Main {
   widgetClassPath=&/WB_FAnim_SimpleTextBlock
   bIsVariable=true
   namedSlotHost=Expandable_Main
   widget=TextBox_Main
   properties {fontSize=11, isBold=true, text=Main}
  }
  VerticalBox {widgetClassPath=&VerticalBox, namedSlotHost=Expandable_Main, widget=VerticalBox}
  HorizontalBox {widgetClassPath=&HorizontalBox, parent=VerticalBox, widget=HorizontalBox, slot {class=VerticalBoxSlot}}
  Button_EnableMainAnimDebug {
   widgetClassPath=&/EUW_FAnim_SimpleConsoleButton
   bIsVariable=true
   parent=HorizontalBox
   widget=Button_EnableMainAnimDebug
   properties {buttonName=Enable Main Debug, command Has Int State=true, console Command=Fort.Anim.Debug}
   slot {class=HorizontalBoxSlot, properties {size {sizeRule=Fill, value=1}}}
  }
  HorizontalBox_Targets {
   widgetClassPath=&HorizontalBox
   bIsVariable=true
   parent=VerticalBox
   widget=HorizontalBox_Targets
   slot {class=VerticalBoxSlot, properties {size {sizeRule=Fill, value=1}}}
  }
  EUW_SimpleConsoleButton_9 {
   widgetClassPath=&EUW_FAnim_SimpleConsoleButton
   parent=HorizontalBox_Targets
   widget=EUW_SimpleConsoleButton_9
   properties {buttonName=Previous Target, console Command=PreviousDebugTarget, run On Server=false}
   slot {class=HorizontalBoxSlot, properties {size {sizeRule=Fill, value=1}}}
  }
  EUW_SimpleConsoleButton_8 {
   widgetClassPath=&EUW_FAnim_SimpleConsoleButton
   parent=HorizontalBox_Targets
   widget=EUW_SimpleConsoleButton_8
   properties {buttonName=Next Target, console Command=NextDebugTarget, run On Server=false}
   slot {class=HorizontalBoxSlot, properties {size {sizeRule=Fill, value=1}}}
  }
  HorizontalBox_PropertyFilters {
   widgetClassPath=&HorizontalBox
   bIsVariable=true
   parent=VerticalBox
   widget=HorizontalBox_PropertyFilters
   slot {class=VerticalBoxSlot, properties {size {sizeRule=Fill, value=1}}}
  }
  HorizontalBox_496 {
   widgetClassPath=&HorizontalBox
   parent=HorizontalBox_PropertyFilters
   widget=HorizontalBox_496
   slot {class=HorizontalBoxSlot, properties {size {sizeRule=Fill, value=0.75}}}
  }
  WB_SimpleTextBlock_718 {
   widgetClassPath=&WB_FAnim_SimpleTextBlock
   parent=HorizontalBox_496
   widget=WB_SimpleTextBlock_718
   properties {
    text=Property filters
    toolTipText="Add properties matching entered text for debug. Enable exact match to only list properties exactly matching the name.\r\nProperties that exist in multiple anim instances and have the same value will only be showned on the first anim instance that has it."
    boxSize=<90 25>
   }
   slot {class=HorizontalBoxSlot}
  }
  TextBox_PropertyFilter {
   widgetClassPath=&EditorUtilityEditableTextBox
   bIsVariable=true
   parent=HorizontalBox_496
   widget=TextBox_PropertyFilter
   slot {class=HorizontalBoxSlot, properties {size {sizeRule=Fill, value=1}}}
  }
  Button_AddPropertyFilter {
   widgetClassPath=&/EUW_FAnim_SimpleStringConsoleButton
   bIsVariable=true
   parent=HorizontalBox_496
   widget=Button_AddPropertyFilter
   properties {
    button Width=50
    buttonName=Add
    console Command=AddPropertyFilter
    run On Server=false
    toolTipText="Add properties matching entered text for debug. Enable exact match to only list properties exactly matching the name.\r\nProperties that exist in multiple anim instances and have the same value will only be showned on the first anim instance that has it."
   }
   slot {class=HorizontalBoxSlot}
  }
  HorizontalBox_749 {
   widgetClassPath=&HorizontalBox
   parent=HorizontalBox_PropertyFilters
   widget=HorizontalBox_749
   slot {class=HorizontalBoxSlot, properties {size {sizeRule=Fill, value=1}}}
  }
  WB_SimpleTextBlock_203 {
   widgetClassPath=&WB_FAnim_SimpleTextBlock
   parent=HorizontalBox_749
   widget=WB_SimpleTextBlock_203
   properties {text=Exact match, boxSize=<90 25>, textPadding=<ltrb 5 5 5 5>}
   slot {class=HorizontalBoxSlot}
  }
  CheckBox_ExactMatch {
   widgetClassPath=&EditorUtilityCheckBox
   bIsVariable=true
   parent=HorizontalBox_749
   widget=CheckBox_ExactMatch
   properties {
    horizontalAlignment=HAlign_Center
    widgetStyle {
     checkBoxType=CheckBox
     backgroundHoveredImage {
      bIsDynamicallyLoaded=false
      drawAs=RoundedBox
      imageType=NoImage
      mirroring=NoMirror
      resourceName=None
      resourceObject=null
      tiling=NoTile
      imageSize=<18 18>
      margin=<ltrb 0 0 0 0>
      outlineSettings {
       bUseBrushTransparency=false
       roundingType=FixedRadius
       width=1
       color {colorUseRule=UseColor_ColorTable, specifiedColor=<rgba 0 0 0 0>}
       cornerRadii=<3 3 3 3>
      }
      tintColor {colorUseRule=UseColor_Specified, specifiedColor=<rgba 0.0048 0.0048 0.0048 1>}
      uVRegion {bIsValid=false, max=<0 0>, min=<0 0>}
     }
     backgroundImage {
      bIsDynamicallyLoaded=false
      drawAs=RoundedBox
      imageType=NoImage
      mirroring=NoMirror
      resourceName=None
      resourceObject=null
      tiling=NoTile
      imageSize=<18 18>
      margin=<ltrb 0 0 0 0>
      outlineSettings {
       bUseBrushTransparency=false
       roundingType=FixedRadius
       width=1
       color {colorUseRule=UseColor_ColorTable, specifiedColor=<rgba 0 0 0 0>}
       cornerRadii=<3 3 3 3>
      }
      tintColor {colorUseRule=UseColor_Specified, specifiedColor=<rgba 0.0048 0.0048 0.0048 1>}
      uVRegion {bIsValid=false, max=<0 0>, min=<0 0>}
     }
     backgroundPressedImage {
      bIsDynamicallyLoaded=false
      drawAs=RoundedBox
      imageType=NoImage
      mirroring=NoMirror
      resourceName=None
      resourceObject=null
      tiling=NoTile
      imageSize=<18 18>
      margin=<ltrb 0 0 0 0>
      outlineSettings {
       bUseBrushTransparency=false
       roundingType=FixedRadius
       width=1
       color {colorUseRule=UseColor_ColorTable, specifiedColor=<rgba 0 0 0 0>}
       cornerRadii=<3 3 3 3>
      }
      tintColor {colorUseRule=UseColor_Specified, specifiedColor=<rgba 0.0048 0.0048 0.0048 1>}
      uVRegion {bIsValid=false, max=<0 0>, min=<0 0>}
     }
     borderBackgroundColor {colorUseRule=UseColor_Specified, specifiedColor=<rgba 1 1 1 1>}
     checkedForeground {colorUseRule=UseColor_Specified, specifiedColor=<rgba 1 1 1 1>}
     checkedHoveredForeground {colorUseRule=UseColor_Specified, specifiedColor=<rgba 1 1 1 1>}
     checkedHoveredImage {
      bIsDynamicallyLoaded=false
      drawAs=Image
      imageType=Vector
      mirroring=NoMirror
      resourceName=../../../Engine/Content/Slate/Starship/CoreWidgets/CheckBox/check.svg
      resourceObject=null
      tiling=NoTile
      imageSize=<16 16>
      margin=<ltrb 0 0 0 0>
      outlineSettings {
       bUseBrushTransparency=false
       roundingType=HalfHeightRadius
       width=0
       color {colorUseRule=UseColor_Specified, specifiedColor=<rgba 0 0 0 0>}
       cornerRadii=<0 0 0 0>
      }
      tintColor {colorUseRule=UseColor_Specified, specifiedColor=<rgba 0.0044 0.2384 1 1>}
      uVRegion {bIsValid=false, max=<0 0>, min=<0 0>}
     }
     checkedImage {
      bIsDynamicallyLoaded=false
      drawAs=Image
      imageType=Vector
      mirroring=NoMirror
      resourceName=../../../Engine/Content/Slate/Starship/CoreWidgets/CheckBox/check.svg
      resourceObject=null
      tiling=NoTile
      imageSize=<16 16>
      margin=<ltrb 0 0 0 0>
      outlineSettings {
       bUseBrushTransparency=false
       roundingType=HalfHeightRadius
       width=0
       color {colorUseRule=UseColor_Specified, specifiedColor=<rgba 0 0 0 0>}
       cornerRadii=<0 0 0 0>
      }
      tintColor {colorUseRule=UseColor_Specified, specifiedColor=<rgba 0 0.162 0.7454 1>}
      uVRegion {bIsValid=false, max=<0 0>, min=<0 0>}
     }
     checkedPressedForeground {colorUseRule=UseColor_Specified, specifiedColor=<rgba 1 1 1 1>}
     checkedPressedImage {
      bIsDynamicallyLoaded=false
      drawAs=Image
      imageType=Vector
      mirroring=NoMirror
      resourceName=../../../Engine/Content/Slate/Starship/CoreWidgets/CheckBox/check.svg
      resourceObject=null
      tiling=NoTile
      imageSize=<16 16>
      margin=<ltrb 0 0 0 0>
      outlineSettings {
       bUseBrushTransparency=false
       roundingType=HalfHeightRadius
       width=0
       color {colorUseRule=UseColor_Specified, specifiedColor=<rgba 0 0 0 0>}
       cornerRadii=<0 0 0 0>
      }
      tintColor {colorUseRule=UseColor_Specified, specifiedColor=<rgba 0 0.0802 0.3515 1>}
      uVRegion {bIsValid=false, max=<0 0>, min=<0 0>}
     }
     checkedSlateSound {resourceObject=null}
     foregroundColor {colorUseRule=UseColor_Specified, specifiedColor=<rgba 1 1 1 1>}
     hoveredForeground {colorUseRule=UseColor_Specified, specifiedColor=<rgba 1 1 1 1>}
     hoveredSlateSound {resourceObject=null}
     padding=<ltrb 2 0 0 0>
     pressedForeground {colorUseRule=UseColor_Specified, specifiedColor=<rgba 1 1 1 1>}
     uncheckedHoveredImage {
      bIsDynamicallyLoaded=false
      drawAs=NoDrawType
      imageType=NoImage
      mirroring=NoMirror
      resourceName=None
      resourceObject=null
      tiling=NoTile
      imageSize=<0 0>
      margin=<ltrb 0 0 0 0>
      outlineSettings {
       bUseBrushTransparency=false
       roundingType=HalfHeightRadius
       width=0
       color {colorUseRule=UseColor_Specified, specifiedColor=<rgba 0 0 0 0>}
       cornerRadii=<0 0 0 0>
      }
      tintColor {colorUseRule=UseColor_Specified, specifiedColor=<rgba 1 1 1 1>}
      uVRegion {bIsValid=false, max=<0 0>, min=<0 0>}
     }
     uncheckedImage {
      bIsDynamicallyLoaded=false
      drawAs=NoDrawType
      imageType=NoImage
      mirroring=NoMirror
      resourceName=None
      resourceObject=null
      tiling=NoTile
      imageSize=<0 0>
      margin=<ltrb 0 0 0 0>
      outlineSettings {
       bUseBrushTransparency=false
       roundingType=HalfHeightRadius
       width=0
       color {colorUseRule=UseColor_Specified, specifiedColor=<rgba 0 0 0 0>}
       cornerRadii=<0 0 0 0>
      }
      tintColor {colorUseRule=UseColor_Specified, specifiedColor=<rgba 1 1 1 1>}
      uVRegion {bIsValid=false, max=<0 0>, min=<0 0>}
     }
     uncheckedPressedImage {
      bIsDynamicallyLoaded=false
      drawAs=NoDrawType
      imageType=NoImage
      mirroring=NoMirror
      resourceName=None
      resourceObject=null
      tiling=NoTile
      imageSize=<0 0>
      margin=<ltrb 0 0 0 0>
      outlineSettings {
       bUseBrushTransparency=false
       roundingType=HalfHeightRadius
       width=0
       color {colorUseRule=UseColor_Specified, specifiedColor=<rgba 0 0 0 0>}
       cornerRadii=<0 0 0 0>
      }
      tintColor {colorUseRule=UseColor_Specified, specifiedColor=<rgba 1 1 1 1>}
      uVRegion {bIsValid=false, max=<0 0>, min=<0 0>}
     }
     uncheckedSlateSound {resourceObject=null}
     undeterminedForeground {colorUseRule=UseColor_Specified, specifiedColor=<rgba 1 1 1 1>}
     undeterminedHoveredImage {
      bIsDynamicallyLoaded=false
      drawAs=Image
      imageType=Vector
      mirroring=NoMirror
      resourceName=../../../Engine/Content/Slate/Starship/CoreWidgets/CheckBox/indeterminate.svg
      resourceObject=null
      tiling=NoTile
      imageSize=<16 16>
      margin=<ltrb 0 0 0 0>
      outlineSettings {
       bUseBrushTransparency=false
       roundingType=HalfHeightRadius
       width=0
       color {colorUseRule=UseColor_Specified, specifiedColor=<rgba 0 0 0 0>}
       cornerRadii=<0 0 0 0>
      }
      tintColor {colorUseRule=UseColor_Specified, specifiedColor=<rgba 0.0044 0.2384 1 1>}
      uVRegion {bIsValid=false, max=<0 0>, min=<0 0>}
     }
     undeterminedImage {
      bIsDynamicallyLoaded=false
      drawAs=Image
      imageType=Vector
      mirroring=NoMirror
      resourceName=../../../Engine/Content/Slate/Starship/CoreWidgets/CheckBox/indeterminate.svg
      resourceObject=null
      tiling=NoTile
      imageSize=<16 16>
      margin=<ltrb 0 0 0 0>
      outlineSettings {
       bUseBrushTransparency=false
       roundingType=HalfHeightRadius
       width=0
       color {colorUseRule=UseColor_Specified, specifiedColor=<rgba 0 0 0 0>}
       cornerRadii=<0 0 0 0>
      }
      tintColor {colorUseRule=UseColor_Specified, specifiedColor=<rgba 0 0.162 0.7454 1>}
      uVRegion {bIsValid=false, max=<0 0>, min=<0 0>}
     }
     undeterminedPressedImage {
      bIsDynamicallyLoaded=false
      drawAs=Image
      imageType=Vector
      mirroring=NoMirror
      resourceName=../../../Engine/Content/Slate/Starship/CoreWidgets/CheckBox/indeterminate.svg
      resourceObject=null
      tiling=NoTile
      imageSize=<16 16>
      margin=<ltrb 0 0 0 0>
      outlineSettings {
       bUseBrushTransparency=false
       roundingType=HalfHeightRadius
       width=0
       color {colorUseRule=UseColor_Specified, specifiedColor=<rgba 0 0 0 0>}
       cornerRadii=<0 0 0 0>
      }
      tintColor {colorUseRule=UseColor_Specified, specifiedColor=<rgba 0 0.0802 0.3515 1>}
      uVRegion {bIsValid=false, max=<0 0>, min=<0 0>}
     }
    }
   }
   slot {class=HorizontalBoxSlot, properties {horizontalAlignment=HAlign_Center, size {sizeRule=Fill, value=1}}}
  }
  Button_ResetPropertyFilter {
   widgetClassPath=&EUW_FAnim_SimpleConsoleButton
   bIsVariable=true
   parent=HorizontalBox_749
   widget=Button_ResetPropertyFilter
   properties {buttonName=Clear filters, console Command=ResetPropertyFilters, run On Server=false, toolTipText=Clear all active property filters}
   slot {class=HorizontalBoxSlot, properties {padding=<ltrb 1 0 0 0>, size {sizeRule=Fill, value=1}}}
  }
  HorizontalBox_PropertyFilterTokens {
   widgetClassPath=&HorizontalBox
   bIsVariable=true
   parent=VerticalBox
   widget=HorizontalBox_PropertyFilterTokens
   slot {class=VerticalBoxSlot}
  }
  WrapBox_PropertyFilters {
   widgetClassPath=&WrapBox
   bIsVariable=true
   parent=HorizontalBox_PropertyFilterTokens
   widget=WrapBox_PropertyFilters
   slot {class=HorizontalBoxSlot, properties {size {sizeRule=Fill, value=5}}}
  }
  HorizontalBox_BoneFilters {
   widgetClassPath=&HorizontalBox
   bIsVariable=true
   parent=VerticalBox
   widget=HorizontalBox_BoneFilters
   slot {class=VerticalBoxSlot, properties {size {sizeRule=Fill, value=1}}}
  }
  HorizontalBox_658 {
   widgetClassPath=&HorizontalBox
   parent=HorizontalBox_BoneFilters
   widget=HorizontalBox_658
   slot {class=HorizontalBoxSlot, properties {size {sizeRule=Fill, value=0.75}}}
  }
  WB_SimpleTextBlock_2 {
   widgetClassPath=&WB_FAnim_SimpleTextBlock
   parent=HorizontalBox_658
   widget=WB_SimpleTextBlock_2
   properties {text=Bone filters, toolTipText=Add bones for debug. Name must match exact bone name., boxSize=<90 25>}
   slot {class=HorizontalBoxSlot}
  }
  TextBox_BoneFilter {
   widgetClassPath=&EditorUtilityEditableTextBox
   bIsVariable=true
   parent=HorizontalBox_658
   widget=TextBox_BoneFilter
   slot {class=HorizontalBoxSlot, properties {size {sizeRule=Fill, value=1}}}
  }
  Button_AddBoneFilter {
   widgetClassPath=&EUW_FAnim_SimpleStringConsoleButton
   bIsVariable=true
   parent=HorizontalBox_658
   widget=Button_AddBoneFilter
   properties {
    button Width=50
    buttonName=Add
    console Command=AddBoneFilter
    run On Server=false
    toolTipText=Add bones for debug. Name must be exact bone name.
   }
   slot {class=HorizontalBoxSlot}
  }
  HorizontalBox_1146 {
   widgetClassPath=&HorizontalBox
   parent=HorizontalBox_BoneFilters
   widget=HorizontalBox_1146
   slot {class=HorizontalBoxSlot, properties {size {sizeRule=Fill, value=1}}}
  }
  WB_FAnim_SimpleTextBlock {
   widgetClassPath=&WB_FAnim_SimpleTextBlock
   parent=HorizontalBox_1146
   widget=WB_FAnim_SimpleTextBlock
   properties {text=Space, boxSize=<90 25>, textPadding=<ltrb 5 5 5 5>}
   slot {class=HorizontalBoxSlot}
  }
  ComboBox_BoneTransformSpace {
   widgetClassPath=/Script/UMG.ComboBoxString
   bIsVariable=true
   parent=HorizontalBox_1146
   widget=ComboBox_BoneTransformSpace
   properties {
    contentPadding=<ltrb 2 2 2 2>
    font {
     bForceMonospaced=false
     bMaterialIsStencil=false
     fontMaterial=null
     fontObject=&Roboto
     letterSpacing=0
     monospacedWidth=1
     size=9
     skewAmount=0
     typefaceFontName=Regular
     outlineSettings {
      bApplyOutlineToDropShadows=false
      bMiteredCorners=false
      bSeparateFillAlpha=false
      outlineMaterial=null
      outlineSize=0
      outlineColor=<rgba 0 0 0 1>
     }
    }
    foregroundColor {colorUseRule=UseColor_Specified, specifiedColor=<rgba 0.5271 0.5271 0.5271 1>}
    itemStyle {
     bUseParentRowBrush=false
     activeBrush {
      bIsDynamicallyLoaded=false
      drawAs=Image
      imageType=NoImage
      mirroring=NoMirror
      resourceName=None
      resourceObject=null
      tiling=NoTile
      imageSize=<0 0>
      margin=<ltrb 0 0 0 0>
      outlineSettings {
       bUseBrushTransparency=false
       roundingType=HalfHeightRadius
       width=0
       color {colorUseRule=UseColor_Specified, specifiedColor=<rgba 0 0 0 0>}
       cornerRadii=<0 0 0 0>
      }
      tintColor {colorUseRule=UseColor_Specified, specifiedColor=<rgba 0 0.162 0.7454 1>}
      uVRegion {bIsValid=false, max=<0 0>, min=<0 0>}
     }
     activeHighlightedBrush {
      bIsDynamicallyLoaded=false
      drawAs=Image
      imageType=NoImage
      mirroring=NoMirror
      resourceName=None
      resourceObject=null
      tiling=NoTile
      imageSize=<0 0>
      margin=<ltrb 0 0 0 0>
      outlineSettings {
       bUseBrushTransparency=false
       roundingType=HalfHeightRadius
       width=0
       color {colorUseRule=UseColor_Specified, specifiedColor=<rgba 0 0 0 0>}
       cornerRadii=<0 0 0 0>
      }
      tintColor {colorUseRule=UseColor_Specified, specifiedColor=<rgba 0.0252 0.0319 0.0423 1>}
      uVRegion {bIsValid=false, max=<0 0>, min=<0 0>}
     }
     activeHoveredBrush {
      bIsDynamicallyLoaded=false
      drawAs=Image
      imageType=NoImage
      mirroring=NoMirror
      resourceName=None
      resourceObject=null
      tiling=NoTile
      imageSize=<0 0>
      margin=<ltrb 0 0 0 0>
      outlineSettings {
       bUseBrushTransparency=false
       roundingType=HalfHeightRadius
       width=0
       color {colorUseRule=UseColor_Specified, specifiedColor=<rgba 0 0 0 0>}
       cornerRadii=<0 0 0 0>
      }
      tintColor {colorUseRule=UseColor_Specified, specifiedColor=<rgba 0 0.162 0.7454 1>}
      uVRegion {bIsValid=false, max=<0 0>, min=<0 0>}
     }
     dropIndicator_Above {
      bIsDynamicallyLoaded=false
      drawAs=Box
      imageType=FullColor
      mirroring=NoMirror
      resourceName=../../../Engine/Content/Slate/Common/DropZoneIndicator_Above.png
      resourceObject=null
      tiling=NoTile
      imageSize=<0 0>
      margin=<ltrb 0.625 0.625 0 0>
      outlineSettings {
       bUseBrushTransparency=false
       roundingType=HalfHeightRadius
       width=0
       color {colorUseRule=UseColor_Specified, specifiedColor=<rgba 0 0 0 0>}
       cornerRadii=<0 0 0 0>
      }
      tintColor {colorUseRule=UseColor_Specified, specifiedColor=<rgba 0.0091 0.3515 0.9647 1>}
      uVRegion {bIsValid=false, max=<0 0>, min=<0 0>}
     }
     dropIndicator_Below {
      bIsDynamicallyLoaded=false
      drawAs=Box
      imageType=FullColor
      mirroring=NoMirror
      resourceName=../../../Engine/Content/Slate/Common/DropZoneIndicator_Below.png
      resourceObject=null
      tiling=NoTile
      imageSize=<0 0>
      margin=<ltrb 0.625 0 0 0.625>
      outlineSettings {
       bUseBrushTransparency=false
       roundingType=HalfHeightRadius
       width=0
       color {colorUseRule=UseColor_Specified, specifiedColor=<rgba 0 0 0 0>}
       cornerRadii=<0 0 0 0>
      }
      tintColor {colorUseRule=UseColor_Specified, specifiedColor=<rgba 0.0091 0.3515 0.9647 1>}
      uVRegion {bIsValid=false, max=<0 0>, min=<0 0>}
     }
     dropIndicator_Onto {
      bIsDynamicallyLoaded=false
      drawAs=Box
      imageType=FullColor
      mirroring=NoMirror
      resourceName=../../../Engine/Content/Slate/Common/DropZoneIndicator_Onto.png
      resourceObject=null
      tiling=NoTile
      imageSize=<0 0>
      margin=<ltrb 0.25 0.25 0.25 0.25>
      outlineSettings {
       bUseBrushTransparency=false
       roundingType=HalfHeightRadius
       width=0
       color {colorUseRule=UseColor_Specified, specifiedColor=<rgba 0 0 0 0>}
       cornerRadii=<0 0 0 0>
      }
      tintColor {colorUseRule=UseColor_Specified, specifiedColor=<rgba 0.0091 0.3515 0.9647 1>}
      uVRegion {bIsValid=false, max=<0 0>, min=<0 0>}
     }
     evenRowBackgroundBrush {
      bIsDynamicallyLoaded=false
      drawAs=Image
      imageType=NoImage
      mirroring=NoMirror
      resourceName=None
      resourceObject=null
      tiling=NoTile
      imageSize=<0 0>
      margin=<ltrb 0 0 0 0>
      outlineSettings {
       bUseBrushTransparency=false
       roundingType=HalfHeightRadius
       width=0
       color {colorUseRule=UseColor_Specified, specifiedColor=<rgba 0 0 0 0>}
       cornerRadii=<0 0 0 0>
      }
      tintColor {colorUseRule=UseColor_Specified, specifiedColor=<rgba 0.0103 0.0103 0.0103 1>}
      uVRegion {bIsValid=false, max=<0 0>, min=<0 0>}
     }
     evenRowBackgroundHoveredBrush {
      bIsDynamicallyLoaded=false
      drawAs=Image
      imageType=NoImage
      mirroring=NoMirror
      resourceName=None
      resourceObject=null
      tiling=NoTile
      imageSize=<0 0>
      margin=<ltrb 0 0 0 0>
      outlineSettings {
       bUseBrushTransparency=false
       roundingType=HalfHeightRadius
       width=0
       color {colorUseRule=UseColor_Specified, specifiedColor=<rgba 0 0 0 0>}
       cornerRadii=<0 0 0 0>
      }
      tintColor {colorUseRule=UseColor_Specified, specifiedColor=<rgba 0.0176 0.0176 0.0176 1>}
      uVRegion {bIsValid=false, max=<0 0>, min=<0 0>}
     }
     inactiveBrush {
      bIsDynamicallyLoaded=false
      drawAs=Image
      imageType=NoImage
      mirroring=NoMirror
      resourceName=None
      resourceObject=null
      tiling=NoTile
      imageSize=<0 0>
      margin=<ltrb 0 0 0 0>
      outlineSettings {
       bUseBrushTransparency=false
       roundingType=HalfHeightRadius
       width=0
       color {colorUseRule=UseColor_Specified, specifiedColor=<rgba 0 0 0 0>}
       cornerRadii=<0 0 0 0>
      }
      tintColor {colorUseRule=UseColor_Specified, specifiedColor=<rgba 0.0513 0.0953 0.159 1>}
      uVRegion {bIsValid=false, max=<0 0>, min=<0 0>}
     }
     inactiveHighlightedBrush {
      bIsDynamicallyLoaded=false
      drawAs=Image
      imageType=NoImage
      mirroring=NoMirror
      resourceName=None
      resourceObject=null
      tiling=NoTile
      imageSize=<0 0>
      margin=<ltrb 0 0 0 0>
      outlineSettings {
       bUseBrushTransparency=false
       roundingType=HalfHeightRadius
       width=0
       color {colorUseRule=UseColor_Specified, specifiedColor=<rgba 0 0 0 0>}
       cornerRadii=<0 0 0 0>
      }
      tintColor {colorUseRule=UseColor_Specified, specifiedColor=<rgba 0.0252 0.0319 0.0423 1>}
      uVRegion {bIsValid=false, max=<0 0>, min=<0 0>}
     }
     inactiveHoveredBrush {
      bIsDynamicallyLoaded=false
      drawAs=Image
      imageType=NoImage
      mirroring=NoMirror
      resourceName=None
      resourceObject=null
      tiling=NoTile
      imageSize=<0 0>
      margin=<ltrb 0 0 0 0>
      outlineSettings {
       bUseBrushTransparency=false
       roundingType=HalfHeightRadius
       width=0
       color {colorUseRule=UseColor_Specified, specifiedColor=<rgba 0 0 0 0>}
       cornerRadii=<0 0 0 0>
      }
      tintColor {colorUseRule=UseColor_Specified, specifiedColor=<rgba 0.0513 0.0953 0.159 1>}
      uVRegion {bIsValid=false, max=<0 0>, min=<0 0>}
     }
     oddRowBackgroundBrush {
      bIsDynamicallyLoaded=false
      drawAs=Image
      imageType=NoImage
      mirroring=NoMirror
      resourceName=None
      resourceObject=null
      tiling=NoTile
      imageSize=<0 0>
      margin=<ltrb 0 0 0 0>
      outlineSettings {
       bUseBrushTransparency=false
       roundingType=HalfHeightRadius
       width=0
       color {colorUseRule=UseColor_Specified, specifiedColor=<rgba 0 0 0 0>}
       cornerRadii=<0 0 0 0>
      }
      tintColor {colorUseRule=UseColor_Specified, specifiedColor=<rgba 0.0103 0.0103 0.0103 1>}
      uVRegion {bIsValid=false, max=<0 0>, min=<0 0>}
     }
     oddRowBackgroundHoveredBrush {
      bIsDynamicallyLoaded=false
      drawAs=Image
      imageType=NoImage
      mirroring=NoMirror
      resourceName=None
      resourceObject=null
      tiling=NoTile
      imageSize=<0 0>
      margin=<ltrb 0 0 0 0>
      outlineSettings {
       bUseBrushTransparency=false
       roundingType=HalfHeightRadius
       width=0
       color {colorUseRule=UseColor_Specified, specifiedColor=<rgba 0 0 0 0>}
       cornerRadii=<0 0 0 0>
      }
      tintColor {colorUseRule=UseColor_Specified, specifiedColor=<rgba 0.0176 0.0176 0.0176 1>}
      uVRegion {bIsValid=false, max=<0 0>, min=<0 0>}
     }
     parentRowBackgroundBrush {
      bIsDynamicallyLoaded=false
      drawAs=Image
      imageType=NoImage
      mirroring=NoMirror
      resourceName=None
      resourceObject=null
      tiling=NoTile
      imageSize=<32 32>
      margin=<ltrb 0 0 0 0>
      outlineSettings {
       bUseBrushTransparency=false
       roundingType=HalfHeightRadius
       width=0
       color {colorUseRule=UseColor_Specified, specifiedColor=<rgba 0 0 0 0>}
       cornerRadii=<0 0 0 0>
      }
      tintColor {colorUseRule=UseColor_Specified, specifiedColor=<rgba 1 1 1 1>}
      uVRegion {bIsValid=false, max=<0 0>, min=<0 0>}
     }
     parentRowBackgroundHoveredBrush {
      bIsDynamicallyLoaded=false
      drawAs=Image
      imageType=NoImage
      mirroring=NoMirror
      resourceName=None
      resourceObject=null
      tiling=NoTile
      imageSize=<32 32>
      margin=<ltrb 0 0 0 0>
      outlineSettings {
       bUseBrushTransparency=false
       roundingType=HalfHeightRadius
       width=0
       color {colorUseRule=UseColor_Specified, specifiedColor=<rgba 0 0 0 0>}
       cornerRadii=<0 0 0 0>
      }
      tintColor {colorUseRule=UseColor_Specified, specifiedColor=<rgba 1 1 1 1>}
      uVRegion {bIsValid=false, max=<0 0>, min=<0 0>}
     }
     selectedTextColor {colorUseRule=UseColor_Specified, specifiedColor=<rgba 1 1 1 1>}
     selectorFocusedBrush {
      bIsDynamicallyLoaded=false
      drawAs=Border
      imageType=FullColor
      mirroring=NoMirror
      resourceName=../../../Engine/Content/Slate/Common/Selector.png
      resourceObject=null
      tiling=Both
      imageSize=<0 0>
      margin=<ltrb 0.25 0.25 0.25 0.25>
      outlineSettings {
       bUseBrushTransparency=false
       roundingType=HalfHeightRadius
       width=0
       color {colorUseRule=UseColor_Specified, specifiedColor=<rgba 0 0 0 0>}
       cornerRadii=<0 0 0 0>
      }
      tintColor {colorUseRule=UseColor_Specified, specifiedColor=<rgba 0.701 0.225 0.003 1>}
      uVRegion {bIsValid=false, max=<0 0>, min=<0 0>}
     }
     textColor {colorUseRule=UseColor_Specified, specifiedColor=<rgba 0.5271 0.5271 0.5271 1>}
    }
    scrollBarStyle {
     thickness=8
     draggedThumbImage {
      bIsDynamicallyLoaded=false
      drawAs=RoundedBox
      imageType=NoImage
      mirroring=NoMirror
      resourceName=None
      resourceObject=null
      tiling=NoTile
      imageSize=<0 0>
      margin=<ltrb 0 0 0 0>
      outlineSettings {
       bUseBrushTransparency=false
       roundingType=FixedRadius
       width=0
       color {colorUseRule=UseColor_Specified, specifiedColor=<rgba 0 0 0 0>}
       cornerRadii=<4 4 4 4>
      }
      tintColor {colorUseRule=UseColor_Specified, specifiedColor=<rgba 0.2159 0.2159 0.2159 1>}
      uVRegion {bIsValid=false, max=<0 0>, min=<0 0>}
     }
     horizontalBackgroundImage {
      bIsDynamicallyLoaded=false
      drawAs=NoDrawType
      imageType=NoImage
      mirroring=NoMirror
      resourceName=None
      resourceObject=null
      tiling=NoTile
      imageSize=<0 0>
      margin=<ltrb 0 0 0 0>
      outlineSettings {
       bUseBrushTransparency=false
       roundingType=HalfHeightRadius
       width=0
       color {colorUseRule=UseColor_Specified, specifiedColor=<rgba 0 0 0 0>}
       cornerRadii=<0 0 0 0>
      }
      tintColor {colorUseRule=UseColor_Specified, specifiedColor=<rgba 1 1 1 1>}
      uVRegion {bIsValid=false, max=<0 0>, min=<0 0>}
     }
     horizontalBottomSlotImage {
      bIsDynamicallyLoaded=false
      drawAs=NoDrawType
      imageType=NoImage
      mirroring=NoMirror
      resourceName=None
      resourceObject=null
      tiling=NoTile
      imageSize=<0 0>
      margin=<ltrb 0 0 0 0>
      outlineSettings {
       bUseBrushTransparency=false
       roundingType=HalfHeightRadius
       width=0
       color {colorUseRule=UseColor_Specified, specifiedColor=<rgba 0 0 0 0>}
       cornerRadii=<0 0 0 0>
      }
      tintColor {colorUseRule=UseColor_Specified, specifiedColor=<rgba 1 1 1 1>}
      uVRegion {bIsValid=false, max=<0 0>, min=<0 0>}
     }
     horizontalTopSlotImage {
      bIsDynamicallyLoaded=false
      drawAs=NoDrawType
      imageType=NoImage
      mirroring=NoMirror
      resourceName=None
      resourceObject=null
      tiling=NoTile
      imageSize=<0 0>
      margin=<ltrb 0 0 0 0>
      outlineSettings {
       bUseBrushTransparency=false
       roundingType=HalfHeightRadius
       width=0
       color {colorUseRule=UseColor_Specified, specifiedColor=<rgba 0 0 0 0>}
       cornerRadii=<0 0 0 0>
      }
      tintColor {colorUseRule=UseColor_Specified, specifiedColor=<rgba 1 1 1 1>}
      uVRegion {bIsValid=false, max=<0 0>, min=<0 0>}
     }
     hoveredThumbImage {
      bIsDynamicallyLoaded=false
      drawAs=RoundedBox
      imageType=NoImage
      mirroring=NoMirror
      resourceName=None
      resourceObject=null
      tiling=NoTile
      imageSize=<0 0>
      margin=<ltrb 0 0 0 0>
      outlineSettings {
       bUseBrushTransparency=false
       roundingType=FixedRadius
       width=0
       color {colorUseRule=UseColor_Specified, specifiedColor=<rgba 0 0 0 0>}
       cornerRadii=<4 4 4 4>
      }
      tintColor {colorUseRule=UseColor_Specified, specifiedColor=<rgba 0.2159 0.2159 0.2159 1>}
      uVRegion {bIsValid=false, max=<0 0>, min=<0 0>}
     }
     normalThumbImage {
      bIsDynamicallyLoaded=false
      drawAs=RoundedBox
      imageType=NoImage
      mirroring=NoMirror
      resourceName=None
      resourceObject=null
      tiling=NoTile
      imageSize=<0 0>
      margin=<ltrb 0 0 0 0>
      outlineSettings {
       bUseBrushTransparency=false
       roundingType=FixedRadius
       width=0
       color {colorUseRule=UseColor_Specified, specifiedColor=<rgba 0 0 0 0>}
       cornerRadii=<4 4 4 4>
      }
      tintColor {colorUseRule=UseColor_Specified, specifiedColor=<rgba 0.0953 0.0953 0.0953 1>}
      uVRegion {bIsValid=false, max=<0 0>, min=<0 0>}
     }
     verticalBackgroundImage {
      bIsDynamicallyLoaded=false
      drawAs=NoDrawType
      imageType=NoImage
      mirroring=NoMirror
      resourceName=None
      resourceObject=null
      tiling=NoTile
      imageSize=<0 0>
      margin=<ltrb 0 0 0 0>
      outlineSettings {
       bUseBrushTransparency=false
       roundingType=HalfHeightRadius
       width=0
       color {colorUseRule=UseColor_Specified, specifiedColor=<rgba 0 0 0 0>}
       cornerRadii=<0 0 0 0>
      }
      tintColor {colorUseRule=UseColor_Specified, specifiedColor=<rgba 1 1 1 1>}
      uVRegion {bIsValid=false, max=<0 0>, min=<0 0>}
     }
     verticalBottomSlotImage {
      bIsDynamicallyLoaded=false
      drawAs=NoDrawType
      imageType=NoImage
      mirroring=NoMirror
      resourceName=None
      resourceObject=null
      tiling=NoTile
      imageSize=<0 0>
      margin=<ltrb 0 0 0 0>
      outlineSettings {
       bUseBrushTransparency=false
       roundingType=HalfHeightRadius
       width=0
       color {colorUseRule=UseColor_Specified, specifiedColor=<rgba 0 0 0 0>}
       cornerRadii=<0 0 0 0>
      }
      tintColor {colorUseRule=UseColor_Specified, specifiedColor=<rgba 1 1 1 1>}
      uVRegion {bIsValid=false, max=<0 0>, min=<0 0>}
     }
     verticalTopSlotImage {
      bIsDynamicallyLoaded=false
      drawAs=NoDrawType
      imageType=NoImage
      mirroring=NoMirror
      resourceName=None
      resourceObject=null
      tiling=NoTile
      imageSize=<0 0>
      margin=<ltrb 0 0 0 0>
      outlineSettings {
       bUseBrushTransparency=false
       roundingType=HalfHeightRadius
       width=0
       color {colorUseRule=UseColor_Specified, specifiedColor=<rgba 0 0 0 0>}
       cornerRadii=<0 0 0 0>
      }
      tintColor {colorUseRule=UseColor_Specified, specifiedColor=<rgba 1 1 1 1>}
      uVRegion {bIsValid=false, max=<0 0>, min=<0 0>}
     }
    }
    widgetStyle {
     comboButtonStyle {
      downArrowAlign=VAlign_Center
      buttonStyle {
       clickedSlateSound {resourceObject=null}
       disabled {
        bIsDynamicallyLoaded=false
        drawAs=RoundedBox
        imageType=NoImage
        mirroring=NoMirror
        resourceName=None
        resourceObject=null
        tiling=NoTile
        imageSize=<0 0>
        margin=<ltrb 0 0 0 0>
        outlineSettings {
         bUseBrushTransparency=false
         roundingType=FixedRadius
         width=1
         color {colorUseRule=UseColor_ColorTable, specifiedColor=<rgba 0 0 0 0>}
         cornerRadii=<4 4 4 4>
        }
        tintColor {colorUseRule=UseColor_Specified, specifiedColor=<rgba 0.0395 0.0395 0.0395 1>}
        uVRegion {bIsValid=false, max=<0 0>, min=<0 0>}
       }
       disabledForeground {colorUseRule=UseColor_ColorTable, specifiedColor=<rgba 0 0 0 0>}
       hovered {
        bIsDynamicallyLoaded=false
        drawAs=RoundedBox
        imageType=NoImage
        mirroring=NoMirror
        resourceName=None
        resourceObject=null
        tiling=NoTile
        imageSize=<0 0>
        margin=<ltrb 0 0 0 0>
        outlineSettings {
         bUseBrushTransparency=false
         roundingType=FixedRadius
         width=1
         color {colorUseRule=UseColor_ColorTable, specifiedColor=<rgba 0 0 0 0>}
         cornerRadii=<4 4 4 4>
        }
        tintColor {colorUseRule=UseColor_Specified, specifiedColor=<rgba 0.0048 0.0048 0.0048 1>}
        uVRegion {bIsValid=false, max=<0 0>, min=<0 0>}
       }
       hoveredForeground {colorUseRule=UseColor_ColorTable, specifiedColor=<rgba 0 0 0 0>}
       hoveredSlateSound {resourceObject=null}
       normal {
        bIsDynamicallyLoaded=false
        drawAs=RoundedBox
        imageType=NoImage
        mirroring=NoMirror
        resourceName=None
        resourceObject=null
        tiling=NoTile
        imageSize=<0 0>
        margin=<ltrb 0 0 0 0>
        outlineSettings {
         bUseBrushTransparency=false
         roundingType=FixedRadius
         width=1
         color {colorUseRule=UseColor_ColorTable, specifiedColor=<rgba 0 0 0 0>}
         cornerRadii=<4 4 4 4>
        }
        tintColor {colorUseRule=UseColor_Specified, specifiedColor=<rgba 0.0048 0.0048 0.0048 1>}
        uVRegion {bIsValid=false, max=<0 0>, min=<0 0>}
       }
       normalForeground {colorUseRule=UseColor_ColorTable, specifiedColor=<rgba 0 0 0 0>}
       normalPadding=<ltrb 8 1 8 1>
       pressed {
        bIsDynamicallyLoaded=false
        drawAs=RoundedBox
        imageType=NoImage
        mirroring=NoMirror
        resourceName=None
        resourceObject=null
        tiling=NoTile
        imageSize=<0 0>
        margin=<ltrb 0 0 0 0>
        outlineSettings {
         bUseBrushTransparency=false
         roundingType=FixedRadius
         width=1
         color {colorUseRule=UseColor_ColorTable, specifiedColor=<rgba 0 0 0 0>}
         cornerRadii=<4 4 4 4>
        }
        tintColor {colorUseRule=UseColor_Specified, specifiedColor=<rgba 0.0048 0.0048 0.0048 1>}
        uVRegion {bIsValid=false, max=<0 0>, min=<0 0>}
       }
       pressedForeground {colorUseRule=UseColor_ColorTable, specifiedColor=<rgba 0 0 0 0>}
       pressedPadding=<ltrb 8 2 8 0>
       pressedSlateSound {resourceObject=null}
      }
      contentPadding=<ltrb 0 0 0 0>
      downArrowImage {
       bIsDynamicallyLoaded=false
       drawAs=Image
       imageType=Vector
       mirroring=NoMirror
       resourceName=../../../Engine/Content/Slate/Starship/CoreWidgets/ComboBox/wide-chevron-down.svg
       resourceObject=null
       tiling=NoTile
       imageSize=<16 16>
       margin=<ltrb 0 0 0 0>
       outlineSettings {
        bUseBrushTransparency=false
        roundingType=HalfHeightRadius
        width=0
        color {colorUseRule=UseColor_Specified, specifiedColor=<rgba 0 0 0 0>}
        cornerRadii=<0 0 0 0>
       }
       tintColor {colorUseRule=UseColor_Specified, specifiedColor=<rgba 1 1 1 1>}
       uVRegion {bIsValid=false, max=<0 0>, min=<0 0>}
      }
      downArrowPadding=<ltrb 2 2 2 2>
      menuBorderBrush {
       bIsDynamicallyLoaded=false
       drawAs=RoundedBox
       imageType=NoImage
       mirroring=NoMirror
       resourceName=None
       resourceObject=null
       tiling=NoTile
       imageSize=<0 0>
       margin=<ltrb 0 0 0 0>
       outlineSettings {
        bUseBrushTransparency=false
        roundingType=FixedRadius
        width=1
        color {colorUseRule=UseColor_ColorTable, specifiedColor=<rgba 0 0 0 0>}
        cornerRadii=<0 0 0 0>
       }
       tintColor {colorUseRule=UseColor_Specified, specifiedColor=<rgba 0.0395 0.0395 0.0395 1>}
       uVRegion {bIsValid=false, max=<0 0>, min=<0 0>}
      }
      menuBorderPadding=<ltrb 0 0 0 0>
      shadowColorAndOpacity=<rgba 0 0 0 1>
      shadowOffset=<0 0>
     }
     contentPadding=<ltrb 0 0 0 0>
     menuRowPadding=<ltrb 8 3 8 3>
     pressedSlateSound {resourceObject=null}
     selectionChangeSlateSound {resourceObject=null}
    }
   }
   slot {class=HorizontalBoxSlot, properties {size {sizeRule=Fill, value=1}}}
  }
  Button_ResetBoneFilter {
   widgetClassPath=&EUW_FAnim_SimpleConsoleButton
   bIsVariable=true
   parent=HorizontalBox_1146
   widget=Button_ResetBoneFilter
   properties {buttonName=Clear filters, console Command=ResetBoneFilters, run On Server=false, toolTipText=Clear all active bone filters}
   slot {class=HorizontalBoxSlot, properties {padding=<ltrb 1 0 0 0>, size {sizeRule=Fill, value=1}}}
  }
  HorizontalBox_BoneFilterTokens {
   widgetClassPath=&HorizontalBox
   bIsVariable=true
   parent=VerticalBox
   widget=HorizontalBox_BoneFilterTokens
   slot {class=VerticalBoxSlot}
  }
  WrapBox_BoneFilters {
   widgetClassPath=&WrapBox
   bIsVariable=true
   parent=HorizontalBox_BoneFilterTokens
   widget=WrapBox_BoneFilters
   slot {class=HorizontalBoxSlot, properties {size {sizeRule=Fill, value=5}}}
  }
  Expandable_ProceduralLayering {
   widgetClassPath=&EditorUtilityExpandableArea
   bIsVariable=true
   parent=MainScrollBox
   widget=Expandable_ProceduralLayering
   properties {bIsExpanded=true}
   slot {class=ScrollBoxSlot}
  }
  WB_FAnim_SimpleTextBlock_4 {
   widgetClassPath=&WB_FAnim_SimpleTextBlock
   namedSlotHost=Expandable_ProceduralLayering
   widget=WB_FAnim_SimpleTextBlock_4
   properties {fontSize=11, isBold=true, text=Procedural Layering}
  }
  VerticalBox_70 {widgetClassPath=&VerticalBox, namedSlotHost=Expandable_ProceduralLayering, widget=VerticalBox_70}
  ProxyTableAssetRow {
   widgetClassPath=&/EUW_FAnim_AssetEditControl
   bIsVariable=true
   parent=VerticalBox_70
   widget=ProxyTableAssetRow
   properties {assetLabelText=Proxy Table}
   slot {class=VerticalBoxSlot}
  }
  ChooserData_AssetControl {
   widgetClassPath=&EUW_FAnim_AssetEditControl
   bIsVariable=true
   parent=VerticalBox_70
   widget=ChooserData_AssetControl
   properties {assetLabelText=Item Chooser Data}
   slot {class=VerticalBoxSlot}
  }
  HorizontalBox_91 {
   widgetClassPath=&HorizontalBox
   parent=VerticalBox_70
   widget=HorizontalBox_91
   slot {class=VerticalBoxSlot, properties {size {sizeRule=Fill, value=1}}}
  }
  WB_FAnim_SimpleTextBlock_5 {
   widgetClassPath=&WB_FAnim_SimpleTextBlock
   parent=HorizontalBox_91
   widget=WB_FAnim_SimpleTextBlock_5
   properties {text=Display, boxSize=<46 25>}
   slot {class=HorizontalBoxSlot}
  }
  EUW_FAnim_SimpleConsoleButton_79 {
   widgetClassPath=&EUW_FAnim_SimpleConsoleButton
   parent=HorizontalBox_91
   widget=EUW_FAnim_SimpleConsoleButton_79
   properties {
    buttonName=Layering Assets
    command Has Int State=true
    console Command=a.Fort.Layering.DisplayLayeringAssets
    toolTipText=Shows the asset currently in use for procedural layering
   }
   slot {class=HorizontalBoxSlot, properties {size {sizeRule=Fill, value=1}}}
  }
  EUW_FAnim_SimpleConsoleButton_1 {
   widgetClassPath=&EUW_FAnim_SimpleConsoleButton
   parent=HorizontalBox_91
   widget=EUW_FAnim_SimpleConsoleButton_1
   properties {
    buttonName=Layering Weights
    command Has Int State=true
    console Command=a.Fort.Layering.DisplayLayeringWeights
    toolTipText=Shows the currently active weight for the pose layering and additive layering stages
   }
   slot {class=HorizontalBoxSlot, properties {size {sizeRule=Fill, value=1}}}
  }
  EUW_FAnim_SimpleConsoleButton_2 {
   widgetClassPath=&EUW_FAnim_SimpleConsoleButton
   parent=HorizontalBox_91
   widget=EUW_FAnim_SimpleConsoleButton_2
   properties {
    buttonName=Procedural Data
    command Has Int State=true
    console Command=a.Fort.Layering.DisplayProceduralData
    toolTipText=Shows the currently active procedural data for the procedural IK motion stage
   }
   slot {class=HorizontalBoxSlot, properties {size {sizeRule=Fill, value=1}}}
  }
  HorizontalBox_223 {widgetClassPath=&HorizontalBox, parent=VerticalBox_70, widget=HorizontalBox_223, slot {class=VerticalBoxSlot}}
  EUW_FAnim_SimpleConsoleButton_3 {
   widgetClassPath=&EUW_FAnim_SimpleConsoleButton
   parent=HorizontalBox_223
   widget=EUW_FAnim_SimpleConsoleButton_3
   properties {
    buttonName=Enable Additive
    command Has Int State=true
    console Command=a.Fort.Layering.AdditiveLayeringAlphaOverride
    enabledIntValue=-1
    toolTipText=Enable/Disables the additive stage of the procedural layering stack.
   }
   slot {class=HorizontalBoxSlot, properties {size {sizeRule=Fill, value=1}}}
  }
  EUW_FAnim_SimpleConsoleButton_4 {
   widgetClassPath=&EUW_FAnim_SimpleConsoleButton
   parent=HorizontalBox_223
   widget=EUW_FAnim_SimpleConsoleButton_4
   properties {
    buttonName=Enable Procedural
    command Has Int State=true
    console Command=a.Fort.Layering.ProceduralLayeringAlphaOverride
    enabledIntValue=-1
    toolTipText=Enable/Disables the procedural IK motion stage of the procedural layering stack.
   }
   slot {class=HorizontalBoxSlot, properties {size {sizeRule=Fill, value=1}}}
  }
  Expandable_AnimationAssets {
   widgetClassPath=&EditorUtilityExpandableArea
   bIsVariable=true
   parent=MainScrollBox
   widget=Expandable_AnimationAssets
   properties {bIsExpanded=true}
   slot {class=ScrollBoxSlot}
  }
  WB_SimpleTextBlock_1 {
   widgetClassPath=&WB_FAnim_SimpleTextBlock
   namedSlotHost=Expandable_AnimationAssets
   widget=WB_SimpleTextBlock_1
   properties {fontSize=11, isBold=true, text=Animation Assets}
  }
  VerticalBox_160 {widgetClassPath=&VerticalBox, namedSlotHost=Expandable_AnimationAssets, widget=VerticalBox_160}
  ItemLayer_AssetControl {
   widgetClassPath=&EUW_FAnim_AssetEditControl
   bIsVariable=true
   parent=VerticalBox_160
   widget=ItemLayer_AssetControl
   properties {assetLabelText=Anim Layer}
   slot {class=VerticalBoxSlot}
  }
  WeaponAnimSet_AssetControl {
   widgetClassPath=&EUW_FAnim_AssetEditControl
   bIsVariable=true
   parent=VerticalBox_160
   widget=WeaponAnimSet_AssetControl
   properties {assetLabelText=Weapon Anim Set}
   slot {class=VerticalBoxSlot}
  }
  WeaponAdditiveAnimSet_AssetControl {
   widgetClassPath=&EUW_FAnim_AssetEditControl
   bIsVariable=true
   parent=VerticalBox_160
   widget=WeaponAdditiveAnimSet_AssetControl
   properties {assetLabelText=Weapon Additive Anim Set}
   slot {class=VerticalBoxSlot}
  }
  TacticalAnimSet_AssetControl {
   widgetClassPath=&EUW_FAnim_AssetEditControl
   bIsVariable=true
   parent=VerticalBox_160
   widget=TacticalAnimSet_AssetControl
   properties {assetLabelText=Tactical Anim Set}
   slot {class=VerticalBoxSlot}
  }
  Expandable_IK {
   widgetClassPath=&EditorUtilityExpandableArea
   bIsVariable=true
   parent=MainScrollBox
   widget=Expandable_IK
   properties {bIsExpanded=true}
   slot {class=ScrollBoxSlot}
  }
  WB_FAnim_SimpleTextBlock_116 {
   widgetClassPath=&WB_FAnim_SimpleTextBlock
   namedSlotHost=Expandable_IK
   widget=WB_FAnim_SimpleTextBlock_116
   properties {fontSize=11, isBold=true, text=IK}
  }
  VerticalBox_57 {widgetClassPath=&VerticalBox, namedSlotHost=Expandable_IK, widget=VerticalBox_57}
  HorizontalBox_446 {
   widgetClassPath=&HorizontalBox
   parent=VerticalBox_57
   widget=HorizontalBox_446
   slot {class=VerticalBoxSlot, properties {size {sizeRule=Fill, value=1}}}
  }
  WB_FAnim_SimpleTextBlock_1 {
   widgetClassPath=&WB_FAnim_SimpleTextBlock
   parent=HorizontalBox_446
   widget=WB_FAnim_SimpleTextBlock_1
   properties {
    text=Leg IK
    toolTipText="Override Leg IK\r\nDefault: No override\r\nForce IK will force IK alpha to 1\r\nForce FK will force IK alpha to 0"
    boxSize=<60 25>
    textPadding=<ltrb 5 5 5 5>
   }
   slot {class=HorizontalBoxSlot}
  }
  ComboBox_LegIK {
   widgetClassPath=&ComboBoxString
   bIsVariable=true
   parent=HorizontalBox_446
   widget=ComboBox_LegIK
   properties {
    contentPadding=<ltrb 2 2 2 2>
    font {
     bForceMonospaced=false
     bMaterialIsStencil=false
     fontMaterial=null
     fontObject=&Roboto
     letterSpacing=0
     monospacedWidth=1
     size=9
     skewAmount=0
     typefaceFontName=Regular
     outlineSettings {
      bApplyOutlineToDropShadows=false
      bMiteredCorners=false
      bSeparateFillAlpha=false
      outlineMaterial=null
      outlineSize=0
      outlineColor=<rgba 0 0 0 1>
     }
    }
    foregroundColor {colorUseRule=UseColor_Specified, specifiedColor=<rgba 0.5271 0.5271 0.5271 1>}
    itemStyle {
     bUseParentRowBrush=false
     activeBrush {
      bIsDynamicallyLoaded=false
      drawAs=Image
      imageType=NoImage
      mirroring=NoMirror
      resourceName=None
      resourceObject=null
      tiling=NoTile
      imageSize=<0 0>
      margin=<ltrb 0 0 0 0>
      outlineSettings {
       bUseBrushTransparency=false
       roundingType=HalfHeightRadius
       width=0
       color {colorUseRule=UseColor_Specified, specifiedColor=<rgba 0 0 0 0>}
       cornerRadii=<0 0 0 0>
      }
      tintColor {colorUseRule=UseColor_Specified, specifiedColor=<rgba 0 0.162 0.7454 1>}
      uVRegion {bIsValid=false, max=<0 0>, min=<0 0>}
     }
     activeHighlightedBrush {
      bIsDynamicallyLoaded=false
      drawAs=Image
      imageType=NoImage
      mirroring=NoMirror
      resourceName=None
      resourceObject=null
      tiling=NoTile
      imageSize=<0 0>
      margin=<ltrb 0 0 0 0>
      outlineSettings {
       bUseBrushTransparency=false
       roundingType=HalfHeightRadius
       width=0
       color {colorUseRule=UseColor_Specified, specifiedColor=<rgba 0 0 0 0>}
       cornerRadii=<0 0 0 0>
      }
      tintColor {colorUseRule=UseColor_Specified, specifiedColor=<rgba 0.0252 0.0319 0.0423 1>}
      uVRegion {bIsValid=false, max=<0 0>, min=<0 0>}
     }
     activeHoveredBrush {
      bIsDynamicallyLoaded=false
      drawAs=Image
      imageType=NoImage
      mirroring=NoMirror
      resourceName=None
      resourceObject=null
      tiling=NoTile
      imageSize=<0 0>
      margin=<ltrb 0 0 0 0>
      outlineSettings {
       bUseBrushTransparency=false
       roundingType=HalfHeightRadius
       width=0
       color {colorUseRule=UseColor_Specified, specifiedColor=<rgba 0 0 0 0>}
       cornerRadii=<0 0 0 0>
      }
      tintColor {colorUseRule=UseColor_Specified, specifiedColor=<rgba 0 0.162 0.7454 1>}
      uVRegion {bIsValid=false, max=<0 0>, min=<0 0>}
     }
     dropIndicator_Above {
      bIsDynamicallyLoaded=false
      drawAs=Box
      imageType=FullColor
      mirroring=NoMirror
      resourceName=../../../Engine/Content/Slate/Common/DropZoneIndicator_Above.png
      resourceObject=null
      tiling=NoTile
      imageSize=<0 0>
      margin=<ltrb 0.625 0.625 0 0>
      outlineSettings {
       bUseBrushTransparency=false
       roundingType=HalfHeightRadius
       width=0
       color {colorUseRule=UseColor_Specified, specifiedColor=<rgba 0 0 0 0>}
       cornerRadii=<0 0 0 0>
      }
      tintColor {colorUseRule=UseColor_Specified, specifiedColor=<rgba 0.0091 0.3515 0.9647 1>}
      uVRegion {bIsValid=false, max=<0 0>, min=<0 0>}
     }
     dropIndicator_Below {
      bIsDynamicallyLoaded=false
      drawAs=Box
      imageType=FullColor
      mirroring=NoMirror
      resourceName=../../../Engine/Content/Slate/Common/DropZoneIndicator_Below.png
      resourceObject=null
      tiling=NoTile
      imageSize=<0 0>
      margin=<ltrb 0.625 0 0 0.625>
      outlineSettings {
       bUseBrushTransparency=false
       roundingType=HalfHeightRadius
       width=0
       color {colorUseRule=UseColor_Specified, specifiedColor=<rgba 0 0 0 0>}
       cornerRadii=<0 0 0 0>
      }
      tintColor {colorUseRule=UseColor_Specified, specifiedColor=<rgba 0.0091 0.3515 0.9647 1>}
      uVRegion {bIsValid=false, max=<0 0>, min=<0 0>}
     }
     dropIndicator_Onto {
      bIsDynamicallyLoaded=false
      drawAs=Box
      imageType=FullColor
      mirroring=NoMirror
      resourceName=../../../Engine/Content/Slate/Common/DropZoneIndicator_Onto.png
      resourceObject=null
      tiling=NoTile
      imageSize=<0 0>
      margin=<ltrb 0.25 0.25 0.25 0.25>
      outlineSettings {
       bUseBrushTransparency=false
       roundingType=HalfHeightRadius
       width=0
       color {colorUseRule=UseColor_Specified, specifiedColor=<rgba 0 0 0 0>}
       cornerRadii=<0 0 0 0>
      }
      tintColor {colorUseRule=UseColor_Specified, specifiedColor=<rgba 0.0091 0.3515 0.9647 1>}
      uVRegion {bIsValid=false, max=<0 0>, min=<0 0>}
     }
     evenRowBackgroundBrush {
      bIsDynamicallyLoaded=false
      drawAs=Image
      imageType=NoImage
      mirroring=NoMirror
      resourceName=None
      resourceObject=null
      tiling=NoTile
      imageSize=<0 0>
      margin=<ltrb 0 0 0 0>
      outlineSettings {
       bUseBrushTransparency=false
       roundingType=HalfHeightRadius
       width=0
       color {colorUseRule=UseColor_Specified, specifiedColor=<rgba 0 0 0 0>}
       cornerRadii=<0 0 0 0>
      }
      tintColor {colorUseRule=UseColor_Specified, specifiedColor=<rgba 0.0103 0.0103 0.0103 1>}
      uVRegion {bIsValid=false, max=<0 0>, min=<0 0>}
     }
     evenRowBackgroundHoveredBrush {
      bIsDynamicallyLoaded=false
      drawAs=Image
      imageType=NoImage
      mirroring=NoMirror
      resourceName=None
      resourceObject=null
      tiling=NoTile
      imageSize=<0 0>
      margin=<ltrb 0 0 0 0>
      outlineSettings {
       bUseBrushTransparency=false
       roundingType=HalfHeightRadius
       width=0
       color {colorUseRule=UseColor_Specified, specifiedColor=<rgba 0 0 0 0>}
       cornerRadii=<0 0 0 0>
      }
      tintColor {colorUseRule=UseColor_Specified, specifiedColor=<rgba 0.0176 0.0176 0.0176 1>}
      uVRegion {bIsValid=false, max=<0 0>, min=<0 0>}
     }
     inactiveBrush {
      bIsDynamicallyLoaded=false
      drawAs=Image
      imageType=NoImage
      mirroring=NoMirror
      resourceName=None
      resourceObject=null
      tiling=NoTile
      imageSize=<0 0>
      margin=<ltrb 0 0 0 0>
      outlineSettings {
       bUseBrushTransparency=false
       roundingType=HalfHeightRadius
       width=0
       color {colorUseRule=UseColor_Specified, specifiedColor=<rgba 0 0 0 0>}
       cornerRadii=<0 0 0 0>
      }
      tintColor {colorUseRule=UseColor_Specified, specifiedColor=<rgba 0.0513 0.0953 0.159 1>}
      uVRegion {bIsValid=false, max=<0 0>, min=<0 0>}
     }
     inactiveHighlightedBrush {
      bIsDynamicallyLoaded=false
      drawAs=Image
      imageType=NoImage
      mirroring=NoMirror
      resourceName=None
      resourceObject=null
      tiling=NoTile
      imageSize=<0 0>
      margin=<ltrb 0 0 0 0>
      outlineSettings {
       bUseBrushTransparency=false
       roundingType=HalfHeightRadius
       width=0
       color {colorUseRule=UseColor_Specified, specifiedColor=<rgba 0 0 0 0>}
       cornerRadii=<0 0 0 0>
      }
      tintColor {colorUseRule=UseColor_Specified, specifiedColor=<rgba 0.0252 0.0319 0.0423 1>}
      uVRegion {bIsValid=false, max=<0 0>, min=<0 0>}
     }
     inactiveHoveredBrush {
      bIsDynamicallyLoaded=false
      drawAs=Image
      imageType=NoImage
      mirroring=NoMirror
      resourceName=None
      resourceObject=null
      tiling=NoTile
      imageSize=<0 0>
      margin=<ltrb 0 0 0 0>
      outlineSettings {
       bUseBrushTransparency=false
       roundingType=HalfHeightRadius
       width=0
       color {colorUseRule=UseColor_Specified, specifiedColor=<rgba 0 0 0 0>}
       cornerRadii=<0 0 0 0>
      }
      tintColor {colorUseRule=UseColor_Specified, specifiedColor=<rgba 0.0513 0.0953 0.159 1>}
      uVRegion {bIsValid=false, max=<0 0>, min=<0 0>}
     }
     oddRowBackgroundBrush {
      bIsDynamicallyLoaded=false
      drawAs=Image
      imageType=NoImage
      mirroring=NoMirror
      resourceName=None
      resourceObject=null
      tiling=NoTile
      imageSize=<0 0>
      margin=<ltrb 0 0 0 0>
      outlineSettings {
       bUseBrushTransparency=false
       roundingType=HalfHeightRadius
       width=0
       color {colorUseRule=UseColor_Specified, specifiedColor=<rgba 0 0 0 0>}
       cornerRadii=<0 0 0 0>
      }
      tintColor {colorUseRule=UseColor_Specified, specifiedColor=<rgba 0.0103 0.0103 0.0103 1>}
      uVRegion {bIsValid=false, max=<0 0>, min=<0 0>}
     }
     oddRowBackgroundHoveredBrush {
      bIsDynamicallyLoaded=false
      drawAs=Image
      imageType=NoImage
      mirroring=NoMirror
      resourceName=None
      resourceObject=null
      tiling=NoTile
      imageSize=<0 0>
      margin=<ltrb 0 0 0 0>
      outlineSettings {
       bUseBrushTransparency=false
       roundingType=HalfHeightRadius
       width=0
       color {colorUseRule=UseColor_Specified, specifiedColor=<rgba 0 0 0 0>}
       cornerRadii=<0 0 0 0>
      }
      tintColor {colorUseRule=UseColor_Specified, specifiedColor=<rgba 0.0176 0.0176 0.0176 1>}
      uVRegion {bIsValid=false, max=<0 0>, min=<0 0>}
     }
     parentRowBackgroundBrush {
      bIsDynamicallyLoaded=false
      drawAs=Image
      imageType=NoImage
      mirroring=NoMirror
      resourceName=None
      resourceObject=null
      tiling=NoTile
      imageSize=<32 32>
      margin=<ltrb 0 0 0 0>
      outlineSettings {
       bUseBrushTransparency=false
       roundingType=HalfHeightRadius
       width=0
       color {colorUseRule=UseColor_Specified, specifiedColor=<rgba 0 0 0 0>}
       cornerRadii=<0 0 0 0>
      }
      tintColor {colorUseRule=UseColor_Specified, specifiedColor=<rgba 1 1 1 1>}
      uVRegion {bIsValid=false, max=<0 0>, min=<0 0>}
     }
     parentRowBackgroundHoveredBrush {
      bIsDynamicallyLoaded=false
      drawAs=Image
      imageType=NoImage
      mirroring=NoMirror
      resourceName=None
      resourceObject=null
      tiling=NoTile
      imageSize=<32 32>
      margin=<ltrb 0 0 0 0>
      outlineSettings {
       bUseBrushTransparency=false
       roundingType=HalfHeightRadius
       width=0
       color {colorUseRule=UseColor_Specified, specifiedColor=<rgba 0 0 0 0>}
       cornerRadii=<0 0 0 0>
      }
      tintColor {colorUseRule=UseColor_Specified, specifiedColor=<rgba 1 1 1 1>}
      uVRegion {bIsValid=false, max=<0 0>, min=<0 0>}
     }
     selectedTextColor {colorUseRule=UseColor_Specified, specifiedColor=<rgba 1 1 1 1>}
     selectorFocusedBrush {
      bIsDynamicallyLoaded=false
      drawAs=Border
      imageType=FullColor
      mirroring=NoMirror
      resourceName=../../../Engine/Content/Slate/Common/Selector.png
      resourceObject=null
      tiling=Both
      imageSize=<0 0>
      margin=<ltrb 0.25 0.25 0.25 0.25>
      outlineSettings {
       bUseBrushTransparency=false
       roundingType=HalfHeightRadius
       width=0
       color {colorUseRule=UseColor_Specified, specifiedColor=<rgba 0 0 0 0>}
       cornerRadii=<0 0 0 0>
      }
      tintColor {colorUseRule=UseColor_Specified, specifiedColor=<rgba 0.701 0.225 0.003 1>}
      uVRegion {bIsValid=false, max=<0 0>, min=<0 0>}
     }
     textColor {colorUseRule=UseColor_Specified, specifiedColor=<rgba 0.5271 0.5271 0.5271 1>}
    }
    scrollBarStyle {
     thickness=8
     draggedThumbImage {
      bIsDynamicallyLoaded=false
      drawAs=RoundedBox
      imageType=NoImage
      mirroring=NoMirror
      resourceName=None
      resourceObject=null
      tiling=NoTile
      imageSize=<0 0>
      margin=<ltrb 0 0 0 0>
      outlineSettings {
       bUseBrushTransparency=false
       roundingType=FixedRadius
       width=0
       color {colorUseRule=UseColor_Specified, specifiedColor=<rgba 0 0 0 0>}
       cornerRadii=<4 4 4 4>
      }
      tintColor {colorUseRule=UseColor_Specified, specifiedColor=<rgba 0.2159 0.2159 0.2159 1>}
      uVRegion {bIsValid=false, max=<0 0>, min=<0 0>}
     }
     horizontalBackgroundImage {
      bIsDynamicallyLoaded=false
      drawAs=NoDrawType
      imageType=NoImage
      mirroring=NoMirror
      resourceName=None
      resourceObject=null
      tiling=NoTile
      imageSize=<0 0>
      margin=<ltrb 0 0 0 0>
      outlineSettings {
       bUseBrushTransparency=false
       roundingType=HalfHeightRadius
       width=0
       color {colorUseRule=UseColor_Specified, specifiedColor=<rgba 0 0 0 0>}
       cornerRadii=<0 0 0 0>
      }
      tintColor {colorUseRule=UseColor_Specified, specifiedColor=<rgba 1 1 1 1>}
      uVRegion {bIsValid=false, max=<0 0>, min=<0 0>}
     }
     horizontalBottomSlotImage {
      bIsDynamicallyLoaded=false
      drawAs=NoDrawType
      imageType=NoImage
      mirroring=NoMirror
      resourceName=None
      resourceObject=null
      tiling=NoTile
      imageSize=<0 0>
      margin=<ltrb 0 0 0 0>
      outlineSettings {
       bUseBrushTransparency=false
       roundingType=HalfHeightRadius
       width=0
       color {colorUseRule=UseColor_Specified, specifiedColor=<rgba 0 0 0 0>}
       cornerRadii=<0 0 0 0>
      }
      tintColor {colorUseRule=UseColor_Specified, specifiedColor=<rgba 1 1 1 1>}
      uVRegion {bIsValid=false, max=<0 0>, min=<0 0>}
     }
     horizontalTopSlotImage {
      bIsDynamicallyLoaded=false
      drawAs=NoDrawType
      imageType=NoImage
      mirroring=NoMirror
      resourceName=None
      resourceObject=null
      tiling=NoTile
      imageSize=<0 0>
      margin=<ltrb 0 0 0 0>
      outlineSettings {
       bUseBrushTransparency=false
       roundingType=HalfHeightRadius
       width=0
       color {colorUseRule=UseColor_Specified, specifiedColor=<rgba 0 0 0 0>}
       cornerRadii=<0 0 0 0>
      }
      tintColor {colorUseRule=UseColor_Specified, specifiedColor=<rgba 1 1 1 1>}
      uVRegion {bIsValid=false, max=<0 0>, min=<0 0>}
     }
     hoveredThumbImage {
      bIsDynamicallyLoaded=false
      drawAs=RoundedBox
      imageType=NoImage
      mirroring=NoMirror
      resourceName=None
      resourceObject=null
      tiling=NoTile
      imageSize=<0 0>
      margin=<ltrb 0 0 0 0>
      outlineSettings {
       bUseBrushTransparency=false
       roundingType=FixedRadius
       width=0
       color {colorUseRule=UseColor_Specified, specifiedColor=<rgba 0 0 0 0>}
       cornerRadii=<4 4 4 4>
      }
      tintColor {colorUseRule=UseColor_Specified, specifiedColor=<rgba 0.2159 0.2159 0.2159 1>}
      uVRegion {bIsValid=false, max=<0 0>, min=<0 0>}
     }
     normalThumbImage {
      bIsDynamicallyLoaded=false
      drawAs=RoundedBox
      imageType=NoImage
      mirroring=NoMirror
      resourceName=None
      resourceObject=null
      tiling=NoTile
      imageSize=<0 0>
      margin=<ltrb 0 0 0 0>
      outlineSettings {
       bUseBrushTransparency=false
       roundingType=FixedRadius
       width=0
       color {colorUseRule=UseColor_Specified, specifiedColor=<rgba 0 0 0 0>}
       cornerRadii=<4 4 4 4>
      }
      tintColor {colorUseRule=UseColor_Specified, specifiedColor=<rgba 0.0953 0.0953 0.0953 1>}
      uVRegion {bIsValid=false, max=<0 0>, min=<0 0>}
     }
     verticalBackgroundImage {
      bIsDynamicallyLoaded=false
      drawAs=NoDrawType
      imageType=NoImage
      mirroring=NoMirror
      resourceName=None
      resourceObject=null
      tiling=NoTile
      imageSize=<0 0>
      margin=<ltrb 0 0 0 0>
      outlineSettings {
       bUseBrushTransparency=false
       roundingType=HalfHeightRadius
       width=0
       color {colorUseRule=UseColor_Specified, specifiedColor=<rgba 0 0 0 0>}
       cornerRadii=<0 0 0 0>
      }
      tintColor {colorUseRule=UseColor_Specified, specifiedColor=<rgba 1 1 1 1>}
      uVRegion {bIsValid=false, max=<0 0>, min=<0 0>}
     }
     verticalBottomSlotImage {
      bIsDynamicallyLoaded=false
      drawAs=NoDrawType
      imageType=NoImage
      mirroring=NoMirror
      resourceName=None
      resourceObject=null
      tiling=NoTile
      imageSize=<0 0>
      margin=<ltrb 0 0 0 0>
      outlineSettings {
       bUseBrushTransparency=false
       roundingType=HalfHeightRadius
       width=0
       color {colorUseRule=UseColor_Specified, specifiedColor=<rgba 0 0 0 0>}
       cornerRadii=<0 0 0 0>
      }
      tintColor {colorUseRule=UseColor_Specified, specifiedColor=<rgba 1 1 1 1>}
      uVRegion {bIsValid=false, max=<0 0>, min=<0 0>}
     }
     verticalTopSlotImage {
      bIsDynamicallyLoaded=false
      drawAs=NoDrawType
      imageType=NoImage
      mirroring=NoMirror
      resourceName=None
      resourceObject=null
      tiling=NoTile
      imageSize=<0 0>
      margin=<ltrb 0 0 0 0>
      outlineSettings {
       bUseBrushTransparency=false
       roundingType=HalfHeightRadius
       width=0
       color {colorUseRule=UseColor_Specified, specifiedColor=<rgba 0 0 0 0>}
       cornerRadii=<0 0 0 0>
      }
      tintColor {colorUseRule=UseColor_Specified, specifiedColor=<rgba 1 1 1 1>}
      uVRegion {bIsValid=false, max=<0 0>, min=<0 0>}
     }
    }
    widgetStyle {
     comboButtonStyle {
      downArrowAlign=VAlign_Center
      buttonStyle {
       clickedSlateSound {resourceObject=null}
       disabled {
        bIsDynamicallyLoaded=false
        drawAs=RoundedBox
        imageType=NoImage
        mirroring=NoMirror
        resourceName=None
        resourceObject=null
        tiling=NoTile
        imageSize=<0 0>
        margin=<ltrb 0 0 0 0>
        outlineSettings {
         bUseBrushTransparency=false
         roundingType=FixedRadius
         width=1
         color {colorUseRule=UseColor_ColorTable, specifiedColor=<rgba 0 0 0 0>}
         cornerRadii=<4 4 4 4>
        }
        tintColor {colorUseRule=UseColor_Specified, specifiedColor=<rgba 0.0395 0.0395 0.0395 1>}
        uVRegion {bIsValid=false, max=<0 0>, min=<0 0>}
       }
       disabledForeground {colorUseRule=UseColor_ColorTable, specifiedColor=<rgba 0 0 0 0>}
       hovered {
        bIsDynamicallyLoaded=false
        drawAs=RoundedBox
        imageType=NoImage
        mirroring=NoMirror
        resourceName=None
        resourceObject=null
        tiling=NoTile
        imageSize=<0 0>
        margin=<ltrb 0 0 0 0>
        outlineSettings {
         bUseBrushTransparency=false
         roundingType=FixedRadius
         width=1
         color {colorUseRule=UseColor_ColorTable, specifiedColor=<rgba 0 0 0 0>}
         cornerRadii=<4 4 4 4>
        }
        tintColor {colorUseRule=UseColor_Specified, specifiedColor=<rgba 0.0048 0.0048 0.0048 1>}
        uVRegion {bIsValid=false, max=<0 0>, min=<0 0>}
       }
       hoveredForeground {colorUseRule=UseColor_ColorTable, specifiedColor=<rgba 0 0 0 0>}
       hoveredSlateSound {resourceObject=null}
       normal {
        bIsDynamicallyLoaded=false
        drawAs=RoundedBox
        imageType=NoImage
        mirroring=NoMirror
        resourceName=None
        resourceObject=null
        tiling=NoTile
        imageSize=<0 0>
        margin=<ltrb 0 0 0 0>
        outlineSettings {
         bUseBrushTransparency=false
         roundingType=FixedRadius
         width=1
         color {colorUseRule=UseColor_ColorTable, specifiedColor=<rgba 0 0 0 0>}
         cornerRadii=<4 4 4 4>
        }
        tintColor {colorUseRule=UseColor_Specified, specifiedColor=<rgba 0.0048 0.0048 0.0048 1>}
        uVRegion {bIsValid=false, max=<0 0>, min=<0 0>}
       }
       normalForeground {colorUseRule=UseColor_ColorTable, specifiedColor=<rgba 0 0 0 0>}
       normalPadding=<ltrb 8 1 8 1>
       pressed {
        bIsDynamicallyLoaded=false
        drawAs=RoundedBox
        imageType=NoImage
        mirroring=NoMirror
        resourceName=None
        resourceObject=null
        tiling=NoTile
        imageSize=<0 0>
        margin=<ltrb 0 0 0 0>
        outlineSettings {
         bUseBrushTransparency=false
         roundingType=FixedRadius
         width=1
         color {colorUseRule=UseColor_ColorTable, specifiedColor=<rgba 0 0 0 0>}
         cornerRadii=<4 4 4 4>
        }
        tintColor {colorUseRule=UseColor_Specified, specifiedColor=<rgba 0.0048 0.0048 0.0048 1>}
        uVRegion {bIsValid=false, max=<0 0>, min=<0 0>}
       }
       pressedForeground {colorUseRule=UseColor_ColorTable, specifiedColor=<rgba 0 0 0 0>}
       pressedPadding=<ltrb 8 2 8 0>
       pressedSlateSound {resourceObject=null}
      }
      contentPadding=<ltrb 0 0 0 0>
      downArrowImage {
       bIsDynamicallyLoaded=false
       drawAs=Image
       imageType=Vector
       mirroring=NoMirror
       resourceName=../../../Engine/Content/Slate/Starship/CoreWidgets/ComboBox/wide-chevron-down.svg
       resourceObject=null
       tiling=NoTile
       imageSize=<16 16>
       margin=<ltrb 0 0 0 0>
       outlineSettings {
        bUseBrushTransparency=false
        roundingType=HalfHeightRadius
        width=0
        color {colorUseRule=UseColor_Specified, specifiedColor=<rgba 0 0 0 0>}
        cornerRadii=<0 0 0 0>
       }
       tintColor {colorUseRule=UseColor_Specified, specifiedColor=<rgba 1 1 1 1>}
       uVRegion {bIsValid=false, max=<0 0>, min=<0 0>}
      }
      downArrowPadding=<ltrb 2 2 2 2>
      menuBorderBrush {
       bIsDynamicallyLoaded=false
       drawAs=RoundedBox
       imageType=NoImage
       mirroring=NoMirror
       resourceName=None
       resourceObject=null
       tiling=NoTile
       imageSize=<0 0>
       margin=<ltrb 0 0 0 0>
       outlineSettings {
        bUseBrushTransparency=false
        roundingType=FixedRadius
        width=1
        color {colorUseRule=UseColor_ColorTable, specifiedColor=<rgba 0 0 0 0>}
        cornerRadii=<0 0 0 0>
       }
       tintColor {colorUseRule=UseColor_Specified, specifiedColor=<rgba 0.0395 0.0395 0.0395 1>}
       uVRegion {bIsValid=false, max=<0 0>, min=<0 0>}
      }
      menuBorderPadding=<ltrb 0 0 0 0>
      shadowColorAndOpacity=<rgba 0 0 0 1>
      shadowOffset=<0 0>
     }
     contentPadding=<ltrb 0 0 0 0>
     menuRowPadding=<ltrb 8 3 8 3>
     pressedSlateSound {resourceObject=null}
     selectionChangeSlateSound {resourceObject=null}
    }
   }
   slot {class=HorizontalBoxSlot, properties {size {sizeRule=Fill, value=1}}}
  }
  WB_FAnim_SimpleTextBlock_2 {
   widgetClassPath=&WB_FAnim_SimpleTextBlock
   parent=HorizontalBox_446
   widget=WB_FAnim_SimpleTextBlock_2
   properties {
    text=Hand IK
    toolTipText="Override Hand IK\r\nDefault: No override\r\nForce IK will force IK alpha to 1\r\nForce FK will force IK alpha to 0"
    boxSize=<60 25>
    textPadding=<ltrb 5 5 5 5>
   }
   slot {class=HorizontalBoxSlot}
  }
  ComboBox_HandIK {
   widgetClassPath=&ComboBoxString
   bIsVariable=true
   parent=HorizontalBox_446
   widget=ComboBox_HandIK
   properties {
    contentPadding=<ltrb 2 2 2 2>
    font {
     bForceMonospaced=false
     bMaterialIsStencil=false
     fontMaterial=null
     fontObject=&Roboto
     letterSpacing=0
     monospacedWidth=1
     size=9
     skewAmount=0
     typefaceFontName=Regular
     outlineSettings {
      bApplyOutlineToDropShadows=false
      bMiteredCorners=false
      bSeparateFillAlpha=false
      outlineMaterial=null
      outlineSize=0
      outlineColor=<rgba 0 0 0 1>
     }
    }
    foregroundColor {colorUseRule=UseColor_Specified, specifiedColor=<rgba 0.5271 0.5271 0.5271 1>}
    itemStyle {
     bUseParentRowBrush=false
     activeBrush {
      bIsDynamicallyLoaded=false
      drawAs=Image
      imageType=NoImage
      mirroring=NoMirror
      resourceName=None
      resourceObject=null
      tiling=NoTile
      imageSize=<0 0>
      margin=<ltrb 0 0 0 0>
      outlineSettings {
       bUseBrushTransparency=false
       roundingType=HalfHeightRadius
       width=0
       color {colorUseRule=UseColor_Specified, specifiedColor=<rgba 0 0 0 0>}
       cornerRadii=<0 0 0 0>
      }
      tintColor {colorUseRule=UseColor_Specified, specifiedColor=<rgba 0 0.162 0.7454 1>}
      uVRegion {bIsValid=false, max=<0 0>, min=<0 0>}
     }
     activeHighlightedBrush {
      bIsDynamicallyLoaded=false
      drawAs=Image
      imageType=NoImage
      mirroring=NoMirror
      resourceName=None
      resourceObject=null
      tiling=NoTile
      imageSize=<0 0>
      margin=<ltrb 0 0 0 0>
      outlineSettings {
       bUseBrushTransparency=false
       roundingType=HalfHeightRadius
       width=0
       color {colorUseRule=UseColor_Specified, specifiedColor=<rgba 0 0 0 0>}
       cornerRadii=<0 0 0 0>
      }
      tintColor {colorUseRule=UseColor_Specified, specifiedColor=<rgba 0.0252 0.0319 0.0423 1>}
      uVRegion {bIsValid=false, max=<0 0>, min=<0 0>}
     }
     activeHoveredBrush {
      bIsDynamicallyLoaded=false
      drawAs=Image
      imageType=NoImage
      mirroring=NoMirror
      resourceName=None
      resourceObject=null
      tiling=NoTile
      imageSize=<0 0>
      margin=<ltrb 0 0 0 0>
      outlineSettings {
       bUseBrushTransparency=false
       roundingType=HalfHeightRadius
       width=0
       color {colorUseRule=UseColor_Specified, specifiedColor=<rgba 0 0 0 0>}
       cornerRadii=<0 0 0 0>
      }
      tintColor {colorUseRule=UseColor_Specified, specifiedColor=<rgba 0 0.162 0.7454 1>}
      uVRegion {bIsValid=false, max=<0 0>, min=<0 0>}
     }
     dropIndicator_Above {
      bIsDynamicallyLoaded=false
      drawAs=Box
      imageType=FullColor
      mirroring=NoMirror
      resourceName=../../../Engine/Content/Slate/Common/DropZoneIndicator_Above.png
      resourceObject=null
      tiling=NoTile
      imageSize=<0 0>
      margin=<ltrb 0.625 0.625 0 0>
      outlineSettings {
       bUseBrushTransparency=false
       roundingType=HalfHeightRadius
       width=0
       color {colorUseRule=UseColor_Specified, specifiedColor=<rgba 0 0 0 0>}
       cornerRadii=<0 0 0 0>
      }
      tintColor {colorUseRule=UseColor_Specified, specifiedColor=<rgba 0.0091 0.3515 0.9647 1>}
      uVRegion {bIsValid=false, max=<0 0>, min=<0 0>}
     }
     dropIndicator_Below {
      bIsDynamicallyLoaded=false
      drawAs=Box
      imageType=FullColor
      mirroring=NoMirror
      resourceName=../../../Engine/Content/Slate/Common/DropZoneIndicator_Below.png
      resourceObject=null
      tiling=NoTile
      imageSize=<0 0>
      margin=<ltrb 0.625 0 0 0.625>
      outlineSettings {
       bUseBrushTransparency=false
       roundingType=HalfHeightRadius
       width=0
       color {colorUseRule=UseColor_Specified, specifiedColor=<rgba 0 0 0 0>}
       cornerRadii=<0 0 0 0>
      }
      tintColor {colorUseRule=UseColor_Specified, specifiedColor=<rgba 0.0091 0.3515 0.9647 1>}
      uVRegion {bIsValid=false, max=<0 0>, min=<0 0>}
     }
     dropIndicator_Onto {
      bIsDynamicallyLoaded=false
      drawAs=Box
      imageType=FullColor
      mirroring=NoMirror
      resourceName=../../../Engine/Content/Slate/Common/DropZoneIndicator_Onto.png
      resourceObject=null
      tiling=NoTile
      imageSize=<0 0>
      margin=<ltrb 0.25 0.25 0.25 0.25>
      outlineSettings {
       bUseBrushTransparency=false
       roundingType=HalfHeightRadius
       width=0
       color {colorUseRule=UseColor_Specified, specifiedColor=<rgba 0 0 0 0>}
       cornerRadii=<0 0 0 0>
      }
      tintColor {colorUseRule=UseColor_Specified, specifiedColor=<rgba 0.0091 0.3515 0.9647 1>}
      uVRegion {bIsValid=false, max=<0 0>, min=<0 0>}
     }
     evenRowBackgroundBrush {
      bIsDynamicallyLoaded=false
      drawAs=Image
      imageType=NoImage
      mirroring=NoMirror
      resourceName=None
      resourceObject=null
      tiling=NoTile
      imageSize=<0 0>
      margin=<ltrb 0 0 0 0>
      outlineSettings {
       bUseBrushTransparency=false
       roundingType=HalfHeightRadius
       width=0
       color {colorUseRule=UseColor_Specified, specifiedColor=<rgba 0 0 0 0>}
       cornerRadii=<0 0 0 0>
      }
      tintColor {colorUseRule=UseColor_Specified, specifiedColor=<rgba 0.0103 0.0103 0.0103 1>}
      uVRegion {bIsValid=false, max=<0 0>, min=<0 0>}
     }
     evenRowBackgroundHoveredBrush {
      bIsDynamicallyLoaded=false
      drawAs=Image
      imageType=NoImage
      mirroring=NoMirror
      resourceName=None
      resourceObject=null
      tiling=NoTile
      imageSize=<0 0>
      margin=<ltrb 0 0 0 0>
      outlineSettings {
       bUseBrushTransparency=false
       roundingType=HalfHeightRadius
       width=0
       color {colorUseRule=UseColor_Specified, specifiedColor=<rgba 0 0 0 0>}
       cornerRadii=<0 0 0 0>
      }
      tintColor {colorUseRule=UseColor_Specified, specifiedColor=<rgba 0.0176 0.0176 0.0176 1>}
      uVRegion {bIsValid=false, max=<0 0>, min=<0 0>}
     }
     inactiveBrush {
      bIsDynamicallyLoaded=false
      drawAs=Image
      imageType=NoImage
      mirroring=NoMirror
      resourceName=None
      resourceObject=null
      tiling=NoTile
      imageSize=<0 0>
      margin=<ltrb 0 0 0 0>
      outlineSettings {
       bUseBrushTransparency=false
       roundingType=HalfHeightRadius
       width=0
       color {colorUseRule=UseColor_Specified, specifiedColor=<rgba 0 0 0 0>}
       cornerRadii=<0 0 0 0>
      }
      tintColor {colorUseRule=UseColor_Specified, specifiedColor=<rgba 0.0513 0.0953 0.159 1>}
      uVRegion {bIsValid=false, max=<0 0>, min=<0 0>}
     }
     inactiveHighlightedBrush {
      bIsDynamicallyLoaded=false
      drawAs=Image
      imageType=NoImage
      mirroring=NoMirror
      resourceName=None
      resourceObject=null
      tiling=NoTile
      imageSize=<0 0>
      margin=<ltrb 0 0 0 0>
      outlineSettings {
       bUseBrushTransparency=false
       roundingType=HalfHeightRadius
       width=0
       color {colorUseRule=UseColor_Specified, specifiedColor=<rgba 0 0 0 0>}
       cornerRadii=<0 0 0 0>
      }
      tintColor {colorUseRule=UseColor_Specified, specifiedColor=<rgba 0.0252 0.0319 0.0423 1>}
      uVRegion {bIsValid=false, max=<0 0>, min=<0 0>}
     }
     inactiveHoveredBrush {
      bIsDynamicallyLoaded=false
      drawAs=Image
      imageType=NoImage
      mirroring=NoMirror
      resourceName=None
      resourceObject=null
      tiling=NoTile
      imageSize=<0 0>
      margin=<ltrb 0 0 0 0>
      outlineSettings {
       bUseBrushTransparency=false
       roundingType=HalfHeightRadius
       width=0
       color {colorUseRule=UseColor_Specified, specifiedColor=<rgba 0 0 0 0>}
       cornerRadii=<0 0 0 0>
      }
      tintColor {colorUseRule=UseColor_Specified, specifiedColor=<rgba 0.0513 0.0953 0.159 1>}
      uVRegion {bIsValid=false, max=<0 0>, min=<0 0>}
     }
     oddRowBackgroundBrush {
      bIsDynamicallyLoaded=false
      drawAs=Image
      imageType=NoImage
      mirroring=NoMirror
      resourceName=None
      resourceObject=null
      tiling=NoTile
      imageSize=<0 0>
      margin=<ltrb 0 0 0 0>
      outlineSettings {
       bUseBrushTransparency=false
       roundingType=HalfHeightRadius
       width=0
       color {colorUseRule=UseColor_Specified, specifiedColor=<rgba 0 0 0 0>}
       cornerRadii=<0 0 0 0>
      }
      tintColor {colorUseRule=UseColor_Specified, specifiedColor=<rgba 0.0103 0.0103 0.0103 1>}
      uVRegion {bIsValid=false, max=<0 0>, min=<0 0>}
     }
     oddRowBackgroundHoveredBrush {
      bIsDynamicallyLoaded=false
      drawAs=Image
      imageType=NoImage
      mirroring=NoMirror
      resourceName=None
      resourceObject=null
      tiling=NoTile
      imageSize=<0 0>
      margin=<ltrb 0 0 0 0>
      outlineSettings {
       bUseBrushTransparency=false
       roundingType=HalfHeightRadius
       width=0
       color {colorUseRule=UseColor_Specified, specifiedColor=<rgba 0 0 0 0>}
       cornerRadii=<0 0 0 0>
      }
      tintColor {colorUseRule=UseColor_Specified, specifiedColor=<rgba 0.0176 0.0176 0.0176 1>}
      uVRegion {bIsValid=false, max=<0 0>, min=<0 0>}
     }
     parentRowBackgroundBrush {
      bIsDynamicallyLoaded=false
      drawAs=Image
      imageType=NoImage
      mirroring=NoMirror
      resourceName=None
      resourceObject=null
      tiling=NoTile
      imageSize=<32 32>
      margin=<ltrb 0 0 0 0>
      outlineSettings {
       bUseBrushTransparency=false
       roundingType=HalfHeightRadius
       width=0
       color {colorUseRule=UseColor_Specified, specifiedColor=<rgba 0 0 0 0>}
       cornerRadii=<0 0 0 0>
      }
      tintColor {colorUseRule=UseColor_Specified, specifiedColor=<rgba 1 1 1 1>}
      uVRegion {bIsValid=false, max=<0 0>, min=<0 0>}
     }
     parentRowBackgroundHoveredBrush {
      bIsDynamicallyLoaded=false
      drawAs=Image
      imageType=NoImage
      mirroring=NoMirror
      resourceName=None
      resourceObject=null
      tiling=NoTile
      imageSize=<32 32>
      margin=<ltrb 0 0 0 0>
      outlineSettings {
       bUseBrushTransparency=false
       roundingType=HalfHeightRadius
       width=0
       color {colorUseRule=UseColor_Specified, specifiedColor=<rgba 0 0 0 0>}
       cornerRadii=<0 0 0 0>
      }
      tintColor {colorUseRule=UseColor_Specified, specifiedColor=<rgba 1 1 1 1>}
      uVRegion {bIsValid=false, max=<0 0>, min=<0 0>}
     }
     selectedTextColor {colorUseRule=UseColor_Specified, specifiedColor=<rgba 1 1 1 1>}
     selectorFocusedBrush {
      bIsDynamicallyLoaded=false
      drawAs=Border
      imageType=FullColor
      mirroring=NoMirror
      resourceName=../../../Engine/Content/Slate/Common/Selector.png
      resourceObject=null
      tiling=Both
      imageSize=<0 0>
      margin=<ltrb 0.25 0.25 0.25 0.25>
      outlineSettings {
       bUseBrushTransparency=false
       roundingType=HalfHeightRadius
       width=0
       color {colorUseRule=UseColor_Specified, specifiedColor=<rgba 0 0 0 0>}
       cornerRadii=<0 0 0 0>
      }
      tintColor {colorUseRule=UseColor_Specified, specifiedColor=<rgba 0.701 0.225 0.003 1>}
      uVRegion {bIsValid=false, max=<0 0>, min=<0 0>}
     }
     textColor {colorUseRule=UseColor_Specified, specifiedColor=<rgba 0.5271 0.5271 0.5271 1>}
    }
    scrollBarStyle {
     thickness=8
     draggedThumbImage {
      bIsDynamicallyLoaded=false
      drawAs=RoundedBox
      imageType=NoImage
      mirroring=NoMirror
      resourceName=None
      resourceObject=null
      tiling=NoTile
      imageSize=<0 0>
      margin=<ltrb 0 0 0 0>
      outlineSettings {
       bUseBrushTransparency=false
       roundingType=FixedRadius
       width=0
       color {colorUseRule=UseColor_Specified, specifiedColor=<rgba 0 0 0 0>}
       cornerRadii=<4 4 4 4>
      }
      tintColor {colorUseRule=UseColor_Specified, specifiedColor=<rgba 0.2159 0.2159 0.2159 1>}
      uVRegion {bIsValid=false, max=<0 0>, min=<0 0>}
     }
     horizontalBackgroundImage {
      bIsDynamicallyLoaded=false
      drawAs=NoDrawType
      imageType=NoImage
      mirroring=NoMirror
      resourceName=None
      resourceObject=null
      tiling=NoTile
      imageSize=<0 0>
      margin=<ltrb 0 0 0 0>
      outlineSettings {
       bUseBrushTransparency=false
       roundingType=HalfHeightRadius
       width=0
       color {colorUseRule=UseColor_Specified, specifiedColor=<rgba 0 0 0 0>}
       cornerRadii=<0 0 0 0>
      }
      tintColor {colorUseRule=UseColor_Specified, specifiedColor=<rgba 1 1 1 1>}
      uVRegion {bIsValid=false, max=<0 0>, min=<0 0>}
     }
     horizontalBottomSlotImage {
      bIsDynamicallyLoaded=false
      drawAs=NoDrawType
      imageType=NoImage
      mirroring=NoMirror
      resourceName=None
      resourceObject=null
      tiling=NoTile
      imageSize=<0 0>
      margin=<ltrb 0 0 0 0>
      outlineSettings {
       bUseBrushTransparency=false
       roundingType=HalfHeightRadius
       width=0
       color {colorUseRule=UseColor_Specified, specifiedColor=<rgba 0 0 0 0>}
       cornerRadii=<0 0 0 0>
      }
      tintColor {colorUseRule=UseColor_Specified, specifiedColor=<rgba 1 1 1 1>}
      uVRegion {bIsValid=false, max=<0 0>, min=<0 0>}
     }
     horizontalTopSlotImage {
      bIsDynamicallyLoaded=false
      drawAs=NoDrawType
      imageType=NoImage
      mirroring=NoMirror
      resourceName=None
      resourceObject=null
      tiling=NoTile
      imageSize=<0 0>
      margin=<ltrb 0 0 0 0>
      outlineSettings {
       bUseBrushTransparency=false
       roundingType=HalfHeightRadius
       width=0
       color {colorUseRule=UseColor_Specified, specifiedColor=<rgba 0 0 0 0>}
       cornerRadii=<0 0 0 0>
      }
      tintColor {colorUseRule=UseColor_Specified, specifiedColor=<rgba 1 1 1 1>}
      uVRegion {bIsValid=false, max=<0 0>, min=<0 0>}
     }
     hoveredThumbImage {
      bIsDynamicallyLoaded=false
      drawAs=RoundedBox
      imageType=NoImage
      mirroring=NoMirror
      resourceName=None
      resourceObject=null
      tiling=NoTile
      imageSize=<0 0>
      margin=<ltrb 0 0 0 0>
      outlineSettings {
       bUseBrushTransparency=false
       roundingType=FixedRadius
       width=0
       color {colorUseRule=UseColor_Specified, specifiedColor=<rgba 0 0 0 0>}
       cornerRadii=<4 4 4 4>
      }
      tintColor {colorUseRule=UseColor_Specified, specifiedColor=<rgba 0.2159 0.2159 0.2159 1>}
      uVRegion {bIsValid=false, max=<0 0>, min=<0 0>}
     }
     normalThumbImage {
      bIsDynamicallyLoaded=false
      drawAs=RoundedBox
      imageType=NoImage
      mirroring=NoMirror
      resourceName=None
      resourceObject=null
      tiling=NoTile
      imageSize=<0 0>
      margin=<ltrb 0 0 0 0>
      outlineSettings {
       bUseBrushTransparency=false
       roundingType=FixedRadius
       width=0
       color {colorUseRule=UseColor_Specified, specifiedColor=<rgba 0 0 0 0>}
       cornerRadii=<4 4 4 4>
      }
      tintColor {colorUseRule=UseColor_Specified, specifiedColor=<rgba 0.0953 0.0953 0.0953 1>}
      uVRegion {bIsValid=false, max=<0 0>, min=<0 0>}
     }
     verticalBackgroundImage {
      bIsDynamicallyLoaded=false
      drawAs=NoDrawType
      imageType=NoImage
      mirroring=NoMirror
      resourceName=None
      resourceObject=null
      tiling=NoTile
      imageSize=<0 0>
      margin=<ltrb 0 0 0 0>
      outlineSettings {
       bUseBrushTransparency=false
       roundingType=HalfHeightRadius
       width=0
       color {colorUseRule=UseColor_Specified, specifiedColor=<rgba 0 0 0 0>}
       cornerRadii=<0 0 0 0>
      }
      tintColor {colorUseRule=UseColor_Specified, specifiedColor=<rgba 1 1 1 1>}
      uVRegion {bIsValid=false, max=<0 0>, min=<0 0>}
     }
     verticalBottomSlotImage {
      bIsDynamicallyLoaded=false
      drawAs=NoDrawType
      imageType=NoImage
      mirroring=NoMirror
      resourceName=None
      resourceObject=null
      tiling=NoTile
      imageSize=<0 0>
      margin=<ltrb 0 0 0 0>
      outlineSettings {
       bUseBrushTransparency=false
       roundingType=HalfHeightRadius
       width=0
       color {colorUseRule=UseColor_Specified, specifiedColor=<rgba 0 0 0 0>}
       cornerRadii=<0 0 0 0>
      }
      tintColor {colorUseRule=UseColor_Specified, specifiedColor=<rgba 1 1 1 1>}
      uVRegion {bIsValid=false, max=<0 0>, min=<0 0>}
     }
     verticalTopSlotImage {
      bIsDynamicallyLoaded=false
      drawAs=NoDrawType
      imageType=NoImage
      mirroring=NoMirror
      resourceName=None
      resourceObject=null
      tiling=NoTile
      imageSize=<0 0>
      margin=<ltrb 0 0 0 0>
      outlineSettings {
       bUseBrushTransparency=false
       roundingType=HalfHeightRadius
       width=0
       color {colorUseRule=UseColor_Specified, specifiedColor=<rgba 0 0 0 0>}
       cornerRadii=<0 0 0 0>
      }
      tintColor {colorUseRule=UseColor_Specified, specifiedColor=<rgba 1 1 1 1>}
      uVRegion {bIsValid=false, max=<0 0>, min=<0 0>}
     }
    }
    widgetStyle {
     comboButtonStyle {
      downArrowAlign=VAlign_Center
      buttonStyle {
       clickedSlateSound {resourceObject=null}
       disabled {
        bIsDynamicallyLoaded=false
        drawAs=RoundedBox
        imageType=NoImage
        mirroring=NoMirror
        resourceName=None
        resourceObject=null
        tiling=NoTile
        imageSize=<0 0>
        margin=<ltrb 0 0 0 0>
        outlineSettings {
         bUseBrushTransparency=false
         roundingType=FixedRadius
         width=1
         color {colorUseRule=UseColor_ColorTable, specifiedColor=<rgba 0 0 0 0>}
         cornerRadii=<4 4 4 4>
        }
        tintColor {colorUseRule=UseColor_Specified, specifiedColor=<rgba 0.0395 0.0395 0.0395 1>}
        uVRegion {bIsValid=false, max=<0 0>, min=<0 0>}
       }
       disabledForeground {colorUseRule=UseColor_ColorTable, specifiedColor=<rgba 0 0 0 0>}
       hovered {
        bIsDynamicallyLoaded=false
        drawAs=RoundedBox
        imageType=NoImage
        mirroring=NoMirror
        resourceName=None
        resourceObject=null
        tiling=NoTile
        imageSize=<0 0>
        margin=<ltrb 0 0 0 0>
        outlineSettings {
         bUseBrushTransparency=false
         roundingType=FixedRadius
         width=1
         color {colorUseRule=UseColor_ColorTable, specifiedColor=<rgba 0 0 0 0>}
         cornerRadii=<4 4 4 4>
        }
        tintColor {colorUseRule=UseColor_Specified, specifiedColor=<rgba 0.0048 0.0048 0.0048 1>}
        uVRegion {bIsValid=false, max=<0 0>, min=<0 0>}
       }
       hoveredForeground {colorUseRule=UseColor_ColorTable, specifiedColor=<rgba 0 0 0 0>}
       hoveredSlateSound {resourceObject=null}
       normal {
        bIsDynamicallyLoaded=false
        drawAs=RoundedBox
        imageType=NoImage
        mirroring=NoMirror
        resourceName=None
        resourceObject=null
        tiling=NoTile
        imageSize=<0 0>
        margin=<ltrb 0 0 0 0>
        outlineSettings {
         bUseBrushTransparency=false
         roundingType=FixedRadius
         width=1
         color {colorUseRule=UseColor_ColorTable, specifiedColor=<rgba 0 0 0 0>}
         cornerRadii=<4 4 4 4>
        }
        tintColor {colorUseRule=UseColor_Specified, specifiedColor=<rgba 0.0048 0.0048 0.0048 1>}
        uVRegion {bIsValid=false, max=<0 0>, min=<0 0>}
       }
       normalForeground {colorUseRule=UseColor_ColorTable, specifiedColor=<rgba 0 0 0 0>}
       normalPadding=<ltrb 8 1 8 1>
       pressed {
        bIsDynamicallyLoaded=false
        drawAs=RoundedBox
        imageType=NoImage
        mirroring=NoMirror
        resourceName=None
        resourceObject=null
        tiling=NoTile
        imageSize=<0 0>
        margin=<ltrb 0 0 0 0>
        outlineSettings {
         bUseBrushTransparency=false
         roundingType=FixedRadius
         width=1
         color {colorUseRule=UseColor_ColorTable, specifiedColor=<rgba 0 0 0 0>}
         cornerRadii=<4 4 4 4>
        }
        tintColor {colorUseRule=UseColor_Specified, specifiedColor=<rgba 0.0048 0.0048 0.0048 1>}
        uVRegion {bIsValid=false, max=<0 0>, min=<0 0>}
       }
       pressedForeground {colorUseRule=UseColor_ColorTable, specifiedColor=<rgba 0 0 0 0>}
       pressedPadding=<ltrb 8 2 8 0>
       pressedSlateSound {resourceObject=null}
      }
      contentPadding=<ltrb 0 0 0 0>
      downArrowImage {
       bIsDynamicallyLoaded=false
       drawAs=Image
       imageType=Vector
       mirroring=NoMirror
       resourceName=../../../Engine/Content/Slate/Starship/CoreWidgets/ComboBox/wide-chevron-down.svg
       resourceObject=null
       tiling=NoTile
       imageSize=<16 16>
       margin=<ltrb 0 0 0 0>
       outlineSettings {
        bUseBrushTransparency=false
        roundingType=HalfHeightRadius
        width=0
        color {colorUseRule=UseColor_Specified, specifiedColor=<rgba 0 0 0 0>}
        cornerRadii=<0 0 0 0>
       }
       tintColor {colorUseRule=UseColor_Specified, specifiedColor=<rgba 1 1 1 1>}
       uVRegion {bIsValid=false, max=<0 0>, min=<0 0>}
      }
      downArrowPadding=<ltrb 2 2 2 2>
      menuBorderBrush {
       bIsDynamicallyLoaded=false
       drawAs=RoundedBox
       imageType=NoImage
       mirroring=NoMirror
       resourceName=None
       resourceObject=null
       tiling=NoTile
       imageSize=<0 0>
       margin=<ltrb 0 0 0 0>
       outlineSettings {
        bUseBrushTransparency=false
        roundingType=FixedRadius
        width=1
        color {colorUseRule=UseColor_ColorTable, specifiedColor=<rgba 0 0 0 0>}
        cornerRadii=<0 0 0 0>
       }
       tintColor {colorUseRule=UseColor_Specified, specifiedColor=<rgba 0.0395 0.0395 0.0395 1>}
       uVRegion {bIsValid=false, max=<0 0>, min=<0 0>}
      }
      menuBorderPadding=<ltrb 0 0 0 0>
      shadowColorAndOpacity=<rgba 0 0 0 1>
      shadowOffset=<0 0>
     }
     contentPadding=<ltrb 0 0 0 0>
     menuRowPadding=<ltrb 8 3 8 3>
     pressedSlateSound {resourceObject=null}
     selectionChangeSlateSound {resourceObject=null}
    }
   }
   slot {class=HorizontalBoxSlot, properties {size {sizeRule=Fill, value=1}}}
  }
  EUW_FAnim_SimpleConsoleButton {
   widgetClassPath=&EUW_FAnim_SimpleConsoleButton
   parent=HorizontalBox_446
   widget=EUW_FAnim_SimpleConsoleButton
   properties {
    buttonName=Enable Slope Warping
    command Has Int State=true
    console Command=a.FortAnimNode.SlopeWarping.Enable
    run On Server=false
    toolTipText="Enable/Disables slope warping, which is responsible for aligning the IK sockets with the ground."
   }
   slot {class=HorizontalBoxSlot, properties {size {sizeRule=Fill, value=1}}}
  }
  HorizontalBox_341 {
   widgetClassPath=&HorizontalBox
   parent=VerticalBox_57
   widget=HorizontalBox_341
   slot {class=VerticalBoxSlot, properties {size {sizeRule=Fill, value=1}}}
  }
  WB_FAnim_SimpleTextBlock_102 {
   widgetClassPath=&WB_FAnim_SimpleTextBlock
   parent=HorizontalBox_341
   widget=WB_FAnim_SimpleTextBlock_102
   properties {text=Property Filters, toolTipText=Add or remove IK property filters, boxSize=<85 25>, textPadding=<ltrb 5 0 0 0>}
   slot {class=HorizontalBoxSlot}
  }
  Button_AddIKFilters {
   widgetClassPath=&/EUW_FAnim_SimpleButton
   bIsVariable=true
   parent=HorizontalBox_341
   widget=Button_AddIKFilters
   properties {button Width=50, buttonName=Add, toolTipText=Add IK related property filters}
   slot {class=HorizontalBoxSlot, properties {size {sizeRule=Fill, value=1}}}
  }
  Button_RemoveIKFilters {
   widgetClassPath=&EUW_FAnim_SimpleButton
   bIsVariable=true
   parent=HorizontalBox_341
   widget=Button_RemoveIKFilters
   properties {button Width=50, buttonName=Remove, toolTipText=Remove IK related property filters}
   slot {class=HorizontalBoxSlot, properties {size {sizeRule=Fill, value=1}}}
  }
  WB_FAnim_SimpleTextBlock_3 {
   widgetClassPath=&WB_FAnim_SimpleTextBlock
   parent=HorizontalBox_341
   widget=WB_FAnim_SimpleTextBlock_3
   properties {text=Bone Filters, toolTipText=Add or remove IK property filters, boxSize=<85 25>, textPadding=<ltrb 5 0 0 0>}
   slot {class=HorizontalBoxSlot}
  }
  Button_AddIKBones {
   widgetClassPath=&EUW_FAnim_SimpleButton
   bIsVariable=true
   parent=HorizontalBox_341
   widget=Button_AddIKBones
   properties {button Width=50, buttonName=Add, toolTipText=Add IK related property filters}
   slot {class=HorizontalBoxSlot, properties {size {sizeRule=Fill, value=1}}}
  }
  Button_RemoveIKBones {
   widgetClassPath=&EUW_FAnim_SimpleButton
   bIsVariable=true
   parent=HorizontalBox_341
   widget=Button_RemoveIKBones
   properties {button Width=50, buttonName=Remove, toolTipText=Remove IK related property filters}
   slot {class=HorizontalBoxSlot, properties {size {sizeRule=Fill, value=1}}}
  }
  Expandable_2DDisplay {
   widgetClassPath=&EditorUtilityExpandableArea
   bIsVariable=true
   parent=MainScrollBox
   widget=Expandable_2DDisplay
   properties {bIsExpanded=true}
   slot {class=ScrollBoxSlot, properties {padding=<ltrb 0 2 0 0>}}
  }
  WB_SimpleTextBlock_3 {
   widgetClassPath=&WB_FAnim_SimpleTextBlock
   namedSlotHost=Expandable_2DDisplay
   widget=WB_SimpleTextBlock_3
   properties {fontSize=11, isBold=true, text=2D Display}
  }
  HorizontalBox_1498 {widgetClassPath=&HorizontalBox, namedSlotHost=Expandable_2DDisplay, widget=HorizontalBox_1498}
  EUW_SimpleConsoleButton_10 {
   widgetClassPath=&EUW_FAnim_SimpleConsoleButton
   parent=HorizontalBox_1498
   widget=EUW_SimpleConsoleButton_10
   properties {
    buttonName=Linked Graphs
    command Has Int State=true
    console Command=Fort.Anim.Debug.DrawLinkedGraphs
    toolTipText=Show currently linked graphs and layers
   }
   slot {class=HorizontalBoxSlot, properties {size {sizeRule=Fill, value=1}}}
  }
  EUW_SimpleConsoleButton_12 {
   widgetClassPath=&EUW_FAnim_SimpleConsoleButton
   parent=HorizontalBox_1498
   widget=EUW_SimpleConsoleButton_12
   properties {buttonName=Montages, command Has Int State=true, console Command=Fort.Anim.Debug.DrawMontages, toolTipText=Show active montages}
   slot {class=HorizontalBoxSlot, properties {size {sizeRule=Fill, value=1}}}
  }
  EUW_SimpleConsoleButton_11 {
   widgetClassPath=&EUW_FAnim_SimpleConsoleButton
   parent=HorizontalBox_1498
   widget=EUW_SimpleConsoleButton_11
   properties {
    buttonName=Pawn Transform
    command Has Int State=true
    console Command=Fort.Anim.Debug.ShowPawnTransform
    toolTipText=Show selected pawn world transform
   }
   slot {class=HorizontalBoxSlot, properties {size {sizeRule=Fill, value=1}}}
  }
  Button_GameplayDebugInfo {
   widgetClassPath=&EUW_FAnim_SimpleConsoleButton
   bIsVariable=true
   parent=HorizontalBox_1498
   widget=Button_GameplayDebugInfo
   properties {
    buttonName=Hide Pawn Debug
    command Has Int State=true
    console Command=a.Fort.DrawFortPawnDebug
    invertColorLogic=true
    toolTipText="The Anim debug turns off the Pawn debug and other debug categories  by default to save screen space. Use this button to turn it back on."
   }
   slot {class=HorizontalBoxSlot, properties {size {sizeRule=Fill, value=1}}}
  }
  Expandable_3DDisplay {
   widgetClassPath=&EditorUtilityExpandableArea
   bIsVariable=true
   parent=MainScrollBox
   widget=Expandable_3DDisplay
   properties {bIsExpanded=true}
   slot {class=ScrollBoxSlot}
  }
  WB_SimpleTextBlock_4 {
   widgetClassPath=&WB_FAnim_SimpleTextBlock
   namedSlotHost=Expandable_3DDisplay
   widget=WB_SimpleTextBlock_4
   properties {fontSize=11, isBold=true, text=3D Display}
  }
  HorizontalBox_3 {widgetClassPath=&HorizontalBox, namedSlotHost=Expandable_3DDisplay, widget=HorizontalBox_3}
  EUW_SimpleConsoleButton_15 {
   widgetClassPath=&EUW_FAnim_SimpleConsoleButton
   parent=HorizontalBox_3
   widget=EUW_SimpleConsoleButton_15
   properties {
    buttonName="Actor && Root rotations"
    command Has Int State=true
    console Command=Fort.Anim.Debug.ShowRootAndActorOrientation
    toolTipText="Show actor capsule rotation (blue) and root rotation (orange)"
   }
   slot {class=HorizontalBoxSlot, properties {size {sizeRule=Fill, value=1}}}
  }
  EUW_SimpleConsoleButton_13 {
   widgetClassPath=&EUW_FAnim_SimpleConsoleButton
   parent=HorizontalBox_3
   widget=EUW_SimpleConsoleButton_13
   properties {
    buttonName="Velocity && Accel"
    command Has Int State=true
    console Command=Fort.Anim.Debug.ShowVelocityAndAccel
    toolTipText="Show pawn velocity (yellow) and acceleration (green)"
   }
   slot {class=HorizontalBoxSlot, properties {size {sizeRule=Fill, value=1}}}
  }
  EUW_SimpleConsoleButton_14 {
   widgetClassPath=&EUW_FAnim_SimpleConsoleButton
   parent=HorizontalBox_3
   widget=EUW_SimpleConsoleButton_14
   properties {
    buttonName=Bounding Box
    command Has Int State=true
    console Command=r.Debug.DrawCurrentDebugTargetBoundingBox
    toolTipText=Show green bounding box around selected debug target
   }
   slot {class=HorizontalBoxSlot, properties {size {sizeRule=Fill, value=1}}}
  }
  EUW_SimpleConsoleButton_16 {
   widgetClassPath=&EUW_FAnim_SimpleConsoleButton
   parent=HorizontalBox_3
   widget=EUW_SimpleConsoleButton_16
   properties {
    buttonName=Camera Trajectory
    command Has Int State=true
    console Command=a.Fort.DrawCameraTrajectory
    enabledIntValue=10
    toolTipText="Show last 10 frames of camera trajectory (only when recording)"
   }
   slot {class=HorizontalBoxSlot, properties {size {sizeRule=Fill, value=1}}}
  }
  Expandable_Utils {
   widgetClassPath=&EditorUtilityExpandableArea
   parent=MainScrollBox
   widget=Expandable_Utils
   properties {bIsExpanded=true}
   slot {class=ScrollBoxSlot}
  }
  WB_SimpleTextBlock {
   widgetClassPath=&WB_FAnim_SimpleTextBlock
   bIsVariable=true
   namedSlotHost=Expandable_Utils
   widget=WB_SimpleTextBlock
   properties {fontSize=11, isBold=true, text=Utils}
  }
  VerticalBox_45 {widgetClassPath=&VerticalBox, namedSlotHost=Expandable_Utils, widget=VerticalBox_45}
  HorizontalBox_171 {
   widgetClassPath=&HorizontalBox
   parent=VerticalBox_45
   widget=HorizontalBox_171
   slot {class=VerticalBoxSlot, properties {padding=<ltrb 0 1 0 0>}}
  }
  EUW_SimpleConsoleButton_7 {
   widgetClassPath=&EUW_FAnim_SimpleConsoleButton
   parent=HorizontalBox_171
   widget=EUW_SimpleConsoleButton_7
   properties {
    buttonName=Enable Melee Twist
    command Has Int State=true
    console Command=Fort.Anim.DisableMeleeTwist
    invertColorLogic=true
    toolTipText=Enable/Disable Melee Twist root rotation
   }
   slot {class=HorizontalBoxSlot, properties {size {sizeRule=Fill, value=1}}}
  }
  EUW_SimpleConsoleButton_144 {
   widgetClassPath=&EUW_FAnim_SimpleConsoleButton
   parent=HorizontalBox_171
   widget=EUW_SimpleConsoleButton_144
   properties {
    buttonName=Enable Aim Offsets
    command Has Int State=true
    console Command=Fort.Anim.DisableAimOffsets
    invertColorLogic=true
    toolTipText=Enable/Disables Aim Offsets
   }
   slot {class=HorizontalBoxSlot, properties {size {sizeRule=Fill, value=1}}}
  }
  EUW_SimpleConsoleButton_23 {
   widgetClassPath=&EUW_FAnim_SimpleConsoleButton
   parent=HorizontalBox_171
   widget=EUW_SimpleConsoleButton_23
   properties {
    buttonName=Enable Facial Anims
    command Has Int State=true
    console Command=a.Fort.UseFacialCurves
    toolTipText="Enable/Disables the gameplay, curve driven, facial animations. Useful to remove curve bloat in the rewind debugger."
   }
   slot {class=HorizontalBoxSlot, properties {size {sizeRule=Fill, value=1}}}
  }
  EUW_SimpleConsoleButton_18 {
   widgetClassPath=&EUW_FAnim_SimpleConsoleButton
   parent=HorizontalBox_171
   widget=EUW_SimpleConsoleButton_18
   properties {buttonName=Hide Weapon, command Has Int State=true, console Command=a.Fort.HideWeapon, toolTipText=Hide the weapon}
   slot {class=HorizontalBoxSlot, properties {size {sizeRule=Fill, value=1}}}
  }
  HorizontalBox_335 {widgetClassPath=&HorizontalBox, parent=VerticalBox_45, widget=HorizontalBox_335, slot {class=VerticalBoxSlot}}
  EUW_SimpleConsoleButton_19 {
   widgetClassPath=&EUW_FAnim_SimpleConsoleButton
   parent=HorizontalBox_335
   widget=EUW_SimpleConsoleButton_19
   properties {
    buttonName=Lock Movement
    command Has Int State=true
    console Command=Fort.EnableMoveInPlace
    toolTipText="Prevent the capsule from moving (only works on ground)"
   }
   slot {class=HorizontalBoxSlot, properties {size {sizeRule=Fill, value=1}}}
  }
  EUW_SimpleConsoleButton_22 {
   widgetClassPath=&EUW_FAnim_SimpleConsoleButton
   parent=HorizontalBox_335
   widget=EUW_SimpleConsoleButton_22
   properties {
    buttonName=Infinite Stamina
    command Has Int State=true
    console Command=EnergyComponent.EnergyUsageIsFree
    toolTipText=Toggle infinite stamina
   }
   slot {class=HorizontalBoxSlot, properties {size {sizeRule=Fill, value=1}}}
  }
  HorizontalBox_229 {widgetClassPath=&HorizontalBox, parent=VerticalBox_45, widget=HorizontalBox_229, slot {class=VerticalBoxSlot}}
  EUW_SimpleConsoleButton_20 {
   widgetClassPath=&EUW_FAnim_SimpleConsoleButton
   parent=HorizontalBox_229
   widget=EUW_SimpleConsoleButton_20
   properties {buttonName=Fixed Camera, console Command=Camera Fixed, toolTipText=Set camera mode to fixed}
   slot {class=HorizontalBoxSlot, properties {size {sizeRule=Fill, value=1}}}
  }
  EUW_SimpleConsoleButton_21 {
   widgetClassPath=&EUW_FAnim_SimpleConsoleButton
   parent=HorizontalBox_229
   widget=EUW_SimpleConsoleButton_21
   properties {buttonName=Default Camera, console Command=Camera Default, toolTipText=Set camera mode to default}
   slot {class=HorizontalBoxSlot, properties {size {sizeRule=Fill, value=1}}}
  }
  HorizontalBox_41 {widgetClassPath=&HorizontalBox, parent=VerticalBox_45, widget=HorizontalBox_41, slot {class=VerticalBoxSlot}}
  WB_SimpleTextBlock_71 {
   widgetClassPath=&WB_FAnim_SimpleTextBlock
   parent=HorizontalBox_41
   widget=WB_SimpleTextBlock_71
   properties {text=Timescale}
   slot {class=HorizontalBoxSlot}
  }
  EUW_SimpleConsoleButton {
   widgetClassPath=&EUW_FAnim_SimpleConsoleButton
   parent=HorizontalBox_41
   widget=EUW_SimpleConsoleButton
   properties {buttonName=0.25, console Command=cheat Slomo .25}
   slot {class=HorizontalBoxSlot, properties {padding=<ltrb 1 0 0 0>, size {sizeRule=Fill, value=1}}}
  }
  EUW_SimpleConsoleButton_1 {
   widgetClassPath=&EUW_FAnim_SimpleConsoleButton
   parent=HorizontalBox_41
   widget=EUW_SimpleConsoleButton_1
   properties {buttonName=0.5, console Command=cheat Slomo .50}
   slot {class=HorizontalBoxSlot, properties {padding=<ltrb 1 0 0 0>, size {sizeRule=Fill, value=1}}}
  }
  EUW_SimpleConsoleButton_2 {
   widgetClassPath=&EUW_FAnim_SimpleConsoleButton
   parent=HorizontalBox_41
   widget=EUW_SimpleConsoleButton_2
   properties {buttonName=0.75, console Command=cheat Slomo .75}
   slot {class=HorizontalBoxSlot, properties {padding=<ltrb 1 0 0 0>, size {sizeRule=Fill, value=1}}}
  }
  EUW_SimpleConsoleButton_3 {
   widgetClassPath=&EUW_FAnim_SimpleConsoleButton
   parent=HorizontalBox_41
   widget=EUW_SimpleConsoleButton_3
   properties {buttonName=1, console Command=cheat Slomo 1}
   slot {class=HorizontalBoxSlot, properties {padding=<ltrb 1 0 0 0>, size {sizeRule=Fill, value=1}}}
  }
  EUW_SimpleConsoleButton_4 {
   widgetClassPath=&EUW_FAnim_SimpleConsoleButton
   parent=HorizontalBox_41
   widget=EUW_SimpleConsoleButton_4
   properties {buttonName=1.25, console Command=cheat Slomo 1.25}
   slot {class=HorizontalBoxSlot, properties {padding=<ltrb 1 0 0 0>, size {sizeRule=Fill, value=1}}}
  }
  EUW_SimpleConsoleButton_5 {
   widgetClassPath=&EUW_FAnim_SimpleConsoleButton
   parent=HorizontalBox_41
   widget=EUW_SimpleConsoleButton_5
   properties {buttonName=1.5, console Command=cheat Slomo 1.50}
   slot {class=HorizontalBoxSlot, properties {padding=<ltrb 1 0 0 0>, size {sizeRule=Fill, value=1}}}
  }
  EUW_SimpleConsoleSpinBox {
   widgetClassPath=&/EUW_FAnim_SimpleConsoleSpinBox
   bIsVariable=true
   parent=HorizontalBox_41
   widget=EUW_SimpleConsoleSpinBox
   properties {console Command=slomo, defaultValue=1, max Value=5, min Value=0.05, use Max=true, use Min=true}
   slot {class=HorizontalBoxSlot, properties {padding=<ltrb 1 1 1 1>, size {sizeRule=Fill, value=3}}}
  }
 ]
}
