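# Editor utility widget for editing the animation properties of a chosen FortWeapon Blueprint, such as montages, anim layers and pose offsets.
# A property view bound to its own CurrentItem picks the class, whose CDO feeds a details view filtered to those properties and enables the controls.
# Open launches the Blueprint editor and Save checks out and saves the asset.
# Tick polls dirty packages to relabel Save as "Save*" for unsaved changes.
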EditorUtilityWidgetBlueprint /Game/Characters/Player/Widgets/EUW_FortWeaponAnimProperties
graphs [
 EnableAssetControls {nodes [
   n1 = EnableAssetControls!() -> (then, Enable:bool=~) {is=title K2Node_FunctionEntry}
   > Set bIsEnabled!(execute, bIsEnabled:bool=@n1, self:Widget Object Reference=Get Button_Save() -> (Button_Save:EUW FAnim Simple Button Object Reference)) -> (then, Output_Get:bool)
   > Set bIsEnabled(bIsEnabled=@n1, self=Get Button_Open() -> (Button_Open:EUW FAnim Simple Button Object Reference))
   > Set bIsEnabled(bIsEnabled=@n1, self=Get DetailsView_Item() -> (DetailsView_Item:Details View Object Reference))
   > Return Node()
  ]}
 EventGraph {nodes [
   "On Property Changed (MainAssetPicker)"!() -> (OutputDelegate, then, PropertyName:Name=None) {is=title K2Node_ComponentBoundEvent}
   > n1 = LoadClassAsset_Blocking(AssetClass=Get CurrentItem() -> (CurrentItem:Fort Weapon Soft Class Reference))
   > n2 = Cast To FortWeapon Class!(execute:Exec, Class:Object Class Reference=@n1) -> (then:Exec, CastFailed:Exec, AsFort Weapon:Fort Weapon Class Reference) {is=title K2Node_ClassDynamicCast}
   > n3 = Set CurrentItemClass(CurrentItemClass:Fort Weapon Class Reference=@n2) -> (Output_Get:Fort Weapon Class Reference)
   > n4 = GetBlueprintForClass!(execute, Class:Object Class Reference=@n3) -> (True:Exec, False:Exec, ReturnValue:Blueprint Object Reference)
   > Set ItemBlueprintRef(ItemBlueprintRef:Object Reference=@n4) -> (Output_Get:Object Reference)
   > n5 = GetClassDefaultObject!(execute, Class:Object Class Reference=@n1) -> (then, ReturnValue:Object Reference)
   > n6 = Cast To FortWeapon!(execute, Object=@n5) -> (then, CastFailed, AsFort Weapon:Fort Weapon Object Reference) {is=title K2Node_DynamicCast}
   > n7 = Set CurrentItemCDO(CurrentItemCDO:Fort Weapon Object Reference=@n6) -> (Output_Get:Fort Weapon Object Reference)
   > n8 = Branch(Condition=IsValid(Object=@n7))
   n9 = Get DetailsView_Item()
   > SetObject(self=@n9, NewObject=Get CurrentItemCDO() -> (CurrentItemCDO:Fort Weapon Object Reference))
   > EnableAssetControls!(execute, self, Enable:bool=true) -> (then) {is=title K2Node_CallFunction}
   SetObject(execute=[@n8.else, @n6.CastFailed], self=@n9)
   > EnableAssetControls(Enable=false)
   "On Clicked (Button_Open)"() {is=title K2Node_ComponentBoundEvent}
   > n10 = LoadClassAsset_Blocking(AssetClass=Get CurrentItem())
   > n11 = GetBlueprintForClass(Class=@n10)
   > Branch(Condition=IsValid(Object=@n11))
   > OpenEditorForAssets!(
    execute
    self:Asset Editor Subsystem Object Reference=Get AssetEditorSubsystem()
    Assets:Array of Object References=Make Array("[0]":Object Reference=@n11) -> (Array:Array of Object References)
    OpenedMethod:EAssetTypeActivationOpenedMethod Enum=Edit
   ) -> (then, ReturnValue:bool)
   "On Clicked (Button_Save)"() {is=title K2Node_ComponentBoundEvent}
   > n12 = LoadClassAsset_Blocking(AssetClass=Get CurrentItem())
   > n13 = GetBlueprintForClass(Class=@n12)
   > PromptToCheckoutAndSaveAsset!(execute, Asset:Object Reference=@n13, bCheckDirty:bool=false, bPromptToSave:bool=false) -> (then)
   Event PreConstruct()
   > SetObject(self=Get MainAssetPicker() -> (MainAssetPicker:Single Property View Object Reference), NewObject=Self-Reference())
   > EnableAssetControls(Enable=false)
   Event Tick()
   > n14 = GetDirtyContentPackages!(execute) -> (then, OutDirtyPackages:Array of Package Object References)
   n15 = Get ItemBlueprintRef() -> (ItemBlueprintRef:Object Reference)
   > n16 = Branch(Condition=Contains Item(
     TargetArray:Array of Package Object References=@n14
     ItemToFind:"Package Object Reference (by ref)"=GetPackageForObject!(Object:Object Reference=@n15) -> (ReturnValue:Package Object Reference)
    ))
   n17 = Get Button_Save()
   > SetButtonName!(execute, self:EUW FAnim Simple Button Object Reference=@n17, NewName:Text="LOCTEXT(\"Save*\")") -> (then)
   SetButtonName(execute=@n16.else, self=@n17, NewName="LOCTEXT(\"Save\")")
   Event Construct()
  ]}
]
variables [
 CurrentItem {type=Class, default=null}
 CurrentItemCDO {type=FortWeapon, default=null}
 CurrentItemClass {type=Class, default=null}
 ItemBlueprintRef {type=Object, default=null}
]
widgets {
 info {rootWidgetClass=&EditorUtilityScrollBox, widgetCount=6}
 widgets [
  EditorUtilityScrollBox_35 {widgetClassPath=&EditorUtilityScrollBox, widget=EditorUtilityScrollBox_35}
  HorizontalBox_0 {widgetClassPath=&HorizontalBox, parent=EditorUtilityScrollBox_35, widget=HorizontalBox_0, slot {class=ScrollBoxSlot}}
  MainAssetPicker {
   widgetClassPath=&SinglePropertyView
   parent=HorizontalBox_0
   widget=MainAssetPicker
   properties {propertyName=CurrentItem}
   slot {class=HorizontalBoxSlot, properties {size {sizeRule=Fill, value=1}}}
  }
  Button_Open {
   widgetClassPath=&/EUW_FAnim_SimpleButton
   bIsVariable=true
   parent=HorizontalBox_0
   widget=Button_Open
   properties {buttonName=Open, toolTipText=Open currently selected blueprint asset.}
   slot {class=HorizontalBoxSlot, properties {size {sizeRule=Fill, value=0.33}}}
  }
  Button_Save {
   widgetClassPath=&EUW_FAnim_SimpleButton
   bIsVariable=true
   parent=HorizontalBox_0
   widget=Button_Save
   properties {buttonName=Save, toolTipText=Save currently selected asset.}
   slot {class=HorizontalBoxSlot, properties {size {sizeRule=Fill, value=0.33}}}
  }
  DetailsView_Item {
   widgetClassPath=&DetailsView
   parent=EditorUtilityScrollBox_35
   widget=DetailsView_Item
   properties {
    bAllowFiltering=true
    propertiesToShow [
     WeaponPawnAnimLayerOverlayClass
     ProceduralLayeringData
     EquipAnimation
     ReloadAnimation
     CustomReloadAnimationPerAmmoToFill
     PrimaryAbilityAnimation
     SecondaryAbilityAnimation
     WeaponEquipMontage
     WeaponReloadMontage
     CustomWeaponReloadMontagePerAmmoToFill
     WeaponPrimaryAbilityMontage
     SecondaryAbilityAnimation
     PlayerAnimAssets
     PoseOffsetAnimSequence
     PoseOffsetAnimSequenceFemaleOverride
     FirstTimeEquipAnimation
     WeaponFirstTimeEquipAnimation
     AnimationMode
     AnimClass
     WeaponPawnAnimSet
    ]
   }
   slot {class=ScrollBoxSlot}
  }
 ]
}
